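# AthenaSmallOfferPreview is a compact store offer preview panel showing a single item or bundle with header, price and reward tiles.
# OnOfferSet runs UpdateFromOffer, which rebuilds the item list and details, then sets price and sale state.
# Granted items feed a tile view, whose generated entries become reward widgets with owned state.
# Rarity or series colors drive dynamic materials on the header, rule, name font and glow.
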WidgetBlueprint /FortniteCoreUI/Frontend/AthenaSmallOfferPreview
graphs [
 CanGiftOffer {nodes [
   CanGiftOffer() {is=title K2Node_FunctionEntry}
   n1 = GetOfferInfo!(self:Athena Small Offer Preview Base Object Reference) -> (ReturnValue:Fort Store Front Offer Info Object Reference)
   > Return Node!(
    execute
    Result:EOfferPurchaseError Enum=CanGiftOffer!(
     self:Fort Mcp Context Object Reference=Get FortMcpContext!() -> (ReturnValue:Fort Mcp Context Object Reference) {is=title K2Node_GetBlueprintContextEx}
     OfferId:String=GetOfferId!(self:Fort Store Front Offer Info Object Reference=@n1) -> (ReturnValue:String)
     ReceiverAccountIds:Array of Strings=Get PlaceholderOneRecipient() -> (PlaceholderOneRecipient:Array of Strings)
     GiftWrapTemplateId:String
     UserMessageToRecipients:String
     PriceIndex:int=0
    ) -> (ReturnValue:EOfferPurchaseError Enum=NoError) {is=title K2Node_CallFunction}
   )
  ]}
 CreateAndAddRewardItem {nodes [
   CreateAndAddRewardItem!() -> (then, ItemQuantity:Fort Item Quantity Pair Structure, ItemIndex:int, PreviewPlaceholder:bool=~) {is=title K2Node_FunctionEntry}
   > n1 = Set ItemDef(ItemDef:Item Definition Base Object Reference=GetItemDefinitionFromItemQuantityPair!(ItemQuantityPair:"Fort Item Quantity Pair Structure (by ref)"=Get ItemQuantity() -> (ItemQuantity:Fort Item Quantity Pair Structure)) -> (ReturnValue:Item Definition Base Object Reference)) -> (Output_Get:Item Definition Base Object Reference)
   > n2 = Cast To FortMontageItemDefinitionBase!(execute, Object=@n1) -> (then, CastFailed, AsFort Montage Item Definition Base:Fort Montage Item Definition Base Object Reference) {is=title K2Node_DynamicCast}
   n3 = Branch(execute=[@n2.CastFailed, @n2.then], Condition=AND Boolean(A=IsValid(Object=@n2), B=Get PreviewPlaceholder() -> (PreviewPlaceholder:bool)))
   AddItem!(
    execute=@n3.else
    self:List View Object Reference=Get ItemRewardsTIleView() -> (ItemRewardsTIleView:Tile View Object Reference)
    Item:Object Reference=Get ItemDef() -> (ItemDef:Item Definition Base Object Reference)
   ) -> (then)
  ]}
 DoesOfferAllowQuantity {nodes [
   DoesOfferAllowQuantity() {is=title K2Node_FunctionEntry}
   n1 = GetOfferInfo()
   n2 = GetMaxConcurrentPurchases!(self:Fort Store Front Offer Info Object Reference=@n1) -> (ReturnValue:int=0)
   n3 = IsUnique!(self:Fort Store Front Offer Info Object Reference=@n1) -> (ReturnValue:bool)
   n4 = OfferHasDenyRequirements!(self:Fort Store Front Offer Info Object Reference=@n1) -> (ReturnValue:bool)
   n5 = GetQuantityRemaining!(self:Fort Store Front Offer Info Object Reference=@n1) -> (ReturnValue:int=0)
   > Return Node!(
    execute
    ReturnValue:bool=AND Boolean!(
     A=NOT Boolean(A=@n4)
     B=NOT Boolean(A=@n3)
     C:bool="Not Equal (Integer)"!(A:int=@n5, B:int=0) -> (ReturnValue:bool) {is=title K2Node_CallFunction}
     D:bool="Not Equal (Integer)"(A=@n2, B=1)
    ) -> (ReturnValue)
   )
  ]}
 EventGraph {nodes [
   n1 = "On Entry Generated (ItemRewardsTIleView)"!() -> (OutputDelegate, then, Widget:User Widget Object Reference) {is=title K2Node_ComponentBoundEvent}
   > HandleRewardItemGenerated!(execute, self, RewardItem:User Widget Object Reference=@n1.Widget) -> (then)
   Event OnOfferSet() {is=title K2Node_Event}
   > UpdateFromOffer()
   Event OnItemDefinitionChanged() {is=title K2Node_Event}
   > Is Valid(InputObject=GetItemDefinition!(self:Fort Item Base Widget Object Reference) -> (ReturnValue:Item Definition Base Object Reference))
   > UpdateCard()
  ]}
 GetCardSize {nodes [
   GetCardSize() {is=title K2Node_FunctionEntry}
   n1 = GetSubGame!(self:Fort Global UIContext Object Reference=Get FortGlobalUIContext!() -> (ReturnValue:Fort Global UIContext Object Reference) {is=title K2Node_GetBlueprintContextEx}) -> (ReturnValue:ESubGame Enum=Campaign)
   n2 = IsPrerollOffer!(self:Fort Store Front Offer Info Object Reference=GetOfferInfo()) -> (ReturnValue:bool)
   > Return Node(ReturnValue:EFortItemCardSize Enum=Select(
     Option 0:EFortItemCardSize Enum=Select(Option 0=L, Option 1=S, Index=@n2)
     Option 1:EFortItemCardSize Enum=M
     Index="Equal (Enum)"(A:ESubGame Enum=@n1, B:ESubGame Enum=Athena)
    ) -> (ReturnValue:EFortItemCardSize Enum))
  ]}
 GetDisplayName {nodes [
   n1 = GetDisplayName!() -> (then, ItemQuantityPair:"Fort Item Quantity Pair Structure (by ref)") {is=title K2Node_FunctionEntry}
   > Return Node(ReturnValue:Text=GetDisplayName!(self:Item Definition Base Object Reference=GetItemDefinitionFromItemQuantityPair(ItemQuantityPair=@n1)) -> (ReturnValue:Text) {is=title K2Node_CallFunction})
  ]}
 GetHasBattlepassStars {nodes [
   GetHasBattlepassStars() {is=title K2Node_FunctionEntry}
   n1 = GetOfferInfo()
   n2 = GetItemOfferCount!(self:Fort Store Front Offer Info Object Reference=@n1, TemplateId:String=AccountResource:athenabattlestar) -> (ReturnValue:int=0)
   n3 = GetMetaAsBool!(self:Fort Store Front Offer Info Object Reference=@n1, MetaTag:String=DontPreviewStars) -> (ReturnValue:bool)
   > Return Node(ReturnValue:bool=AND Boolean!(A=NOT Boolean(A=@n3), B="integer > integer"(A=@n2, B=0) {is=title K2Node_CallFunction}) -> (ReturnValue))
  ]}
 GetNumBattlepassStars {nodes [
   GetNumBattlepassStars() {is=title K2Node_FunctionEntry}
   > Return Node(ReturnValue:int=GetItemQuantity!(self:Fort Store Front Offer Info Object Reference=GetOfferInfo(), TemplateId:String=AccountResource:athenabattlestar) -> (ReturnValue:int=0))
  ]}
 GetOfferDetailsSlotMargin {nodes [
   GetOfferDetailsSlotMargin() {is=title K2Node_FunctionEntry}
   n1 = GetSubGame(self=Get FortGlobalUIContext()) -> (ReturnValue=Campaign)
   > Return Node!(
    execute
    Margin:Margin Structure=Make Margin(
     Left=0
     Top=Select!(Campaign:double=96, Athena:double=0, Invalid:double=0, Count:double=0, Creative:double=0, Index:ESubGame Enum=@n1) -> (ReturnValue:double)
     Right=0
     Bottom=0
    )
   )
  ]}
 GetOwnedIndex {nodes [
   GetOwnedIndex() {is=title K2Node_FunctionEntry}
   n1 = GetQuantityRemaining(self=GetOfferInfo()) -> (ReturnValue=0)
   > Return Node!(
    execute
    ReturnValue:int=Select!(Option 0:int=0, Option 1:int=1, Index:bool="Equal (Integer)"(A=@n1, B=0) {is=title K2Node_CallFunction}) -> (ReturnValue:int)
   )
  ]}
 GetSalesText {nodes [
   GetSalesText() {is=title K2Node_FunctionEntry}
   n1 = GetOfferInfo()
   n2 = GetSaleText!(self:Fort Store Front Offer Info Object Reference=@n1, PriceIndex:int=0, ItemQuantity:int=1) -> (ReturnValue:Text)
   > Return Node!(
    execute
    NewParam:Text=Select!(
     UndecoratedNewPrice:Text
     AmountOff:Text=Format Text!(Format="LOCTEXT(\"{0} V-Bucks Off\")", 0:Text=@n2) -> (Result)
     PercentOff:Text=Format Text(Format="LOCTEXT(\"{0} Off\")", 0=@n2)
     PercentOn:Text=Format Text(Format="LOCTEXT(\"{0} On\")", 0=@n2)
     Strikethrough:Text
     Index:ECatalogSaleType Enum=GetSaleType!(self:Fort Store Front Offer Info Object Reference=@n1, PriceIndex:int=0) -> (ReturnValue:ECatalogSaleType Enum=NotOnSale)
    ) -> (ReturnValue:Text)
   )
  ]}
 GetShowCard {nodes [
   GetShowCard() {is=title K2Node_FunctionEntry}
   n1 = GetOfferInfo()
   > Return Node!(execute, Result:bool=OR Boolean(A=IsSingleGrantedItem!() -> (Result:bool), B=IsPrerollOffer(self=@n1)))
  ]}
 GetShowDescription {nodes [
   GetShowDescription() {is=title K2Node_FunctionEntry}
   n1 = Get FortGlobalUIContext()
   > Return Node!(execute, ReturnValue:bool="Equal (Enum)"(A=GetSubGame(self=@n1) -> (ReturnValue=Campaign), B=Athena))
  ]}
 GetShowDetails {nodes [
   GetShowDetails() {is=title K2Node_FunctionEntry}
   n1 = GetOfferInfo()
   n2 = GetSubGame(self=Get FortGlobalUIContext()) -> (ReturnValue=Campaign)
   > Return Node(ReturnValue=OR Boolean(A="Equal (Enum)"(A=@n2, B=Athena), B=IsPrerollOffer(self=@n1)))
  ]}
 GetSortedItems {nodes [
   GetSortedItems() {is=title K2Node_FunctionEntry}
   > Return Node!(
    execute
    ItemQuantityPairs:Array of Fort Item Quantity Pair Structures=GetGrantedItemQuantityPairs!(self:Fort Store Front Offer Info Object Reference=GetOfferInfo(), bPrioritizeUnowned:bool=true, bCombineCurrencyItems:bool=true) -> (ItemQuantityPairs:Array of Fort Item Quantity Pair Structures)
   )
  ]}
 HandleRewardItemGenerated {nodes [
   n1 = HandleRewardItemGenerated!() -> (then, RewardItem:User Widget Object Reference) {is=title K2Node_FunctionEntry}
   > n2 = Cast To AthenaStoreReward!(execute, Object=@n1) -> (then, CastFailed, AsAthena Store Reward:Athena Store Reward Object Reference) {is=title K2Node_DynamicCast}
   > n3 = Set StoreReward(StoreReward:Athena Store Reward Object Reference=@n2) -> (Output_Get:Athena Store Reward Object Reference)
   > n4 = Cast To FortItemDefinition!(execute, Object=GetListItemObject!(UserObjectListEntry:User Object List Entry Interface=@n3) -> (ReturnValue:Object Reference)) -> (then, CastFailed, AsFort Item Definition:Fort Item Definition Object Reference) {is=title K2Node_DynamicCast}
   > n5 = FindItemQuantityPair!(
    execute
    ItemDefinition:Fort Item Definition Object Reference=@n4
    InItemQuantityPairs:Array of Fort Item Quantity Pair Structures=Get SortedItemQuantities() -> (SortedItemQuantities:Array of Fort Item Quantity Pair Structures)
   ) -> (then, OutItemQuantity:Fort Item Quantity Pair Structure, ReturnValue:bool)
   > Branch(Condition=@n5.ReturnValue)
   > SetupRewardItem!(
    execute
    self
    RewardItem:Athena Store Reward Object Reference=Get StoreReward() -> (StoreReward:Athena Store Reward Object Reference)
    PreviewPlaceholder:bool=false
    ItemQuantity:Fort Item Quantity Pair Structure=@n5
   ) -> (then)
   Reroute Node(InputPin:Fort Item Definition Object Reference=@n4) -> (OutputPin:Fort Item Definition Object Reference)
  ]}
 IsDynamicBundle {nodes [
   IsDynamicBundle() {is=title K2Node_FunctionEntry}
   n1 = GetOfferInfo()
   > Return Node!(
    execute
    ReturnValue:bool="Equal (Enum)"(
     A:ECatalogOfferType Enum=GetOfferType!(self:Fort Store Front Offer Info Object Reference=@n1) -> (ReturnValue:ECatalogOfferType Enum=StaticPrice)
     B:ECatalogOfferType Enum=DynamicBundle
    )
   )
  ]}
 IsItemDefOwned {nodes [
   n1 = IsItemDefOwned!() -> (then, ItemDef:Item Definition Base Object Reference) {is=title K2Node_FunctionEntry}
   > Return Node(ReturnValue=IsItemDefOwned!(self:Fort Store Front Offer Info Object Reference=GetOfferInfo(), ItemDef:Item Definition Base Object Reference=@n1) -> (ReturnValue:bool) {is=title K2Node_CallFunction})
  ]}
 IsNonRefundable {nodes [
   IsNonRefundable() {is=title K2Node_FunctionEntry}
   n1 = IsRefundable!(self:Fort Store Front Offer Info Object Reference=GetOfferInfo()) -> (ReturnValue:bool)
   n2 = GetSubGame(self=Get FortGlobalUIContext()) -> (ReturnValue=Campaign)
   > Return Node(ReturnValue=AND Boolean(A=NOT Boolean(A=@n1), B="Equal (Enum)"(A:ESubGame Enum=@n2, B:ESubGame Enum=Athena)))
  ]}
 IsSingleGrantedItem {nodes [
   IsSingleGrantedItem() {is=title K2Node_FunctionEntry}
   n1 = GetOfferInfo()
   > Return Node!(
    execute
    Result:bool="Equal (Integer)"(A=GetNumGrantedItems!(self:Fort Store Front Offer Info Object Reference=@n1) -> (ReturnValue:int=0), B=1)
   )
  ]}
 OverrideHeaderAndDescription {nodes [
   n1 = OverrideHeaderAndDescription!() -> (then, Header:Text, Description:Text, ItemType:Text, bUserAlternateHeaderBG:bool=~, Rarity:EFortRarity Enum=Common) {is=title K2Node_FunctionEntry}
   > "SetText (Text)"(
    self=Get NameTextBlock() -> (NameTextBlock:Common Text Object Reference)
    InText=TextToUpper!(InText:"Text (by ref)"=@n1) -> (ReturnValue:Text)
   )
   > SetText(self=Get RichDescriptionBox() -> (RichDescriptionBox:Common Rich Text Block Object Reference), InText=@n1.Description)
   > "SetText (Text)"(self=Get TypeTextBlock() -> (TypeTextBlock:Common Text Object Reference), InText=@n1.ItemType)
   > SetVisibility(
    self=Get RarityGlow() -> (RarityGlow:Image Object Reference)
    InVisibility=Select!(Option 0:ESlateVisibility Enum=Visible, Option 1:ESlateVisibility Enum=Collapsed, Index:bool=@n1.bUserAlternateHeaderBG) -> (ReturnValue:ESlateVisibility Enum)
   )
   > n2 = GetRarityColorPalette!(execute, Rarity:EFortRarity Enum=@n1.Rarity) -> (then, ReturnValue:Fort Color Palette Structure)
   > UpdateColorRarity!(execute, self, Color:Fort Color Palette Structure=@n2) -> (then)
  ]}
 Owned Bundle Item Count {nodes [
   Owned Bundle Item Count() {is=title K2Node_FunctionEntry}
   n1 = GetDynamicBundleItems!(self:Fort Store Front Offer Info Object Reference=GetOfferInfo()) -> (Items:Array of Bundled Item Info Structures)
   > n2 = For Each Loop!(Exec, Array:Array of Bundled Item Info Structures=@n1) -> (LoopBody, Array Element_TemplateId:String, Array Element_Quantity:int, Array Element_bOwned:bool=~, Array Index, Completed)
   > Branch(Condition=@n2.Array Element_bOwned)
   > Increment Int(Value="Get Owned Bundle Items (local)"() -> ("Owned Bundle Items (local)":int=0))
   n3 = "Get Owned Bundle Items (local)"() -> ("Owned Bundle Items (local)"=0)
   n4 = Length(TargetArray:Array of Bundled Item Info Structures=@n1)
   Return Node!(
    execute=@n2.Completed
    Owned Items:int=@n3
    UnOwned Items:int=int - int(A=@n4, B="Get Owned Bundle Items (local)"() -> ("Owned Bundle Items (local)"=0)) {is=title K2Node_CallFunction}
    bNoneOwned:bool="Equal (Integer)"(A=@n3, B=0)
    bAllOwned:bool="Equal (Integer)"(A=@n3, B=@n4)
   )
  ]}
 SetTypeText {nodes [
   n1 = SetTypeText!() -> (then, RarityText:Text) {is=title K2Node_FunctionEntry}
   > "SetText (Text)"(self=Get ItemRarityOrItemsCount() -> (ItemRarityOrItemsCount:Common Text Object Reference), InText=@n1)
   > SetVisibility(
    self=Get RarityTypeSeperatorImage() -> (RarityTypeSeperatorImage:Image Object Reference)
    InVisibility=Select(Option 0=SelfHitTestInvisible, Option 1=Collapsed, Index=TextIsEmpty(InText=@n1))
   )
  ]}
 SetupPrerollItemHeaderInfo {nodes [
   SetupPrerollItemHeaderInfo() {is=title K2Node_FunctionEntry}
   > Branch(Condition=IsPrerollOffer(self=GetOfferInfo()))
   n1 = GetPrerollOfferItem!(self:Fort Store Front Offer Info Object Reference=GetOfferInfo()) -> (ItemQuantityPair:Fort Item Quantity Pair Structure, ReturnValue:bool)
   > Set PrerollOfferItem(PrerollOfferItem:Fort Item Quantity Pair Structure=@n1) -> (Output_Get:Fort Item Quantity Pair Structure)
   > Branch(Condition=@n1.ReturnValue)
   > SetItemQuantityPairToRepresent!(
    execute
    self:Fort Item Base Widget Object Reference
    InItemQuantityPair:Fort Item Quantity Pair Structure=Get PrerollOfferItem() -> (PrerollOfferItem:Fort Item Quantity Pair Structure)
   ) -> (then, ReturnValue:bool)
   > OverrideHeaderAndDescription!(
    execute
    self
    Header:Text=GetDisplayName!(ItemQuantityPair:"Fort Item Quantity Pair Structure (by ref)"=Get PrerollOfferItem()) -> (ReturnValue)
    Description:Text
    ItemType:Text
    bUserAlternateHeaderBG:bool=false
    Rarity:EFortRarity Enum=Common
   ) -> (then) {is=title K2Node_CallFunction}
  ]}
 SetupPrice {nodes [
   SetupPrice!() -> (
    then
    PriceContainer:Widget Object Reference
    NormalPriceText:Text Object Reference
    SalePriceContainer:Widget Object Reference
    SalePriceText:Text Object Reference
    CurrencyImage:Image Object Reference
   ) {is=title K2Node_FunctionEntry}
   n1 = GetOfferInfo()
   > n2 = Branch(Condition=IsPriceInRealMoney!(self:Fort Store Front Offer Info Object Reference=@n1, PriceIndex:int=0) -> (ReturnValue:bool))
   > SetVisibility(self=Get PriceContainer() -> (PriceContainer:Widget Object Reference), InVisibility=SelfHitTestInvisible)
   > "SetText (Text)"(
    self=Get NormalPriceText() -> (NormalPriceText:Text Object Reference)
    InText=GetRegularPrice!(self:Fort Store Front Offer Info Object Reference=GetOfferInfo(), PriceIndex:int=0, ItemQuantity:int=1) -> (ReturnValue:Text)
   )
   n3 = GetSalePrice!(self:Fort Store Front Offer Info Object Reference=GetOfferInfo(), PriceIndex:int=0, ItemQuantity:int=1) -> (SalePrice:Text, ReturnValue:bool)
   > SetVisibility(
    self=Get SalePriceContainer() -> (SalePriceContainer:Widget Object Reference)
    InVisibility=Select(Option 0=Collapsed, Option 1=SelfHitTestInvisible, Index=@n3.ReturnValue)
   )
   > "SetText (Text)"(self=Get SalePriceText() -> (SalePriceText:Text Object Reference), InText=@n3)
   > SetVisibility(self=Get CurrencyImage() -> (CurrencyImage:Image Object Reference), InVisibility=Collapsed)
   n4 = Is Valid(
    exec=@n2.else
    InputObject=GetPriceItem!(self:Fort Store Front Offer Info Object Reference=@n1, PriceIndex:int=0) -> (RequiredItemCount:int=0, ReturnValue:Fort Account Item Definition Object Reference).ReturnValue
   )
   SetVisibility(execute=@n4.Is Not Valid, self=Get PriceContainer(), InVisibility=Collapsed)
   n5 = Branch(execute=@n4.Is Valid, Condition=IsVotingOffer!(self:Fort Store Front Offer Info Object Reference=@n1) -> (ReturnValue:bool))
   n6 = Get PriceContainer()
   > SetVisibility(self=@n6, InVisibility=Collapsed)
   n7 = GetRegularPrice(self=GetOfferInfo(), PriceIndex=0, ItemQuantity=1)
   > "SetText (Text)"(self=Get NormalPriceText(), InText=@n7)
   > SetVisibility(self=Get SalePriceContainer(), InVisibility=Collapsed)
   n8 = Get SalePriceText()
   n9 = GetSalePrice(self=GetOfferInfo(), PriceIndex=0, ItemQuantity=1)
   > "SetText (Text)"(self=@n8, InText=@n9)
   > SetVisibility(self=Get CurrencyImage(), InVisibility=SelfHitTestInvisible)
   n10 = GetSmallPreviewImageBrush!(self:Fort Item Definition Object Reference=GetPriceItem(self=GetOfferInfo(), PriceIndex=0) -> (RequiredItemCount=0).ReturnValue) -> (ReturnValue:Slate Brush Structure)
   > SetBrush(self=Get CurrencyImage(), InBrush=@n10)
   SetVisibility(execute=@n5.else, self=@n6, InVisibility=SelfHitTestInvisible)
   > "SetText (Text)"(self=Get NormalPriceText(), InText=@n7)
   > SetVisibility(self=Get SalePriceContainer(), InVisibility=Select(Option 0=Collapsed, Option 1=SelfHitTestInvisible, Index=@n9.ReturnValue))
   > "SetText (Text)"(self=@n8, InText=@n9)
   > SetVisibility(self=Get CurrencyImage(), InVisibility=SelfHitTestInvisible)
   > SetBrush(self=Get CurrencyImage(), InBrush=@n10)
  ]}
 SetupRewardItem {nodes [
   n1 = SetupRewardItem!() -> (then, RewardItem:Athena Store Reward Object Reference, PreviewPlaceholder:bool=~, ItemQuantity:Fort Item Quantity Pair Structure) {is=title K2Node_FunctionEntry}
   > Set StoreRewardWidget(StoreRewardWidget:Athena Store Reward Object Reference=@n1) -> (Output_Get:Athena Store Reward Object Reference)
   > n2 = Set ItemDef(ItemDef=GetItemDefinitionFromItemQuantityPair(ItemQuantityPair=@n1.ItemQuantity))
   > SetupRewardItem!(
    execute
    self:Fort Athena Reward Object Reference=Get StoreRewardWidget() -> (StoreRewardWidget:Athena Store Reward Object Reference)
    ItemDefToRepresent:Item Definition Base Object Reference=@n2
    Quantity:int=Break Fort Item Quantity Pair!(FortItemQuantityPair:"Fort Item Quantity Pair Structure (by ref)"=Get ItemQuantity()) -> (Quantity:int) {is=title K2Node_BreakStruct}
    ItemVM:Fort Item VM Object Reference
   ) -> (then) {is=title K2Node_CallFunction}
   > UpdateRewardState!(
    execute
    self:Fort Athena Reward Object Reference=Get StoreRewardWidget()
    StateToSetup:EFortAthenaRewardState Enum=LevelAchieved
    bEnabled:bool=true
   ) -> (then)
   > UpdateRewardState(self=Get StoreRewardWidget(), StateToSetup=Owned, bEnabled=IsItemDefOwned!(ItemDef=Get ItemDef()) -> (ReturnValue))
   > SetViewOnHovered!(execute, self:Fort Athena Reward Object Reference=Get StoreRewardWidget(), bShouldViewOnHovered:bool=true) -> (then)
   > SetVisibility(
    self=Get MultipleItemCB() -> (MultipleItemCB:Common Border Object Reference)
    InVisibility=Select(Option 0=Visible, Option 1=Collapsed, Index=Get PreviewPlaceholder())
   )
   > Branch(Condition="Equal (Object)"(A=Get ItemDef(), B=Get SelectedItemDefinition() -> (SelectedItemDefinition:Fort Item Definition Object Reference)) {is=title K2Node_CallFunction})
   > SetVisibility(self=Get MultipleItemCB(), InVisibility=SelfHitTestInvisible)
  ]}
 SetupSaleInfo {nodes [
   SetupSaleInfo!() -> (then, bHasSale:bool=~, BannerOverrideTag:String, NormalPrice:Text, bHasBattlePassStars:bool=~, NumOfBattlePassStars:int) {is=title K2Node_FunctionEntry}
   > "SetText (Text)"(self=Get TextOriginalPrice1() -> (TextOriginalPrice1:Common Text Object Reference), InText=Get NormalPrice() -> (NormalPrice:Text))
   > SetVisibility(
    self=Get O_Sale_PreviousPrice() -> (O_Sale_PreviousPrice:Overlay Object Reference)
    InVisibility=Select(Option 0=Collapsed, Option 1=SelfHitTestInvisible, Index=Get bHasSale() -> (bHasSale:bool))
   )
  ]}
 Short Bundle Description {nodes [
   Short Bundle Description() {is=title K2Node_FunctionEntry}
   n1 = GetDynamicBundleItems(self=GetOfferInfo())
   > Return Node!(
    execute
    Text:Text=Format Text!(Format="LOCTEXT(\"{total bundle items} ITEM BUNDLE\")", total bundle items:int=Length(TargetArray=@n1)) -> (Result)
   )
  ]}
 UpdateCard {nodes [
   UpdateCard() {is=title K2Node_FunctionEntry}
   n1 = GetItemDefinition()
   > n2 = Is Valid(InputObject=@n1)
   > SetItemToRepresent!(
    execute
    self:Fort Cosmetic Item Card Object Reference=Get Card() -> (Card:Fort Cosmetic Item Card Object Reference)
    InItemToRepresent:Fort Item Object Reference=CreateTemporaryItemInstance!(
     ItemDefinition:Item Definition Base Object Reference=@n1
     Count:int=GetQuantity!(self:Fort Item Base Widget Object Reference) -> (ReturnValue:int=0)
     Level:int=-1
    ) -> (ReturnValue:Fort Item Object Reference)
    InItemVM:Fort Item VM Object Reference
   ) -> (then)
   SetItemToRepresent(execute=@n2.Is Not Valid, self=Get Card())
  ]}
 UpdateColorRarity {nodes [
   n1 = UpdateColorRarity!() -> (then, Color:Fort Color Palette Structure) {is=title K2Node_FunctionEntry}
   > n2 = Sequence!(execute) -> (then_0, then_1, then_2:Exec)
   n3 = GetDynamicMaterial(execute=@n2.then_2, self:Border Object Reference=Get Border_Header() -> (Border_Header:Common Border Object Reference))
   n4 = Break Fort Color Palette!(FortColorPalette:"Fort Color Palette Structure (by ref)"=@n1) -> (
    Color1:Linear Color Structure
    Color2:Linear Color Structure
    Color3:Linear Color Structure
    Color4:Linear Color Structure
    Color5:Linear Color Structure
   ) {is=title K2Node_BreakStruct}
   > SetVectorParameterValue(self=@n3, ParameterName=RGBA0, Value=@n4.Color3)
   > SetVectorParameterValue(self=@n3, ParameterName=RGBA1, Value=@n4.Color2)
   > n5 = GetDynamicMaterial(self=Get HeaderRule() -> (HeaderRule:Common Border Object Reference))
   > SetVectorParameterValue(self=@n5, ParameterName=RGBA0, Value=@n4.Color3)
   > SetVectorParameterValue(self=@n5, ParameterName=RGBA1, Value=@n4.Color2)
   n6 = Get NameTextBlock()
   SetStyle!(
    execute=@n2.then_0
    self:Common Text Object Reference=@n6
    InStyle:Common Text Style Class Reference=/FortUILibrary/Styles/Typography/Foundation/TextStyle-Header-M-SimpleGradient.TextStyle-Header-M-SimpleGradient_C
   ) -> (then)
   > n7 = GetDynamicFontMaterial!(execute, self:Text Object Reference=@n6) -> (then, ReturnValue:Material Instance Dynamic Object Reference)
   > SetVectorParameterValue(self=@n7, ParameterName=RGB1, Value=@n4)
   > SetColorAndOpacity(self=Get ItemRarityOrItemsCount(), InColorAndOpacity=Make Slate Color(SpecifiedColor=@n4))
   n8 = GetDynamicMaterial(execute=@n2.then_1, self:Image Object Reference=Get RarityGlow())
   > SetVectorParameterValue(self=@n8, ParameterName=vRGB, Value=@n4)
  ]}
 UpdateDetails {nodes [
   UpdateDetails!() -> (then, Reward:Fort Item Object Reference) {is=title K2Node_FunctionEntry}
   > n1 = Branch(Condition=IsDynamicBundle!() -> (ReturnValue:bool) {is=title K2Node_CallFunction})
   > UpdateDetailsForBundle()
   UpdateDetailsForSingleItem!(
    execute=@n1.else
    self
    ItemDefinition:Fort Account Item Definition Object Reference=GetFirstGrantItemDefinition!(self:Fort Store Front Offer Info Object Reference=GetOfferInfo()) -> (ReturnValue:Fort Account Item Definition Object Reference)
   ) -> (then)
  ]}
 UpdateDetailsForBundle {nodes [
   UpdateDetailsForBundle() {is=title K2Node_FunctionEntry}
   > n1 = Short Bundle Description!(execute, self) -> (then, Text:Text) {is=title K2Node_CallFunction}
   > "SetText (Text)"(self=Get TypeTextBlock(), InText=@n1)
   > "SetText (Text)"(self=Get NameTextBlock(), InText=GetName!(self:Fort Store Front Offer Info Object Reference=GetOfferInfo()) -> (ReturnValue:Text))
   > SetTypeText!(execute, self, RarityText:Text) -> (then) {is=title K2Node_CallFunction}
   n2 = GetOfferInfo()
   > UpdateColorRarity(Color=GetRarityOrSeriesColors!(self:Fort Item Definition Object Reference=GetFirstGrantItemDefinition(self=@n2)) -> (ReturnValue:Fort Color Palette Structure)) {is=title K2Node_CallFunction}
  ]}
 UpdateDetailsForSingleItem {nodes [
   n1 = UpdateDetailsForSingleItem!() -> (then, ItemDefinition:Fort Account Item Definition Object Reference) {is=title K2Node_FunctionEntry}
   > SetText(self=Get RichDescriptionBox(), InText=GetItemRichDescription!(self:Item Definition Base Object Reference=@n1) -> (ReturnValue:Text))
   > "SetText (Text)"(self=Get NameTextBlock(), InText=GetDisplayName!(self:Item Definition Base Object Reference=@n1) -> (ReturnValue))
   > n2 = GetRarityDataForItemDef!(execute, ItemDef:Item Definition Base Object Reference=@n1) -> (then, ReturnValue:Fort Rarity Item Data Structure)
   > SetTypeText(RarityText=Break Fort Rarity Item Data!(FortRarityItemData:"Fort Rarity Item Data Structure (by ref)"=@n2) -> (
     Name:Text
     Color1:Linear Color Structure
     Color2:Linear Color Structure
     Color3:Linear Color Structure
     Color4:Linear Color Structure
     Color5:Linear Color Structure
     Radius:float=~
     Falloff:float=~
     Brightness:float=~
     Roughness:float=~
     Glow:float=~
    ) {is=title K2Node_BreakStruct})
   > UpdateColorRarity(Color=GetRarityOrSeriesColors(self=@n1))
   > n3 = Branch(Condition="Equal (Enum)"(
     A:EFortItemType Enum=GetItemType!(self:Fort Item Definition Object Reference=@n1) -> (ReturnValue:EFortItemType Enum=WorldItem)
     B:EFortItemType Enum=Currency
    ))
   n4 = Get TypeTextBlock()
   > n5 = "SetText (Text)"(self=@n4, InText=GetItemTypeName!(self:Fort Item Definition Object Reference=@n1, bUsePlural:bool=false) -> (ReturnValue:Text))
   SetText(execute=[@n6.then, @n5.then], self=Get RichDescriptionBox(), InText=GetItemRichDescription(self=@n1))
   > UpdateCard() {is=title K2Node_CallFunction}
   n6 = "SetText (Text)"(execute=@n3.else, self=@n4, InText=GetItemShortDescription!(self:Item Definition Base Object Reference=@n1) -> (ReturnValue:Text))
  ]}
 UpdateFromOffer {nodes [
   UpdateFromOffer() {is=title K2Node_FunctionEntry}
   > Branch(Condition=IsValidOffer!(self:Fort Store Front Offer Info Object Reference=GetOfferInfo()) -> (ReturnValue:bool))
   > n1 = Sequence()
   SetupPrice!(
    execute=@n1.then_1
    self
    PriceContainer:Widget Object Reference=Get HorizontalBox_PurchasePrice() -> (HorizontalBox_PurchasePrice:Horizontal Box Object Reference)
    NormalPriceText:Text Object Reference=Get TextOriginalPrice1()
    SalePriceContainer:Widget Object Reference=Get O_Sale_PreviousPrice()
    SalePriceText:Text Object Reference=Get TextOfferDisplayPrice1() -> (TextOfferDisplayPrice1:Common Text Object Reference)
    CurrencyImage:Image Object Reference=Get Currency1ItemImage2() -> (Currency1ItemImage2:Image Object Reference)
   ) -> (then) {is=title K2Node_CallFunction}
   UpdateSaleStatus(execute=@n1.then_2)
   UpdateItemList(execute=@n1.then_0)
   > UpdateDetails!(execute, self, Reward:Fort Item Object Reference) -> (then) {is=title K2Node_CallFunction}
  ]}
 UpdateItemList {nodes [
   UpdateItemList() {is=title K2Node_FunctionEntry}
   > ClearListItems(self=Get ItemRewardsTIleView())
   n1 = IsSingleGrantedItem() {is=title K2Node_CallFunction}
   > SetActiveWidgetIndex(
    self:Common Visibility Switcher Object Reference=Get BundleSwitcher() -> (BundleSwitcher:Common Visibility Switcher Object Reference)
    Index=SelectInt(A=0, B=1, bPickA=@n1)
   )
   > SetVisibility(self=Get MultipleItemCB(), InVisibility=Select(Option 0=SelfHitTestInvisible, Option 1=Collapsed, Index=@n1))
   > n2 = Branch(Condition=@n1)
   n3 = GetGrantedItemQuantityPair!(self:Fort Store Front Offer Info Object Reference=GetOfferInfo(), GrantedItemIndex:int=0) -> (ItemQuantity:Fort Item Quantity Pair Structure, ReturnValue:bool)
   > Branch(Condition=@n3.ReturnValue)
   > CreateAndAddRewardItem!(execute, self, ItemQuantity:Fort Item Quantity Pair Structure=@n3, ItemIndex:int=0, PreviewPlaceholder:bool=true) -> (then) {is=title K2Node_CallFunction}
   > SetItemQuantityPairToRepresent(InItemQuantityPair=@n3)
   SetupPrerollItemHeaderInfo(execute=@n2.else) {is=title K2Node_CallFunction}
   > n4 = Set SortedItemQuantities(SortedItemQuantities:Array of Fort Item Quantity Pair Structures=GetSortedItems!() -> (ItemQuantityPairs:Array of Fort Item Quantity Pair Structures) {is=title K2Node_CallFunction}) -> (Output_Get:Array of Fort Item Quantity Pair Structures)
   > n5 = For Each Loop!(Exec, Array:Array of Fort Item Quantity Pair Structures=@n4) -> (LoopBody, Array Element:Fort Item Quantity Pair Structure, Array Index, Completed)
   RegenerateAllEntries(execute=@n5.Completed, self:List View Base Object Reference=Get ItemRewardsTIleView())
   CreateAndAddRewardItem(execute=@n5.LoopBody, ItemQuantity=@n5, ItemIndex=@n5.Array Index, PreviewPlaceholder=false)
  ]}
 UpdateSaleStatus {nodes [
   UpdateSaleStatus() {is=title K2Node_FunctionEntry}
   n1 = GetOfferInfo()
   > Set OfferInfo(OfferInfo:Fort Store Front Offer Info Object Reference=@n1) -> (Output_Get:Fort Store Front Offer Info Object Reference)
   n2 = Get OfferInfo() -> (OfferInfo:Fort Store Front Offer Info Object Reference)
   n3 = IsOfferOwned!(self:Fort Store Front Offer Info Object Reference=@n1) -> (ReturnValue:bool)
   n4 = GetSaleType(self=Get OfferInfo(), PriceIndex=0) -> (ReturnValue=NotOnSale)
   > Set IsSaleItem(IsSaleItem:bool=AND Boolean!(
     A=NOT Boolean(A=@n3)
     B=GetSalePrice(self=@n2, PriceIndex=0, ItemQuantity=1).ReturnValue
     C:bool="Not Equal (Enum)"!(A:ECatalogSaleType Enum=@n4, B:ECatalogSaleType Enum=NotOnSale) -> (ReturnValue:bool)
     D:bool="Not Equal (Enum)"(A=@n4, B=UndecoratedNewPrice)
    ) -> (ReturnValue)) -> (Output_Get:bool)
   n5 = Get OfferInfo()
   > n6 = Branch(Condition="Equal (Enum)"(A:ECatalogSaleType Enum=GetSaleType(self=@n5, PriceIndex=0) -> (ReturnValue=NotOnSale), B:ECatalogSaleType Enum=NotOnSale))
   > SetupSaleInfo!(
    execute
    self
    bHasSale:bool=Get IsSaleItem() -> (IsSaleItem:bool)
    BannerOverrideTag:String=GetBannerOverrideTag!(self:Fort Store Front Offer Info Object Reference=Get OfferInfo()) -> (ReturnValue:String)
    NormalPrice:Text=GetRegularPrice(self=Get OfferInfo(), PriceIndex=0, ItemQuantity=1)
    bHasBattlePassStars:bool=GetHasBattlepassStars!() -> (ReturnValue:bool) {is=title K2Node_CallFunction}
    NumOfBattlePassStars:int=GetNumBattlepassStars!() -> (ReturnValue:int=0) {is=title K2Node_CallFunction}
   ) -> (then) {is=title K2Node_CallFunction}
   SetupSaleInfo(
    execute=@n6.else
    bHasSale=Get IsSaleItem()
    BannerOverrideTag=GetBannerOverrideTag(self=Get OfferInfo())
    NormalPrice=GetRegularPrice(self=Get OfferInfo(), PriceIndex=0, ItemQuantity=1)
    bHasBattlePassStars=GetHasBattlepassStars()
    NumOfBattlePassStars=GetNumBattlepassStars() -> (ReturnValue=0)
   )
  ]}
 UpdateTileViewSize {nodes [
   UpdateTileViewSize() {is=title K2Node_FunctionEntry}
   > SetEntryHeight!(execute, self:Tile View Object Reference=Get ItemRewardsTIleView(), NewHeight:float=150) -> (then)
   > SetEntryWidth!(execute, self:Tile View Object Reference=Get ItemRewardsTIleView(), NewWidth:float=120) -> (then)
  ]}
 UpdateViewedItem {nodes [
   UpdateViewedItem!() -> (then, ViewedItem:Fort Item Object Reference) {is=title K2Node_FunctionEntry}
   n1 = GetOfferInfo()
   > n2 = Branch(Condition=NOT Boolean(A=IsPrerollOffer(self=@n1)))
   > n3 = SetItemDefinitionToRepresent!(
    execute
    self:Fort Item Base Widget Object Reference
    InItemDefinition:Fort Item Definition Object Reference=GetItemDefinition(self:Fort Item Object Reference=Get ViewedItem() -> (ViewedItem:Fort Item Object Reference)) -> (ReturnValue:Fort Item Definition Object Reference)
    InItem:Fort Item VM Object Reference
   ) -> (then)
   n4 = GetOfferInfo()
   n5 = GetSubGame!() -> (ReturnValue=Campaign)
   Branch(
    execute=[@n3.then, @n2.else]
    Condition=AND Boolean!(A="Equal (Enum)"(A:ESubGame Enum=@n5, B:ESubGame Enum=Campaign), B=IsPrerollOffer(self=@n4)) -> (ReturnValue)
   )
   > UpdateDetails(Reward=Get ViewedItem())
  ]}
]
variables [
 SelectedItemDefinition {type=FortItemDefinition, default=null}
 SortedItemQuantities {type="array<FortItemQuantityPair>", default []}
]
widgets {
 info {inheritedWidgetCount=0, parentClass=/Script/FortniteUI.AthenaSmallOfferPreviewBase, rootWidgetClass=&SizeBox, widgetCount=44}
 widgets [
  SizeBox_0 {widgetClassPath=&SizeBox, widget=SizeBox_0, properties {bOverride_WidthOverride=true, widthOverride=400}}
  Overlay_3 {widgetClassPath=&Overlay, parent=SizeBox_0, widget=Overlay_3, slot {class=SizeBoxSlot}}
  ColorBackground {
   widgetClassPath=&CommonBorder
   bIsVariable=true
   parent=Overlay_3
   widget=ColorBackground
   properties {style=/FortUILibrary/Styles/Border/Border_Solid_MdBlue.Border_Solid_MdBlue_C, brushColor=<rgba 1 1 1 0.5>, padding=<ltrb 0 0 0 0>}
   slot {class=OverlaySlot, properties {horizontalAlignment=HAlign_Fill, verticalAlignment=VAlign_Fill}}
  }
  VerticalBoxItemDetails {
   widgetClassPath=&VerticalBox
   bIsVariable=true
   parent=Overlay_3
   widget=VerticalBoxItemDetails
   slot {class=OverlaySlot, properties {horizontalAlignment=HAlign_Fill, verticalAlignment=VAlign_Fill}}
  }
  VerticalBox_1 {widgetClassPath=&VerticalBox, parent=VerticalBoxItemDetails, widget=VerticalBox_1, slot {class=VerticalBoxSlot}}
  Border_Header {
   widgetClassPath=&CommonBorder
   bIsVariable=true
   parent=VerticalBox_1
   widget=Border_Header
   properties {
    style=/FortUILibrary/Styles/Border/Border-AthenaItemInfoHeader.Border-AthenaItemInfoHeader_C
    verticalAlignment=VAlign_Center
    padding=<ltrb 12 3 12 0>
   }
   slot {class=VerticalBoxSlot}
  }
  VerticalBox_3 {
   widgetClassPath=&VerticalBox
   parent=Border_Header
   widget=VerticalBox_3
   slot {class=BorderSlot, properties {verticalAlignment=VAlign_Center, padding=<ltrb 12 3 12 0>}}
  }
  ScaleBox_1 {
   widgetClassPath=&ScaleBox
   parent=VerticalBox_3
   widget=ScaleBox_1
   properties {stretch=ScaleToFitX, stretchDirection=DownOnly}
   slot {class=VerticalBoxSlot}
  }
  HorizontalBox_2 {
   widgetClassPath=&HorizontalBox
   parent=ScaleBox_1
   widget=HorizontalBox_2
   slot {class=ScaleBoxSlot, properties {horizontalAlignment=HAlign_Fill}}
  }
  ItemRarityOrItemsCount {
   widgetClassPath=&CommonTextBlock
   bIsVariable=true
   parent=HorizontalBox_2
   widget=ItemRarityOrItemsCount
   properties {
    clipping=OnDemand
    justification=Right
    mobileFontSizeMultiplier=1.2
    scrollStyle=&TextScrollStyle-Base_C
    style=/FortUILibrary/Styles/Typography/Foundation/TextStyle-Base-XS-B.TextStyle-Base-XS-B_C
    text=5 ITEMS BUNDLE
    textTransformPolicy=ToUpper
    font {
     bForceMonospaced=false
     bMaterialIsStencil=false
     fontMaterial=null
     fontObject=&NotoSans
     letterSpacing=0
     monospacedWidth=1
     size=16
     skewAmount=0
     typefaceFontName=NotoSans-Bold
     outlineSettings {
      bApplyOutlineToDropShadows=false
      bMiteredCorners=false
      bSeparateFillAlpha=false
      outlineMaterial=null
      outlineSize=0
      outlineColor=<rgba 0 0 0 1>
     }
    }
    shadowOffset=<0 0>
   }
   slot {class=HorizontalBoxSlot}
  }
  RarityTypeSeperatorImage {
   widgetClassPath=&Image
   bIsVariable=true
   parent=HorizontalBox_2
   widget=RarityTypeSeperatorImage
   properties {
    brush {
     bIsDynamicallyLoaded=false
     drawAs=Image
     imageType=NoImage
     mirroring=NoMirror
     resourceName=None
     resourceObject=null
     tiling=NoTile
     imageSize=<3 20>
     margin=<ltrb 0 0 0 0>
     outlineSettings {
      bUseBrushTransparency=false
      roundingType=HalfHeightRadius
      width=0
      color {colorUseRule=UseColor_Specified, specifiedColor=<rgba 0 0 0 0>}
      cornerRadii=<0 0 0 0>
     }
     tintColor {colorUseRule=UseColor_Specified, specifiedColor=<rgba 1 1 1 1>}
     uVRegion {bIsValid=false, max=<0 0>, min=<0 0>}
    }
    colorAndOpacity=<rgba 1 1 1 0.5>
   }
   slot {class=HorizontalBoxSlot, properties {horizontalAlignment=HAlign_Center, verticalAlignment=VAlign_Center, padding=<ltrb 8 0 8 0>}}
  }
  TypeTextBlock {
   widgetClassPath=&CommonTextBlock
   bIsVariable=true
   parent=HorizontalBox_2
   widget=TypeTextBlock
   properties {
    mobileFontSizeMultiplier=1.2
    scrollStyle=&TextScrollStyle-Base_C
    style=/FortUILibrary/Styles/Typography/Foundation/TextStyle-Base-XS-B-70pc.TextStyle-Base-XS-B-70pc_C
    text="{TYPENAME}"
    textTransformPolicy=ToUpper
    colorAndOpacity {colorUseRule=UseColor_Specified, specifiedColor=<rgba 1 1 1 0.7>}
    font {
     bForceMonospaced=false
     bMaterialIsStencil=false
     fontMaterial=null
     fontObject=&NotoSans
     letterSpacing=0
     monospacedWidth=1
     size=16
     skewAmount=0
     typefaceFontName=NotoSans-Bold
     outlineSettings {
      bApplyOutlineToDropShadows=false
      bMiteredCorners=false
      bSeparateFillAlpha=false
      outlineMaterial=null
      outlineSize=0
      outlineColor=<rgba 0 0 0 1>
     }
    }
    shadowOffset=<0 0>
   }
   slot {class=HorizontalBoxSlot, properties {horizontalAlignment=HAlign_Left, verticalAlignment=VAlign_Center, size {sizeRule=Fill, value=1}}}
  }
  Overlay_10 {
   widgetClassPath=&Overlay
   parent=VerticalBox_3
   widget=Overlay_10
   slot {class=VerticalBoxSlot, properties {horizontalAlignment=HAlign_Left, verticalAlignment=VAlign_Top}}
  }
  RarityGlow {
   widgetClassPath=&Image
   bIsVariable=true
   parent=Overlay_10
   widget=RarityGlow
   properties {
    brush {
     bIsDynamicallyLoaded=false
     drawAs=Box
     imageType=NoImage
     mirroring=NoMirror
     resourceName=None
     resourceObject=/FortUILibrary/Styles/Border/Materials/M_Border_SoftGlow
     tiling=NoTile
     imageSize=<128 32>
     margin=<ltrb 0.5 0 0.5 0>
     outlineSettings {
      bUseBrushTransparency=false
      roundingType=HalfHeightRadius
      width=0
      color {colorUseRule=UseColor_Specified, specifiedColor=<rgba 0 0 0 0>}
      cornerRadii=<0 0 0 0>
     }
     tintColor {colorUseRule=UseColor_Specified, specifiedColor=<rgba 1 1 1 1>}
     uVRegion {bIsValid=false, max=<0 0>, min=<0 0>}
    }
    colorAndOpacity=<rgba 1 1 1 0.4>
   }
   slot {class=OverlaySlot, properties {horizontalAlignment=HAlign_Fill, verticalAlignment=VAlign_Fill, padding=<ltrb -32 -16 -32 -16>}}
  }
  NameTextBlock {
   widgetClassPath=&CommonTextBlock
   bIsVariable=true
   parent=Overlay_10
   widget=NameTextBlock
   properties {
    mobileFontSizeMultiplier=1.2
    scrollStyle=&TextScrollStyle-Base_C
    style=&TextStyle-Header-M-SimpleGradient_C
    text=BUNDLE
    textTransformPolicy=ToUpper
    font {
     bForceMonospaced=false
     bMaterialIsStencil=false
     fontMaterial=/FortUILibrary/Materials/Foundation/M_Font_SimpleGradient_Dynamic
     fontObject=&Burbank
     letterSpacing=0
     monospacedWidth=1
     size=32
     skewAmount=0
     typefaceFontName=Burbank-Black
     outlineSettings {
      bApplyOutlineToDropShadows=false
      bMiteredCorners=false
      bSeparateFillAlpha=false
      outlineMaterial=null
      outlineSize=0
      outlineColor=<rgba 0 0 0 1>
     }
    }
    shadowOffset=<0 0>
   }
   slot {class=OverlaySlot, properties {horizontalAlignment=HAlign_Fill, verticalAlignment=VAlign_Center}}
  }
  HeaderRule {
   widgetClassPath=&CommonBorder
   bIsVariable=true
   parent=VerticalBox_1
   widget=HeaderRule
   properties {
    style=/FortUILibrary/Styles/Border/Border-AthenaItemInfoRule.Border-AthenaItemInfoRule_C
    background {
     bIsDynamicallyLoaded=false
     drawAs=Box
     imageType=NoImage
     mirroring=NoMirror
     resourceName=None
     resourceObject=/FortUILibrary/Styles/Border/Materials/M-Border-ItemInfoHeader_procedural
     tiling=NoTile
     imageSize=<128 32>
     margin=<ltrb 0.1 0.7965 1 0.2611>
     outlineSettings {
      bUseBrushTransparency=false
      roundingType=HalfHeightRadius
      width=0
      color {colorUseRule=UseColor_Specified, specifiedColor=<rgba 0 0 0 0>}
      cornerRadii=<0 0 0 0>
     }
     tintColor {colorUseRule=UseColor_Specified, specifiedColor=<rgba 1 1 1 1>}
     uVRegion {bIsValid=false, max=<0 0>, min=<0 0>}
    }
    padding=<ltrb 0 0 0 0>
   }
   slot {class=VerticalBoxSlot}
  }
  Spacer_8 {
   widgetClassPath=&Spacer
   parent=HeaderRule
   widget=Spacer_8
   properties {size=<1 4>}
   slot {class=BorderSlot, properties {padding=<ltrb 0 0 0 0>}}
  }
  BundleSwitcher {
   widgetClassPath=&CommonVisibilitySwitcher
   bIsVariable=true
   parent=VerticalBoxItemDetails
   widget=BundleSwitcher
   properties {activeWidgetIndex=1}
   slot {class=VerticalBoxSlot, properties {verticalAlignment=VAlign_Center, size {sizeRule=Fill, value=1}}}
  }
  SingleItem {
   widgetClassPath=&CommonBorder
   bIsVariable=true
   parent=BundleSwitcher
   widget=SingleItem
   properties {style=&Border_Solid_MdBlue_C, brushColor=<rgba 1 1 1 0>, padding=<ltrb 0 0 0 0>}
   slot {class=CommonVisibilitySwitcherSlot}
  }
  Overlay_8 {widgetClassPath=&Overlay, parent=SingleItem, widget=Overlay_8, slot {class=BorderSlot, properties {padding=<ltrb 0 0 0 0>}}}
  SingleItem_Content {
   widgetClassPath=&HorizontalBox
   bIsVariable=true
   parent=Overlay_8
   widget=SingleItem_Content
   slot {class=OverlaySlot, properties {horizontalAlignment=HAlign_Fill, verticalAlignment=VAlign_Fill}}
  }
  CardContainer {
   widgetClassPath=&VerticalBox
   bIsVariable=true
   parent=SingleItem_Content
   widget=CardContainer
   slot {class=HorizontalBoxSlot, properties {horizontalAlignment=HAlign_Left, verticalAlignment=VAlign_Top, padding=<ltrb 8 8 8 16>}}
  }
  Overlay_9 {widgetClassPath=&Overlay, parent=CardContainer, widget=Overlay_9, slot {class=VerticalBoxSlot}}
  SoftShadow_2 {
   widgetClassPath=&Image
   parent=Overlay_9
   widget=SoftShadow_2
   properties {
    visibility=SelfHitTestInvisible
    brush {
     bIsDynamicallyLoaded=false
     drawAs=Box
     imageType=NoImage
     mirroring=NoMirror
     resourceName=None
     resourceObject=/FortUILibrary/Styles/Border/Materials/M_Shadow_BoxScaleSoft
     tiling=NoTile
     imageSize=<96 96>
     margin=<ltrb 0.5 0.5 0.5 0.5>
     outlineSettings {
      bUseBrushTransparency=false
      roundingType=HalfHeightRadius
      width=0
      color {colorUseRule=UseColor_Specified, specifiedColor=<rgba 0 0 0 0>}
      cornerRadii=<0 0 0 0>
     }
     tintColor {colorUseRule=UseColor_Specified, specifiedColor=<rgba 1 1 1 1>}
     uVRegion {bIsValid=false, max=<0 0>, min=<0 0>}
    }
    colorAndOpacity=<rgba 1 1 1 0.2>
   }
   slot {class=OverlaySlot, properties {horizontalAlignment=HAlign_Fill, verticalAlignment=VAlign_Fill, padding=<ltrb -24 -24 -24 -24>}}
  }
  ItemCardL_2 {
   widgetClassPath=&Overlay
   parent=Overlay_9
   widget=ItemCardL_2
   slot {class=OverlaySlot, properties {horizontalAlignment=HAlign_Center, verticalAlignment=VAlign_Fill}}
  }
  Card {
   widgetClassPath=/FortniteCoreUI/Athena/UI/Frontend/CosmeticItemCard/FortCosmeticItemCard
   bIsVariable=true
   parent=ItemCardL_2
   widget=Card
   properties {cardWidth=160}
   slot {class=OverlaySlot}
  }
  RemainingContainer {
   widgetClassPath=&SizeBox
   bIsVariable=true
   parent=CardContainer
   widget=RemainingContainer
   properties {heightOverride=35, visibility=Collapsed}
   slot {class=VerticalBoxSlot, properties {horizontalAlignment=HAlign_Center, verticalAlignment=VAlign_Top, padding=<ltrb 8 8 8 8>}}
  }
  CommonBorder_1 {
   widgetClassPath=&CommonBorder
   parent=RemainingContainer
   widget=CommonBorder_1
   properties {style=/FortUILibrary/Styles/Border/Border-RoundBoxM.Border-RoundBoxM_C, brushColor=<rgba 0.0021 0.0048 0.0343 1>, padding=<ltrb 10 4 10 4>}
   slot {class=SizeBoxSlot}
  }
  RemainingText {
   widgetClassPath=&CommonTextBlock
   bIsVariable=true
   parent=CommonBorder_1
   widget=RemainingText
   properties {
    style=/FortUILibrary/Styles/Typography/Foundation/TextStyle-Base-S-B.TextStyle-Base-S-B_C
    text=x888
    font {
     bForceMonospaced=false
     bMaterialIsStencil=false
     fontMaterial=null
     fontObject=&NotoSans
     letterSpacing=0
     monospacedWidth=1
     size=18
     skewAmount=0
     typefaceFontName=NotoSans-Bold
     outlineSettings {
      bApplyOutlineToDropShadows=false
      bMiteredCorners=false
      bSeparateFillAlpha=false
      outlineMaterial=null
      outlineSize=0
      outlineColor=<rgba 0 0 0 1>
     }
    }
    shadowOffset=<0 0>
   }
   slot {class=BorderSlot, properties {padding=<ltrb 10 4 10 4>}}
  }
  DetailsScrollBox_2 {
   widgetClassPath=&ScrollBox
   bIsVariable=true
   parent=SingleItem_Content
   widget=DetailsScrollBox_2
   properties {visibility=Collapsed}
   slot {class=HorizontalBoxSlot, properties {padding=<ltrb 0 20 0 16>, size {sizeRule=Fill, value=1}}}
  }
  LockedTextBlock {
   widgetClassPath=&CommonTextBlock
   bIsVariable=true
   parent=DetailsScrollBox_2
   widget=LockedTextBlock
   properties {
    autoWrapText=true
    mobileFontSizeMultiplier=1.2
    style=/FortUILibrary/Styles/Typography/Foundation/TextStyle-Base-XS-70pc.TextStyle-Base-XS-70pc_C
    text=Unlock at Level 5
    colorAndOpacity {colorUseRule=UseColor_Specified, specifiedColor=<rgba 1 1 1 0.7>}
    font {
     bForceMonospaced=false
     bMaterialIsStencil=false
     fontMaterial=null
     fontObject=&NotoSans
     letterSpacing=0
     monospacedWidth=1
     size=16
     skewAmount=0
     typefaceFontName=NotoSans-Regular
     outlineSettings {
      bApplyOutlineToDropShadows=false
      bMiteredCorners=false
      bSeparateFillAlpha=false
      outlineMaterial=null
      outlineSize=0
      outlineColor=<rgba 0 0 0 1>
     }
    }
    shadowOffset=<0 0>
   }
   slot {class=ScrollBoxSlot}
  }
  RichDescriptionBox {
   widgetClassPath=&CommonRichTextBlock
   bIsVariable=true
   parent=DetailsScrollBox_2
   widget=RichDescriptionBox
   properties {
    autoWrapText=true
    bOverrideDefaultStyle=true
    bTintInlineIcon=true
    mobileTextBlockScale=1.2
    text=Description Text Appears Here
    textStyleSet=/Game/Athena/Items/Cosmetics/Metadata/ItemDescriptionTextStyle
    defaultTextStyleOverride {
     overflowPolicy=Clip
     transformPolicy=None
     colorAndOpacity {colorUseRule=UseColor_Specified, specifiedColor=<rgba 1 1 1 0.7>}
     font {
      bForceMonospaced=false
      bMaterialIsStencil=false
      fontMaterial=null
      fontObject=&NotoSans
      letterSpacing=0
      monospacedWidth=1
      size=16
      skewAmount=0
      typefaceFontName=NotoSans-Regular
      outlineSettings {
       bApplyOutlineToDropShadows=false
       bMiteredCorners=false
       bSeparateFillAlpha=false
       outlineMaterial=null
       outlineSize=0
       outlineColor=<rgba 0 0 0 1>
      }
     }
     highlightColor {colorUseRule=UseColor_Specified, specifiedColor=<rgba 0 0 0 1>}
     highlightShape {
      bIsDynamicallyLoaded=false
      drawAs=Image
      imageType=NoImage
      mirroring=NoMirror
      resourceName=None
      resourceObject=null
      tiling=NoTile
      imageSize=<32 32>
      margin=<ltrb 0 0 0 0>
      outlineSettings {
       bUseBrushTransparency=false
       roundingType=HalfHeightRadius
       width=0
       color {colorUseRule=UseColor_Specified, specifiedColor=<rgba 0 0 0 0>}
       cornerRadii=<0 0 0 0>
      }
      tintColor {colorUseRule=UseColor_Specified, specifiedColor=<rgba 1 1 1 1>}
      uVRegion {bIsValid=false, max=<0 0>, min=<0 0>}
     }
     selectedBackgroundColor {colorUseRule=UseColor_Foreground, specifiedColor=<rgba 1 0 1 1>}
     shadowColorAndOpacity=<rgba 0 0 0 1>
     shadowOffset=<0 0>
     strikeBrush {
      bIsDynamicallyLoaded=false
      drawAs=Image
      imageType=NoImage
      mirroring=NoMirror
      resourceName=None
      resourceObject=null
      tiling=NoTile
      imageSize=<32 32>
      margin=<ltrb 0 0 0 0>
      outlineSettings {
       bUseBrushTransparency=false
       roundingType=HalfHeightRadius
       width=0
       color {colorUseRule=UseColor_Specified, specifiedColor=<rgba 0 0 0 0>}
       cornerRadii=<0 0 0 0>
      }
      tintColor {colorUseRule=UseColor_Specified, specifiedColor=<rgba 1 1 1 1>}
      uVRegion {bIsValid=false, max=<0 0>, min=<0 0>}
     }
     underlineBrush {
      bIsDynamicallyLoaded=false
      drawAs=Image
      imageType=NoImage
      mirroring=NoMirror
      resourceName=None
      resourceObject=null
      tiling=NoTile
      imageSize=<32 32>
      margin=<ltrb 0 0 0 0>
      outlineSettings {
       bUseBrushTransparency=false
       roundingType=HalfHeightRadius
       width=0
       color {colorUseRule=UseColor_Specified, specifiedColor=<rgba 0 0 0 0>}
       cornerRadii=<0 0 0 0>
      }
      tintColor {colorUseRule=UseColor_Specified, specifiedColor=<rgba 1 1 1 1>}
      uVRegion {bIsValid=false, max=<0 0>, min=<0 0>}
     }
    }
   }
   slot {class=ScrollBoxSlot}
  }
  MultipleItemCB {
   widgetClassPath=&CommonBorder
   bIsVariable=true
   parent=BundleSwitcher
   widget=MultipleItemCB
   properties {
    navigation=&self:WidgetTree.MultipleItemCB.WidgetNavigation_0
    style=&Border_Solid_MdBlue_C
    background {
     bIsDynamicallyLoaded=false
     drawAs=Box
     imageType=NoImage
     mirroring=NoMirror
     resourceName=None
     resourceObject=/FortUILibrary/Styles/Border/Materials/M_Border_Solid_MdBlue
     tiling=NoTile
     imageSize=<64 64>
     margin=<ltrb 0.5 0 0.5 0>
     outlineSettings {
      bUseBrushTransparency=false
      roundingType=HalfHeightRadius
      width=0
      color {colorUseRule=UseColor_Specified, specifiedColor=<rgba 0 0 0 0>}
      cornerRadii=<0 0 0 0>
     }
     tintColor {colorUseRule=UseColor_Specified, specifiedColor=<rgba 1 1 1 1>}
     uVRegion {bIsValid=false, max=<0 0>, min=<0 0>}
    }
    brushColor=<rgba 1 1 1 0>
    padding=<ltrb 12 12 12 12>
   }
   slot {class=CommonVisibilitySwitcherSlot}
  }
  VerticalBox_2 {
   widgetClassPath=&VerticalBox
   parent=MultipleItemCB
   widget=VerticalBox_2
   slot {class=BorderSlot, properties {padding=<ltrb 12 12 12 12>}}
  }
  ItemRewardsTIleView {
   widgetClassPath=/Script/UMG.TileView
   bIsVariable=true
   parent=VerticalBox_2
   widget=ItemRewardsTIleView
   properties {
    bIsFocusable=false
    entryHeight=150
    entryWidgetClass=&/AthenaStoreReward.AthenaStoreReward_C
    entryWidth=120
    horizontalEntrySpacing=6
    navigation=&self:WidgetTree.ItemRewardsTIleView.WidgetNavigation_0
    tileAlignment=LeftAligned
    verticalEntrySpacing=6
   }
   slot {class=VerticalBoxSlot, properties {size {sizeRule=Fill, value=1}}}
  }
  Border_0 {
   widgetClassPath=&Border
   parent=VerticalBoxItemDetails
   widget=Border_0
   properties {verticalAlignment=VAlign_Bottom, brushColor=<rgba 0.006 0.0262 0.2502 1>, padding=<ltrb 12 6 12 6>}
   slot {class=VerticalBoxSlot, properties {verticalAlignment=VAlign_Bottom}}
  }
  ScaleBox_0 {
   widgetClassPath=&ScaleBox
   parent=Border_0
   widget=ScaleBox_0
   properties {stretch=ScaleToFitX, stretchDirection=DownOnly}
   slot {class=BorderSlot, properties {verticalAlignment=VAlign_Bottom, padding=<ltrb 12 6 12 6>}}
  }
  HorizontalBox_PurchasePrice {
   widgetClassPath=&HorizontalBox
   bIsVariable=true
   parent=ScaleBox_0
   widget=HorizontalBox_PurchasePrice
   properties {flowDirectionPreference=Culture}
   slot {class=ScaleBoxSlot, properties {horizontalAlignment=HAlign_Fill, verticalAlignment=VAlign_Fill}}
  }
  SizeBox_5 {
   widgetClassPath=&SizeBox
   parent=HorizontalBox_PurchasePrice
   widget=SizeBox_5
   properties {bOverride_HeightOverride=true, bOverride_WidthOverride=true, heightOverride=30, widthOverride=30}
   slot {class=HorizontalBoxSlot, properties {horizontalAlignment=HAlign_Left, verticalAlignment=VAlign_Center}}
  }
  Currency1ItemImage2 {
   widgetClassPath=&Image
   bIsVariable=true
   parent=SizeBox_5
   widget=Currency1ItemImage2
   properties {
    visibility=SelfHitTestInvisible
    brush {
     bIsDynamicallyLoaded=false
     drawAs=Image
     imageType=NoImage
     mirroring=NoMirror
     resourceName=None
     resourceObject=/Game/UI/Foundation/Shell/Textures/T-Icon-VBucks-L
     tiling=NoTile
     imageSize=<40 40>
     margin=<ltrb 0 0 0 0>
     outlineSettings {
      bUseBrushTransparency=false
      roundingType=HalfHeightRadius
      width=0
      color {colorUseRule=UseColor_Specified, specifiedColor=<rgba 0 0 0 0>}
      cornerRadii=<0 0 0 0>
     }
     tintColor {colorUseRule=UseColor_Specified, specifiedColor=<rgba 1 1 1 1>}
     uVRegion {bIsValid=false, max=<0 0>, min=<0 0>}
    }
   }
   slot {class=SizeBoxSlot, properties {horizontalAlignment=HAlign_Center, verticalAlignment=VAlign_Center}}
  }
  TextOfferDisplayPrice1 {
   widgetClassPath=&CommonTextBlock
   bIsVariable=true
   parent=HorizontalBox_PurchasePrice
   widget=TextOfferDisplayPrice1
   properties {
    justification=Center
    text="1,200"
    font {
     bForceMonospaced=false
     bMaterialIsStencil=false
     fontMaterial=null
     fontObject=&BurbankSmall
     letterSpacing=0
     monospacedWidth=1
     size=24
     skewAmount=0
     typefaceFontName=Bold
     outlineSettings {
      bApplyOutlineToDropShadows=false
      bMiteredCorners=false
      bSeparateFillAlpha=false
      outlineMaterial=null
      outlineSize=0
      outlineColor=<rgba 0 0 0 1>
     }
    }
    shadowOffset=<0 0>
   }
   slot {class=HorizontalBoxSlot, properties {horizontalAlignment=HAlign_Left, verticalAlignment=VAlign_Center, padding=<ltrb 12 0 0 0>}}
  }
  O_Sale_PreviousPrice {
   widgetClassPath=&Overlay
   bIsVariable=true
   parent=HorizontalBox_PurchasePrice
   widget=O_Sale_PreviousPrice
   properties {visibility=Collapsed}
   slot {class=HorizontalBoxSlot, properties {verticalAlignment=VAlign_Center, padding=<ltrb 6 0 0 0>}}
  }
  TextOriginalPrice1 {
   widgetClassPath=&CommonTextBlock
   bIsVariable=true
   parent=O_Sale_PreviousPrice
   widget=TextOriginalPrice1
   properties {
    justification=Center
    text="1,500"
    colorAndOpacity {colorUseRule=UseColor_Specified, specifiedColor=<rgba 1 1 1 0.7>}
    font {
     bForceMonospaced=false
     bMaterialIsStencil=false
     fontMaterial=null
     fontObject=&BurbankSmall
     letterSpacing=0
     monospacedWidth=1
     size=24
     skewAmount=0
     typefaceFontName=Bold
     outlineSettings {
      bApplyOutlineToDropShadows=false
      bMiteredCorners=false
      bSeparateFillAlpha=false
      outlineMaterial=null
      outlineSize=0
      outlineColor=<rgba 0 0 0 1>
     }
    }
    shadowOffset=<0 0>
   }
   slot {class=OverlaySlot, properties {horizontalAlignment=HAlign_Fill, verticalAlignment=VAlign_Fill, padding=<ltrb 10 0 10 0>}}
  }
  I_StrikeThrough {
   widgetClassPath=&Image
   parent=O_Sale_PreviousPrice
   widget=I_StrikeThrough
   properties {
    visibility=SelfHitTestInvisible
    brush {
     bIsDynamicallyLoaded=false
     drawAs=Image
     imageType=NoImage
     mirroring=NoMirror
     resourceName=None
     resourceObject=/Game/UI/Foundation/Materials/M_UI_StrikeThrough
     tiling=NoTile
     imageSize=<32 32>
     margin=<ltrb 0 0 0 0>
     outlineSettings {
      bUseBrushTransparency=false
      roundingType=HalfHeightRadius
      width=0
      color {colorUseRule=UseColor_Specified, specifiedColor=<rgba 0 0 0 0>}
      cornerRadii=<0 0 0 0>
     }
     tintColor {colorUseRule=UseColor_Specified, specifiedColor=<rgba 1 1 1 1>}
     uVRegion {bIsValid=false, max=<0 0>, min=<0 0>}
    }
    colorAndOpacity=<rgba 1 1 1 0.7>
    renderTransform {angle=5, scale=<1 1>, shear=<0 0>, translation=<0 4>}
   }
   slot {class=OverlaySlot, properties {horizontalAlignment=HAlign_Fill, verticalAlignment=VAlign_Fill, padding=<ltrb 4 0 0 4>}}
  }
 ]
}
