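# Legacy button widget for a dynamic V-Bucks offer in the item shop, derived from AthenaDynamicMtxOfferButtonLegacy.
# Lays out a sized currency icon switcher, the V-Bucks amount text, and a price banner with an input action glyph over material-driven backgrounds.
# Hover plays the OnHover animation forward and unhover reverses it faster, animating material edge parameters, colors, scale, padding and the glyph's fade-in.
# The graph holds only hover logic.
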WidgetBlueprint /ItemShop/UI/ReloadMtx/DynamicMtxOfferButton_Legacy
animations [OnHover {
  binding_count=10
  length_seconds=0.2
  object=&self:OnHover
  sequence {bindings [
    SizeBox {
     id=#1
     possessed_class=SizeBox
     tracks [
      {
       type=MovieScene2DTransformTrack
       sections [{
         type=MovieScene2DTransformSection
         end_frame=4
         start_frame=0
         channels [
          Translation.X {type=MovieSceneScriptingFloatChannel}
          Translation.Y {type=MovieSceneScriptingFloatChannel, keys [[0,-26],[3.9613,-14]]}
          Angle {type=MovieSceneScriptingFloatChannel}
          Scale.X {type=MovieSceneScriptingFloatChannel, keys [[0,0.75],[3.9613,1]]}
          Scale.Y {type=MovieSceneScriptingFloatChannel, keys [[0,0.75],[3.9613,1]]}
          Shear.X {type=MovieSceneScriptingFloatChannel}
          Shear.Y {type=MovieSceneScriptingFloatChannel}
         ]
        }]
      }
      {
       type=MovieSceneFloatTrack
       display_name=Render Opacity
       sections [{type=MovieSceneFloatSection, end_frame=4, start_frame=0, channels [None {type=MovieSceneScriptingFloatChannel, keys [[0,0],[3.9613,1]]}]}]
      }
     ]
    }
    LoadGuardCurrencySpinner {
     id=#2
     possessed_class=CommonLoadGuard
     tracks [{
       type=MovieScene2DTransformTrack
       sections [{
         type=MovieScene2DTransformSection
         end_frame=4
         start_frame=0
         channels [
          Translation.X {type=MovieSceneScriptingFloatChannel}
          Translation.Y {type=MovieSceneScriptingFloatChannel}
          Angle {type=MovieSceneScriptingFloatChannel}
          Scale.X {type=MovieSceneScriptingFloatChannel, keys [[0,1],[3.9613,1.2]]}
          Scale.Y {type=MovieSceneScriptingFloatChannel, keys [[0,1],[3.9613,1.2]]}
          Shear.X {type=MovieSceneScriptingFloatChannel}
          Shear.Y {type=MovieSceneScriptingFloatChannel}
         ]
        }]
      }]
    }
    IMG_ButtonBackground {
     id=#3
     possessed_class=Image
     tracks [
      {
       type=MovieSceneWidgetMaterialTrack
       sections [{
         type=MovieSceneParameterSection
         end_frame=4
         start_frame=0
         channels [
          BottomEdge_1 {type=MovieSceneScriptingFloatChannel, keys [[0,0.03],[3.9613,0.065]]}
          BottomRotation_1 {type=MovieSceneScriptingFloatChannel, keys [[0,-0.002],[3.9613,0.008]]}
          TopEdge_1 {type=MovieSceneScriptingFloatChannel, keys [[0,0.04],[3.9613,0.01]]}
          TopRotation_1 {type=MovieSceneScriptingFloatChannel, keys [[0,0.005],[3.9613,-0.006]]}
          LeftEdge_1 {type=MovieSceneScriptingFloatChannel, keys [[0,0.01],[3.9613,0.02]]}
          LeftRotation_1 {type=MovieSceneScriptingFloatChannel, keys [[0,-0.0015],[3.9613,0.0055]]}
          RightEdge_1 {type=MovieSceneScriptingFloatChannel, keys [[0,0.01],[3.9613,0.018]]}
          RightRotation_1 {type=MovieSceneScriptingFloatChannel, keys [[0,0],[3.9613,-0.003]]}
         ]
        }]
      }
      {
       type=MovieSceneColorTrack
       display_name=Color and Opacity
       sections [{
         type=MovieSceneColorSection
         end_frame=3
         start_frame=0
         channels [
          Color.R {type=MovieSceneScriptingFloatChannel, keys [[0,0],[3.9613,1]]}
          Color.G {type=MovieSceneScriptingFloatChannel, keys [[0,0.3613],[3.9613,1]]}
          Color.B {type=MovieSceneScriptingFloatChannel, const [1,2]}
          Color.A {type=MovieSceneScriptingFloatChannel, const [1,2]}
         ]
        }]
      }
     ]
    }
    Text_DynamicMtxAmount {
     id=#4
     possessed_class=CommonTextBlock
     tracks [
      {type=MovieScene2DTransformTrack, sections []}
      {
       type=MovieSceneColorTrack
       display_name=Color and Opacity
       sections [{
         type=MovieSceneColorSection
         end_frame=4
         start_frame=0
         channels [
          Color.R {type=MovieSceneScriptingFloatChannel, keys [[0,1],[3.9613,0]]}
          Color.G {type=MovieSceneScriptingFloatChannel, keys [[0,1],[3.9613,0]]}
          Color.B {type=MovieSceneScriptingFloatChannel, keys [[0,1],[3.9613,0]]}
          Color.A {type=MovieSceneScriptingFloatChannel, const [1,2]}
         ]
        }]
      }
     ]
    }
    SizeBoxSlot_0 {
     id=#5
     parent_id=#6
     possessed_class=SizeBoxSlot
     tracks [{
       type=MovieSceneMarginTrack
       display_name=Padding
       sections [{
         type=MovieSceneMarginSection
         end_frame=4
         start_frame=0
         channels [
          Left {type=MovieSceneScriptingFloatChannel, keys [[0,0],[3.9613,-4]]}
          Top {type=MovieSceneScriptingFloatChannel, keys [[0,0],[3.9613,-16]]}
          Right {type=MovieSceneScriptingFloatChannel, keys [[0,0],[3.9613,-4]]}
          Bottom {type=MovieSceneScriptingFloatChannel, keys [[0,0],[3.9613,-16]]}
         ]
        }]
      }]
    }
    IMG_PriceBackground {
     id=#7
     possessed_class=Image
     tracks [{
       type=MovieSceneWidgetMaterialTrack
       sections [{
         type=MovieSceneParameterSection
         end_frame=4
         start_frame=0
         channels [
          BottomRotation_1 {type=MovieSceneScriptingFloatChannel, keys [[0,-0.002],[3.9613,0.02]]}
          BottomEdge_1 {type=MovieSceneScriptingFloatChannel, keys [[0,0.03],[3.9613,0.15]]}
          RightEdge_1 {type=MovieSceneScriptingFloatChannel, keys [[0,0.0095],[3.9613,0.024]]}
          RightRotation_1 {type=MovieSceneScriptingFloatChannel, keys [[0,0],[3.9613,-0.001]]}
          LeftRotation_1 {type=MovieSceneScriptingFloatChannel, keys [[0,-0.0006],[3.9613,0.002]]}
          LeftEdge_1 {type=MovieSceneScriptingFloatChannel, keys [[0,0.0065],[3.9613,0.031]]}
          TopEdge_1 {type=MovieSceneScriptingFloatChannel, keys [[0,0.04],[3.9613,0.01]]}
          TopRotation_1 {type=MovieSceneScriptingFloatChannel, keys [[0,0.005],[3.9613,-0.005]]}
         ]
        }]
      }]
    }
    Overlay {id=#8, possessed_class=Overlay, tracks []}
    IMGCurrency_XL {
     id=#9
     possessed_class=Image
     tracks [{
       type=MovieScene2DTransformTrack
       sections [{
         type=MovieScene2DTransformSection
         end_frame=4
         start_frame=0
         channels [
          Translation.X {type=MovieSceneScriptingFloatChannel, keys [[0,-46],[3.9613,-40]]}
          Translation.Y {type=MovieSceneScriptingFloatChannel}
          Angle {type=MovieSceneScriptingFloatChannel}
          Scale.X {type=MovieSceneScriptingFloatChannel, keys [[0,1],[3.9613,0.85]]}
          Scale.Y {type=MovieSceneScriptingFloatChannel, keys [[0,1],[3.9613,0.85]]}
          Shear.X {type=MovieSceneScriptingFloatChannel}
          Shear.Y {type=MovieSceneScriptingFloatChannel}
         ]
        }]
      }]
    }
    OverlaySlot_3 {
     id=#10
     parent_id=#8
     possessed_class=OverlaySlot
     tracks [{
       type=MovieSceneMarginTrack
       display_name=Padding
       sections [{
         type=MovieSceneMarginSection
         end_frame=4
         start_frame=0
         channels [
          Left {type=MovieSceneScriptingFloatChannel}
          Top {type=MovieSceneScriptingFloatChannel, keys [[0,110],[3.9613,120]]}
          Right {type=MovieSceneScriptingFloatChannel}
          Bottom {type=MovieSceneScriptingFloatChannel}
         ]
        }]
      }]
    }
    Overlay_36 {id=#6, possessed_class=Overlay, tracks []}
   ]}
 }]
graphs [EventGraph {nodes [
   Event On Hovered()
   n1 = Get OnHover() -> (OnHover:Widget Animation Object Reference)
   > PlayAnimationForward(InAnimation=@n1, PlaybackSpeed=1.5)
   Event On Unhovered()
   > PlayAnimationReverse(InAnimation=@n1, PlaybackSpeed=2)
  ]}]
variables []
widgets {
 info {
  inheritedWidgetCount=7
  parentClass=/Script/FortniteUI.AthenaDynamicMtxOfferButtonLegacy
  rootWidgetClass=&Overlay
  widgetCount=27
 }
 widgets [
  Overlay_34 {widgetClassPath=&Overlay, widget=Overlay_34}
  ButtonSize {
   widgetClassPath=&SizeBox
   parent=Overlay_34
   widget=ButtonSize
   properties {
    bOverride_HeightOverride=true
    bOverride_MinDesiredWidth=true
    heightOverride=160
    maxAspectRatio=5
    maxDesiredWidth=850
    minDesiredWidth=650
    widthOverride=650
   }
   slot {class=OverlaySlot, properties {horizontalAlignment=HAlign_Fill, verticalAlignment=VAlign_Fill}}
  }
  Overlay_36 {widgetClassPath=&Overlay, parent=ButtonSize, widget=Overlay_36, slot {class=SizeBoxSlot}}
  IMG_ButtonBackground {
   widgetClassPath=&Image
   bIsVariable=true
   parent=Overlay_36
   widget=IMG_ButtonBackground
   properties {
    brush {
     bIsDynamicallyLoaded=false
     drawAs=Box
     imageType=NoImage
     mirroring=NoMirror
     resourceName=None
     resourceObject=&/MI_UI_ReloadMtx_DynamicOfferButton_Legacy
     tiling=NoTile
     imageSize=<32 32>
     margin=<ltrb 0 0 0 0>
     outlineSettings {
      bUseBrushTransparency=false
      roundingType=HalfHeightRadius
      width=0
      color {colorUseRule=UseColor_Specified, specifiedColor=<rgba 0 0 0 0>}
      cornerRadii=<0 0 0 0>
     }
     tintColor {colorUseRule=UseColor_Specified, specifiedColor=<rgba 1 1 1 1>}
     uVRegion {bIsValid=false, max=<0 0>, min=<0 0>}
    }
    colorAndOpacity=<rgba 0 0.3613 1 1>
   }
   slot {class=OverlaySlot, properties {horizontalAlignment=HAlign_Fill, verticalAlignment=VAlign_Fill}}
  }
  Overlay_0 {
   widgetClassPath=&Overlay
   parent=Overlay_36
   widget=Overlay_0
   slot {class=OverlaySlot, properties {horizontalAlignment=HAlign_Fill, verticalAlignment=VAlign_Fill}}
  }
  Overlay {
   widgetClassPath=&Overlay
   parent=Overlay_0
   widget=Overlay
   slot {class=OverlaySlot, properties {horizontalAlignment=HAlign_Fill, verticalAlignment=VAlign_Fill, padding=<ltrb 0 110 0 0>}}
  }
  IMG_PriceBackground {
   widgetClassPath=&Image
   bIsVariable=true
   parent=Overlay
   widget=IMG_PriceBackground
   properties {
    brush {
     bIsDynamicallyLoaded=false
     drawAs=Box
     imageType=NoImage
     mirroring=NoMirror
     resourceName=None
     resourceObject=&/MI_UI_ReloadMtx_DynamicOfferContainer_Legacy
     tiling=NoTile
     imageSize=<32 32>
     margin=<ltrb 0 0 0 0>
     outlineSettings {
      bUseBrushTransparency=false
      roundingType=HalfHeightRadius
      width=0
      color {colorUseRule=UseColor_Specified, specifiedColor=<rgba 0 0 0 0>}
      cornerRadii=<0 0 0 0>
     }
     tintColor {colorUseRule=UseColor_Specified, specifiedColor=<rgba 1 1 1 1>}
     uVRegion {bIsValid=false, max=<0 0>, min=<0 0>}
    }
    colorAndOpacity=<rgba 1 0.9734 0 1>
   }
   slot {class=OverlaySlot, properties {horizontalAlignment=HAlign_Fill, verticalAlignment=VAlign_Fill}}
  }
  VerticalBox_2 {
   widgetClassPath=&VerticalBox
   parent=Overlay
   widget=VerticalBox_2
   slot {class=OverlaySlot, properties {horizontalAlignment=HAlign_Fill, verticalAlignment=VAlign_Center}}
  }
  Overlay_1 {
   widgetClassPath=&Overlay
   parent=VerticalBox_2
   widget=Overlay_1
   slot {class=VerticalBoxSlot, properties {padding=<ltrb 0 10 0 10>}}
  }
  Text_RealMoneyPrice {
   widgetClassPath=&CommonTextBlock
   bInherited=true
   parent=Overlay_1
   widget=Text_RealMoneyPrice
   properties {
    justification=Center
    scrollStyle=/FortUILibrary/Styles/Typography/Foundation/TextScrollStyle-NoFade.TextScrollStyle-NoFade_C
    text=3.99 $ US
    textTransformPolicy=ToUpper
    colorAndOpacity {colorUseRule=UseColor_Specified, specifiedColor=<rgba 0 0 0 1>}
    font {
     bForceMonospaced=false
     bMaterialIsStencil=false
     fontMaterial=null
     fontObject=&BurbankSmall
     letterSpacing=0
     monospacedWidth=1
     size=20
     skewAmount=0
     typefaceFontName=Black
     outlineSettings {
      bApplyOutlineToDropShadows=false
      bMiteredCorners=false
      bSeparateFillAlpha=false
      outlineMaterial=null
      outlineSize=0
      outlineColor=<rgba 0 0 0 1>
     }
    }
    renderTransform {angle=0, scale=<1 1>, shear=<-8 0>, translation=<0 0>}
    shadowOffset=<0 0>
   }
   slot {class=OverlaySlot, properties {horizontalAlignment=HAlign_Fill, verticalAlignment=VAlign_Center}}
  }
  SizeBox {
   widgetClassPath=&SizeBox
   parent=VerticalBox_2
   widget=SizeBox
   properties {
    bOverride_HeightOverride=true
    bOverride_WidthOverride=true
    heightOverride=48
    widthOverride=48
    renderTransform {angle=0, scale=<1 1>, shear=<0 0>, translation=<0 -26>}
   }
   slot {class=VerticalBoxSlot, properties {horizontalAlignment=HAlign_Center, verticalAlignment=VAlign_Top, size {sizeRule=Automatic, value=0}}}
  }
  InputActionWidget {
   widgetClassPath=&CommonActionWidget
   bInherited=true
   bIsVariable=true
   parent=SizeBox
   widget=InputActionWidget
   properties {iconRimBrush {
     bIsDynamicallyLoaded=false
     drawAs=Image
     imageType=NoImage
     mirroring=NoMirror
     resourceName=None
     resourceObject=/FortUILibrary/Foundation/Buttons/Materials/M-Button-Console-YellowOutline
     tiling=NoTile
     imageSize=<48 48>
     margin=<ltrb 0 0 0 0>
     outlineSettings {
      bUseBrushTransparency=false
      roundingType=HalfHeightRadius
      width=0
      color {colorUseRule=UseColor_Specified, specifiedColor=<rgba 0 0 0 0>}
      cornerRadii=<0 0 0 0>
     }
     tintColor {colorUseRule=UseColor_Specified, specifiedColor=<rgba 1 1 1 1>}
     uVRegion {bIsValid=false, max=<0 0>, min=<0 0>}
    }}
   slot {class=SizeBoxSlot}
  }
  HorizontalBox_DynamicOffer {
   widgetClassPath=&HorizontalBox
   parent=Overlay_0
   widget=HorizontalBox_DynamicOffer
   slot {class=OverlaySlot, properties {horizontalAlignment=HAlign_Fill, verticalAlignment=VAlign_Fill}}
  }
  LoadGuardCurrencySpinner {
   widgetClassPath=/Script/CommonUI.CommonLoadGuard
   bInherited=true
   bIsVariable=true
   parent=HorizontalBox_DynamicOffer
   widget=LoadGuardCurrencySpinner
   properties {loadingText="", textStyle=&TextStyle-Base_C}
   slot {class=HorizontalBoxSlot, properties {horizontalAlignment=HAlign_Left, size {sizeRule=Fill, value=0.3}}}
  }
  ScaleBox_0 {
   widgetClassPath=&ScaleBox
   parent=LoadGuardCurrencySpinner
   widget=ScaleBox_0
   properties {stretchDirection=DownOnly}
   slot {class=LoadGuardSlot}
  }
  OptionalSwitcher_CurrencyIcons {
   widgetClassPath=&CommonVisibilitySwitcher
   bInherited=true
   parent=ScaleBox_0
   widget=OptionalSwitcher_CurrencyIcons
   slot {class=ScaleBoxSlot}
  }
  IMGCurrency_XS {
   widgetClassPath=&Image
   bIsVariable=true
   parent=OptionalSwitcher_CurrencyIcons
   widget=IMGCurrency_XS
   properties {brush {
     bIsDynamicallyLoaded=false
     drawAs=Image
     imageType=NoImage
     mirroring=NoMirror
     resourceName=None
     resourceObject=/Game/UI/Foundation/Textures/Founders/T_VBucks_Tile_1000_New
     tiling=NoTile
     imageSize=<512 512>
     margin=<ltrb 0 0 0 0>
     outlineSettings {
      bUseBrushTransparency=false
      roundingType=HalfHeightRadius
      width=0
      color {colorUseRule=UseColor_Specified, specifiedColor=<rgba 0 0 0 0>}
      cornerRadii=<0 0 0 0>
     }
     tintColor {colorUseRule=UseColor_Specified, specifiedColor=<rgba 1 1 1 1>}
     uVRegion {bIsValid=false, max=<0 0>, min=<0 0>}
    }}
   slot {class=CommonVisibilitySwitcherSlot}
  }
  IMGCurrency_S {
   widgetClassPath=&Image
   bIsVariable=true
   parent=OptionalSwitcher_CurrencyIcons
   widget=IMGCurrency_S
   properties {brush {
     bIsDynamicallyLoaded=false
     drawAs=Image
     imageType=NoImage
     mirroring=NoMirror
     resourceName=None
     resourceObject=/Game/UI/Foundation/Textures/Founders/T_VBucks_Tile_2800_New
     tiling=NoTile
     imageSize=<512 512>
     margin=<ltrb 0 0 0 0>
     outlineSettings {
      bUseBrushTransparency=false
      roundingType=HalfHeightRadius
      width=0
      color {colorUseRule=UseColor_Specified, specifiedColor=<rgba 0 0 0 0>}
      cornerRadii=<0 0 0 0>
     }
     tintColor {colorUseRule=UseColor_Specified, specifiedColor=<rgba 1 1 1 1>}
     uVRegion {bIsValid=false, max=<0 0>, min=<0 0>}
    }}
   slot {class=CommonVisibilitySwitcherSlot}
  }
  IMGCurrency_M {
   widgetClassPath=&Image
   bIsVariable=true
   parent=OptionalSwitcher_CurrencyIcons
   widget=IMGCurrency_M
   properties {brush {
     bIsDynamicallyLoaded=false
     drawAs=Image
     imageType=NoImage
     mirroring=NoMirror
     resourceName=None
     resourceObject=/Game/UI/Foundation/Textures/Founders/T-VBucks-Tile-5000
     tiling=NoTile
     imageSize=<512 512>
     margin=<ltrb 0 0 0 0>
     outlineSettings {
      bUseBrushTransparency=false
      roundingType=HalfHeightRadius
      width=0
      color {colorUseRule=UseColor_Specified, specifiedColor=<rgba 0 0 0 0>}
      cornerRadii=<0 0 0 0>
     }
     tintColor {colorUseRule=UseColor_Specified, specifiedColor=<rgba 1 1 1 1>}
     uVRegion {bIsValid=false, max=<0 0>, min=<0 0>}
    }}
   slot {class=CommonVisibilitySwitcherSlot, properties {padding=<ltrb -20 -20 -20 -20>}}
  }
  IMGCurrency_L {
   widgetClassPath=&Image
   bIsVariable=true
   parent=OptionalSwitcher_CurrencyIcons
   widget=IMGCurrency_L
   properties {brush {
     bIsDynamicallyLoaded=false
     drawAs=Image
     imageType=NoImage
     mirroring=NoMirror
     resourceName=None
     resourceObject=/Game/UI/Foundation/Textures/Founders/T-VBuckpile-7800_Tile
     tiling=NoTile
     imageSize=<512 512>
     margin=<ltrb 0 0 0 0>
     outlineSettings {
      bUseBrushTransparency=false
      roundingType=HalfHeightRadius
      width=0
      color {colorUseRule=UseColor_Specified, specifiedColor=<rgba 0 0 0 0>}
      cornerRadii=<0 0 0 0>
     }
     tintColor {colorUseRule=UseColor_Specified, specifiedColor=<rgba 1 1 1 1>}
     uVRegion {bIsValid=false, max=<0 0>, min=<0 0>}
    }}
   slot {class=CommonVisibilitySwitcherSlot, properties {padding=<ltrb -30 -30 -30 -30>}}
  }
  IMGCurrency_XL {
   widgetClassPath=&Image
   bIsVariable=true
   parent=OptionalSwitcher_CurrencyIcons
   widget=IMGCurrency_XL
   properties {
    brush {
     bIsDynamicallyLoaded=false
     drawAs=Image
     imageType=NoImage
     mirroring=NoMirror
     resourceName=None
     resourceObject=/Game/UI/Foundation/Textures/Founders/T_VBucks_Tile_13500_New
     tiling=NoTile
     imageSize=<512 512>
     margin=<ltrb 0 0 0 0>
     outlineSettings {
      bUseBrushTransparency=false
      roundingType=HalfHeightRadius
      width=0
      color {colorUseRule=UseColor_Specified, specifiedColor=<rgba 0 0 0 0>}
      cornerRadii=<0 0 0 0>
     }
     tintColor {colorUseRule=UseColor_Specified, specifiedColor=<rgba 1 1 1 1>}
     uVRegion {bIsValid=false, max=<0 0>, min=<0 0>}
    }
    renderTransform {angle=0, scale=<1 1>, shear=<0 0>, translation=<-46 0>}
   }
   slot {class=CommonVisibilitySwitcherSlot, properties {padding=<ltrb -150 -120 -80 -120>}}
  }
  Overlay_33 {
   widgetClassPath=&Overlay
   parent=HorizontalBox_DynamicOffer
   widget=Overlay_33
   slot {class=HorizontalBoxSlot, properties {padding=<ltrb 0 0 0 50>, size {sizeRule=Fill, value=0.8}}}
  }
  ScaleBox_1 {
   widgetClassPath=&ScaleBox
   parent=Overlay_33
   widget=ScaleBox_1
   slot {class=OverlaySlot, properties {horizontalAlignment=HAlign_Center, verticalAlignment=VAlign_Center}}
  }
  Text_DynamicMtxAmount {
   widgetClassPath=&CommonTextBlock
   bInherited=true
   bIsVariable=true
   parent=ScaleBox_1
   widget=Text_DynamicMtxAmount
   properties {
    justification=Center
    lineHeightPercentage=0.83
    scrollStyle=&TextScrollStyle-NoFade_C
    text=Get 400 V-Bucks
    textTransformPolicy=ToUpper
    font {
     bForceMonospaced=false
     bMaterialIsStencil=false
     fontMaterial=null
     fontObject=&BurbankSmall
     letterSpacing=0
     monospacedWidth=1
     size=30
     skewAmount=0
     typefaceFontName=Black
     outlineSettings {
      bApplyOutlineToDropShadows=false
      bMiteredCorners=false
      bSeparateFillAlpha=false
      outlineMaterial=null
      outlineSize=0
      outlineColor=<rgba 0 0 0 1>
     }
    }
    renderTransform {angle=0, scale=<1 1>, shear=<-8 0>, translation=<0 0>}
    shadowOffset=<0 0>
   }
   slot {class=ScaleBoxSlot, properties {horizontalAlignment=HAlign_Fill}}
  }
  Spacer_48 {
   widgetClassPath=&Spacer
   parent=HorizontalBox_DynamicOffer
   widget=Spacer_48
   slot {class=HorizontalBoxSlot, properties {size {sizeRule=Fill, value=0.3}}}
  }
  Text_MtxAmount {widgetClassPath=&CommonTextBlock, bInherited=true, widget=null}
  OptionalIMG_StarterKit {widgetClassPath=&Image, bInherited=true, widget=null}
 ]
}
