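# Legacy starter kit offer button widget for the Item Shop's ReloadMtx V-Bucks screen, built on AthenaStarterKitOfferButtonLegacy.
# Lays out kit art, currency image, name, contents, a price strip and a "Limited time" timer tag, overriding several inherited parent widgets.
# Hover plays OnHover forward, unhover reverses it faster.
# The animation skews material-driven shapes via edge/rotation parameters, whitens the background, blackens text, shifts padding and reveals the input prompt.
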WidgetBlueprint /ItemShop/UI/ReloadMtx/DynamicMtxStarterKitButton_Legacy
animations [OnHover {
  binding_count=14
  length_seconds=0.2101
  object=&self:OnHover
  sequence {bindings [
    OverlaySlot_6 {
     id=#1
     parent_id=#2
     possessed_class=OverlaySlot
     tracks [{
       type=MovieSceneMarginTrack
       display_name=Padding
       sections [{
         type=MovieSceneMarginSection
         end_frame=4
         start_frame=-1
         channels [
          Left {type=MovieSceneScriptingFloatChannel, const [0,2]}
          Top {type=MovieSceneScriptingFloatChannel, keys [[0.0003,0],[3.9613,-10]]}
          Right {type=MovieSceneScriptingFloatChannel}
          Bottom {type=MovieSceneScriptingFloatChannel}
         ]
        }]
      }]
    }
    Text_DisplayName {
     id=#3
     possessed_class=CommonTextBlock
     tracks [{
       type=MovieSceneColorTrack
       display_name=Color and Opacity
       sections [{
         type=MovieSceneColorSection
         end_frame=4
         start_frame=0
         channels [
          Color.R {type=MovieSceneScriptingFloatChannel, keys [[0.0003,1],[3.9613,0]]}
          Color.G {type=MovieSceneScriptingFloatChannel, keys [[0.0003,1],[3.9613,0]]}
          Color.B {type=MovieSceneScriptingFloatChannel, keys [[0.0003,1],[3.9613,0]]}
          Color.A {type=MovieSceneScriptingFloatChannel, const [1,2]}
         ]
        }]
      }]
    }
    HorizontalBoxSlot_4 {
     id=#4
     parent_id=#5
     possessed_class=HorizontalBoxSlot
     tracks [{
       type=MovieSceneMarginTrack
       display_name=Padding
       sections [{
         type=MovieSceneMarginSection
         end_frame=4
         start_frame=-1
         channels [
          Left {type=MovieSceneScriptingFloatChannel}
          Top {type=MovieSceneScriptingFloatChannel, keys [[0.0003,-20],[3.9613,-16]]}
          Right {type=MovieSceneScriptingFloatChannel}
          Bottom {type=MovieSceneScriptingFloatChannel}
         ]
        }]
      }]
    }
    ScaleBox_IMGStarterKit {id=#2, possessed_class=ScaleBox, tracks []}
    OverlayKeyImage {id=#5, possessed_class=Overlay, tracks []}
    IMG_ButtonBackground {
     id=#6
     possessed_class=Image
     tracks [
      {
       type=MovieSceneWidgetMaterialTrack
       sections [{
         type=MovieSceneParameterSection
         end_frame=4
         start_frame=0
         channels [
          BottomEdge_1 {type=MovieSceneScriptingFloatChannel, keys [[0,0.03],[3.9613,0.065]]}
          BottomRotation_1 {type=MovieSceneScriptingFloatChannel, keys [[0,-0.002],[3.9613,0.008]]}
          TopEdge_1 {type=MovieSceneScriptingFloatChannel, keys [[0,0.04],[3.9613,0.01]]}
          TopRotation_1 {type=MovieSceneScriptingFloatChannel, keys [[0,0.005],[3.9613,-0.006]]}
          LeftEdge_1 {type=MovieSceneScriptingFloatChannel, keys [[0,0.01],[3.9613,0.02]]}
          LeftRotation_1 {type=MovieSceneScriptingFloatChannel, keys [[0,-0.0015],[3.9613,0.0055]]}
          RightEdge_1 {type=MovieSceneScriptingFloatChannel, keys [[0,0.01],[3.9613,0.018]]}
          RightRotation_1 {type=MovieSceneScriptingFloatChannel, keys [[0,0],[3.9613,-0.003]]}
         ]
        }]
      }
      {
       type=MovieSceneColorTrack
       display_name=Color and Opacity
       sections [{
         type=MovieSceneColorSection
         end_frame=4
         start_frame=0
         channels [
          Color.R {type=MovieSceneScriptingFloatChannel, keys [[0,0],[3.9613,1]]}
          Color.G {type=MovieSceneScriptingFloatChannel, keys [[0,0.3613],[3.9613,1]]}
          Color.B {type=MovieSceneScriptingFloatChannel, const [1,2]}
          Color.A {type=MovieSceneScriptingFloatChannel, const [1,2]}
         ]
        }]
      }
     ]
    }
    VerticalBoxSlot_1 {
     id=#7
     parent_id=#8
     possessed_class=VerticalBoxSlot
     tracks [{
       type=MovieSceneMarginTrack
       display_name=Padding
       sections [{
         type=MovieSceneMarginSection
         end_frame=4
         start_frame=0
         channels [
          Left {type=MovieSceneScriptingFloatChannel}
          Top {type=MovieSceneScriptingFloatChannel, keys [[0,110],[3.9613,120]]}
          Right {type=MovieSceneScriptingFloatChannel}
          Bottom {type=MovieSceneScriptingFloatChannel}
         ]
        }]
      }]
    }
    Text_OfferContent {
     id=#9
     possessed_class=CommonTextBlock
     tracks [{
       type=MovieSceneColorTrack
       display_name=Color and Opacity
       sections [{
         type=MovieSceneColorSection
         end_frame=4
         start_frame=-1
         channels [
          Color.R {type=MovieSceneScriptingFloatChannel, keys [[0.0003,1],[3.9613,0]]}
          Color.G {type=MovieSceneScriptingFloatChannel, keys [[0.0003,1],[3.9613,0]]}
          Color.B {type=MovieSceneScriptingFloatChannel, keys [[0.0003,1],[3.9613,0]]}
          Color.A {type=MovieSceneScriptingFloatChannel, const [1,2]}
         ]
        }]
      }]
    }
    OptionalIMG_StarterKit {
     id=#10
     possessed_class=Image
     tracks [{
       type=MovieSceneWidgetMaterialTrack
       sections [{
         type=MovieSceneParameterSection
         end_frame=4
         start_frame=-1
         channels [
          BottomEdge_1 {type=MovieSceneScriptingFloatChannel, keys [[0.0003,0.012],[3.9613,0.0215]]}
          BottomRotation_1 {type=MovieSceneScriptingFloatChannel, keys [[0.0003,-0.0001],[3.9613,0.0018]]}
         ]
        }]
      }]
    }
    Overlay {id=#8, possessed_class=Overlay, tracks []}
    IMG_PriceBackground {
     id=#11
     possessed_class=Image
     tracks [{
       type=MovieSceneWidgetMaterialTrack
       sections [{
         type=MovieSceneParameterSection
         end_frame=4
         start_frame=0
         channels [
          BottomRotation_1 {type=MovieSceneScriptingFloatChannel, keys [[0,-0.002],[3.9613,0.02]]}
          BottomEdge_1 {type=MovieSceneScriptingFloatChannel, keys [[0,0.03],[3.9613,0.15]]}
          RightEdge_1 {type=MovieSceneScriptingFloatChannel, keys [[0,0.0095],[3.9613,0.024]]}
          RightRotation_1 {type=MovieSceneScriptingFloatChannel, keys [[0,0],[3.9613,-0.001]]}
          LeftRotation_1 {type=MovieSceneScriptingFloatChannel, keys [[0,-0.0006],[3.9613,0.002]]}
          LeftEdge_1 {type=MovieSceneScriptingFloatChannel, keys [[0,0.0065],[3.9613,0.031]]}
          TopEdge_1 {type=MovieSceneScriptingFloatChannel, keys [[0,0.04],[3.9613,0.01]]}
          TopRotation_1 {type=MovieSceneScriptingFloatChannel, keys [[0,0.005],[3.9613,-0.005]]}
         ]
        }]
      }]
    }
    SizeBoxSlot_0 {
     id=#12
     parent_id=#13
     possessed_class=SizeBoxSlot
     tracks [{
       type=MovieSceneMarginTrack
       display_name=Padding
       sections [{
         type=MovieSceneMarginSection
         end_frame=4
         start_frame=0
         channels [
          Left {type=MovieSceneScriptingFloatChannel, keys [[0,0],[3.9613,-4]]}
          Top {type=MovieSceneScriptingFloatChannel, keys [[0,0],[3.9613,-16]]}
          Right {type=MovieSceneScriptingFloatChannel, keys [[0,0],[3.9613,-4]]}
          Bottom {type=MovieSceneScriptingFloatChannel, keys [[0,0],[3.9613,-16]]}
         ]
        }]
      }]
    }
    Overlay_1 {id=#13, possessed_class=Overlay, tracks []}
    SizeBox {
     id=#14
     possessed_class=SizeBox
     tracks [
      {
       type=MovieSceneFloatTrack
       display_name=Render Opacity
       sections [{type=MovieSceneFloatSection, end_frame=4, start_frame=0, channels [None {type=MovieSceneScriptingFloatChannel, keys [[0,0],[3.9613,1]]}]}]
      }
      {
       type=MovieScene2DTransformTrack
       sections [{
         type=MovieScene2DTransformSection
         end_frame=4
         start_frame=0
         channels [
          Translation.X {type=MovieSceneScriptingFloatChannel}
          Translation.Y {type=MovieSceneScriptingFloatChannel, keys [[0,-26],[3.9613,-14]]}
          Angle {type=MovieSceneScriptingFloatChannel}
          Scale.X {type=MovieSceneScriptingFloatChannel, keys [[0,0.75],[3.9613,1]]}
          Scale.Y {type=MovieSceneScriptingFloatChannel, keys [[0,0.75],[3.9613,1]]}
          Shear.X {type=MovieSceneScriptingFloatChannel}
          Shear.Y {type=MovieSceneScriptingFloatChannel}
         ]
        }]
      }
     ]
    }
   ]}
 }]
graphs [EventGraph {nodes [
   Event On Hovered()
   n1 = Get OnHover() -> (OnHover:Widget Animation Object Reference)
   > PlayAnimationForward(InAnimation=@n1, PlaybackSpeed=1.5)
   Event On Unhovered()
   > PlayAnimationReverse(InAnimation=@n1, PlaybackSpeed=2)
  ]}]
variables []
widgets {
 info {
  inheritedWidgetCount=8
  parentClass=/Script/FortniteUI.AthenaStarterKitOfferButtonLegacy
  rootWidgetClass=&Overlay
  widgetCount=34
 }
 widgets [
  Overlay_36 {widgetClassPath=&Overlay, widget=Overlay_36}
  ButtonSize {
   widgetClassPath=&SizeBox
   parent=Overlay_36
   widget=ButtonSize
   properties {bOverride_HeightOverride=true, bOverride_MinDesiredWidth=true, heightOverride=160, minDesiredWidth=840, widthOverride=650}
   slot {class=OverlaySlot, properties {horizontalAlignment=HAlign_Fill, verticalAlignment=VAlign_Fill}}
  }
  Overlay_1 {widgetClassPath=&Overlay, parent=ButtonSize, widget=Overlay_1, slot {class=SizeBoxSlot}}
  IMG_ButtonBackground {
   widgetClassPath=&Image
   bIsVariable=true
   parent=Overlay_1
   widget=IMG_ButtonBackground
   properties {
    brush {
     bIsDynamicallyLoaded=false
     drawAs=Box
     imageType=NoImage
     mirroring=NoMirror
     resourceName=None
     resourceObject=&/MI_UI_ReloadMtx_DynamicOfferButton_Legacy
     tiling=NoTile
     imageSize=<32 32>
     margin=<ltrb 0 0 0 0>
     outlineSettings {
      bUseBrushTransparency=false
      roundingType=HalfHeightRadius
      width=0
      color {colorUseRule=UseColor_Specified, specifiedColor=<rgba 0 0 0 0>}
      cornerRadii=<0 0 0 0>
     }
     tintColor {colorUseRule=UseColor_Specified, specifiedColor=<rgba 1 1 1 1>}
     uVRegion {bIsValid=false, max=<0 0>, min=<0 0>}
    }
    colorAndOpacity=<rgba 0 0.3613 1 1>
   }
   slot {class=OverlaySlot, properties {horizontalAlignment=HAlign_Fill, verticalAlignment=VAlign_Fill}}
  }
  Overlay_0 {
   widgetClassPath=&Overlay
   parent=Overlay_1
   widget=Overlay_0
   slot {class=OverlaySlot, properties {horizontalAlignment=HAlign_Fill, verticalAlignment=VAlign_Fill}}
  }
  Overlay {
   widgetClassPath=&Overlay
   parent=Overlay_0
   widget=Overlay
   slot {class=OverlaySlot, properties {horizontalAlignment=HAlign_Fill, verticalAlignment=VAlign_Fill, padding=<ltrb 0 110 0 0>}}
  }
  IMG_PriceBackground {
   widgetClassPath=&Image
   bIsVariable=true
   parent=Overlay
   widget=IMG_PriceBackground
   properties {
    brush {
     bIsDynamicallyLoaded=false
     drawAs=Image
     imageType=NoImage
     mirroring=NoMirror
     resourceName=None
     resourceObject=&/MI_UI_ReloadMtx_DynamicOfferContainer_Legacy
     tiling=NoTile
     imageSize=<32 32>
     margin=<ltrb 0.0092 0 0 0>
     outlineSettings {
      bUseBrushTransparency=false
      roundingType=HalfHeightRadius
      width=0
      color {colorUseRule=UseColor_Specified, specifiedColor=<rgba 0 0 0 0>}
      cornerRadii=<0 0 0 0>
     }
     tintColor {colorUseRule=UseColor_Specified, specifiedColor=<rgba 1 1 1 1>}
     uVRegion {bIsValid=false, max=<0 0>, min=<0 0>}
    }
    colorAndOpacity=<rgba 1 0.9734 0 1>
   }
   slot {class=OverlaySlot, properties {horizontalAlignment=HAlign_Fill, verticalAlignment=VAlign_Fill}}
  }
  VerticalBox_2 {
   widgetClassPath=&VerticalBox
   parent=Overlay
   widget=VerticalBox_2
   slot {class=OverlaySlot, properties {horizontalAlignment=HAlign_Fill, verticalAlignment=VAlign_Center}}
  }
  Overlay_2 {
   widgetClassPath=&Overlay
   parent=VerticalBox_2
   widget=Overlay_2
   slot {class=VerticalBoxSlot, properties {padding=<ltrb 0 10 0 10>}}
  }
  Text_RealMoneyPrice {
   widgetClassPath=&CommonTextBlock
   bInherited=true
   parent=Overlay_2
   widget=Text_RealMoneyPrice
   properties {
    justification=Center
    scrollStyle=/FortUILibrary/Styles/Typography/Foundation/TextScrollStyle-NoFade.TextScrollStyle-NoFade_C
    text=3.99 $ US
    textTransformPolicy=ToUpper
    colorAndOpacity {colorUseRule=UseColor_Specified, specifiedColor=<rgba 0 0 0 1>}
    font {
     bForceMonospaced=false
     bMaterialIsStencil=false
     fontMaterial=null
     fontObject=&BurbankSmall
     letterSpacing=0
     monospacedWidth=1
     size=20
     skewAmount=0
     typefaceFontName=Black
     outlineSettings {
      bApplyOutlineToDropShadows=false
      bMiteredCorners=false
      bSeparateFillAlpha=false
      outlineMaterial=null
      outlineSize=0
      outlineColor=<rgba 0 0 0 1>
     }
    }
    renderTransform {angle=0, scale=<1 1>, shear=<-8 0>, translation=<0 0>}
    shadowOffset=<0 0>
   }
   slot {class=OverlaySlot, properties {horizontalAlignment=HAlign_Fill, verticalAlignment=VAlign_Center}}
  }
  SizeBox {
   widgetClassPath=&SizeBox
   parent=VerticalBox_2
   widget=SizeBox
   properties {
    bOverride_HeightOverride=true
    bOverride_WidthOverride=true
    heightOverride=48
    widthOverride=48
    renderTransform {angle=0, scale=<1 1>, shear=<0 0>, translation=<0 -26>}
   }
   slot {class=VerticalBoxSlot, properties {horizontalAlignment=HAlign_Center, verticalAlignment=VAlign_Top}}
  }
  InputActionWidget {
   widgetClassPath=&CommonActionWidget
   bInherited=true
   bIsVariable=true
   parent=SizeBox
   widget=InputActionWidget
   properties {iconRimBrush {
     bIsDynamicallyLoaded=false
     drawAs=Image
     imageType=NoImage
     mirroring=NoMirror
     resourceName=None
     resourceObject=/FortUILibrary/Foundation/Buttons/Materials/M-Button-Console-YellowOutline
     tiling=NoTile
     imageSize=<48 48>
     margin=<ltrb 0 0 0 0>
     outlineSettings {
      bUseBrushTransparency=false
      roundingType=HalfHeightRadius
      width=0
      color {colorUseRule=UseColor_Specified, specifiedColor=<rgba 0 0 0 0>}
      cornerRadii=<0 0 0 0>
     }
     tintColor {colorUseRule=UseColor_Specified, specifiedColor=<rgba 1 1 1 1>}
     uVRegion {bIsValid=false, max=<0 0>, min=<0 0>}
    }}
   slot {class=SizeBoxSlot}
  }
  HorizontalBox_0 {
   widgetClassPath=&HorizontalBox
   parent=Overlay_0
   widget=HorizontalBox_0
   slot {class=OverlaySlot, properties {horizontalAlignment=HAlign_Fill, verticalAlignment=VAlign_Fill}}
  }
  OverlayKeyImage {
   widgetClassPath=&Overlay
   parent=HorizontalBox_0
   widget=OverlayKeyImage
   properties {visibility=HitTestInvisible, renderTransform {angle=0, scale=<1 1>, shear=<0 0>, translation=<-42 0>}}
   slot {class=HorizontalBoxSlot, properties {horizontalAlignment=HAlign_Left, padding=<ltrb 0 -20 -80 0>, size {sizeRule=Fill, value=0.4}}}
  }
  LoadGuardCurrencySpinner {
   widgetClassPath=/Script/CommonUI.CommonLoadGuard
   bInherited=true
   bIsVariable=true
   parent=OverlayKeyImage
   widget=LoadGuardCurrencySpinner
   properties {
    loadingText=""
    textStyle=&TextStyle-Base_C
    renderTransform {angle=0, scale=<1 1>, shear=<0 0>, translation=<80 0>}
   }
   slot {class=OverlaySlot, properties {horizontalAlignment=HAlign_Fill, verticalAlignment=VAlign_Fill}}
  }
  ScaleBox {
   widgetClassPath=&ScaleBox
   parent=LoadGuardCurrencySpinner
   widget=ScaleBox
   properties {stretch=ScaleToFitY, stretchDirection=DownOnly}
   slot {class=LoadGuardSlot}
  }
  IMGCurrency {
   widgetClassPath=&Image
   bIsVariable=true
   parent=ScaleBox
   widget=IMGCurrency
   properties {brush {
     bIsDynamicallyLoaded=false
     drawAs=Image
     imageType=NoImage
     mirroring=NoMirror
     resourceName=None
     resourceObject=/Game/UI/Foundation/Textures/Founders/T-VBucks-Tile-1000
     tiling=NoTile
     imageSize=<512 512>
     margin=<ltrb 0 0 0 0>
     outlineSettings {
      bUseBrushTransparency=false
      roundingType=HalfHeightRadius
      width=0
      color {colorUseRule=UseColor_Specified, specifiedColor=<rgba 0 0 0 0>}
      cornerRadii=<0 0 0 0>
     }
     tintColor {colorUseRule=UseColor_Specified, specifiedColor=<rgba 1 1 1 1>}
     uVRegion {bIsValid=false, max=<0 0>, min=<0 0>}
    }}
   slot {class=ScaleBoxSlot, properties {horizontalAlignment=HAlign_Right}}
  }
  ScaleBox_IMGStarterKit {
   widgetClassPath=&ScaleBox
   bIsVariable=true
   parent=OverlayKeyImage
   widget=ScaleBox_IMGStarterKit
   properties {stretch=ScaleToFitY, stretchDirection=DownOnly}
   slot {class=OverlaySlot, properties {verticalAlignment=VAlign_Bottom}}
  }
  OptionalIMG_StarterKit {
   widgetClassPath=&Image
   bInherited=true
   bIsVariable=true
   parent=ScaleBox_IMGStarterKit
   widget=OptionalIMG_StarterKit
   properties {brush {
     bIsDynamicallyLoaded=false
     drawAs=Image
     imageType=NoImage
     mirroring=NoMirror
     resourceName=None
     resourceObject=&/MI_UI_ReloadMtx_DynamicStarterKitArt_Legacy
     tiling=NoTile
     imageSize=<512 512>
     margin=<ltrb 0 0 0 0>
     outlineSettings {
      bUseBrushTransparency=false
      roundingType=HalfHeightRadius
      width=0
      color {colorUseRule=UseColor_Specified, specifiedColor=<rgba 0 0 0 0>}
      cornerRadii=<0 0 0 0>
     }
     tintColor {colorUseRule=UseColor_Specified, specifiedColor=<rgba 1 1 1 1>}
     uVRegion {bIsValid=false, max=<0 0>, min=<0 0>}
    }}
   slot {class=ScaleBoxSlot, properties {verticalAlignment=VAlign_Bottom}}
  }
  Overlay_3 {
   widgetClassPath=&Overlay
   parent=HorizontalBox_0
   widget=Overlay_3
   slot {class=HorizontalBoxSlot, properties {padding=<ltrb 40 0 0 40>, size {sizeRule=Fill, value=0.8}}}
  }
  HorizontalBox {
   widgetClassPath=&HorizontalBox
   parent=Overlay_3
   widget=HorizontalBox
   slot {class=OverlaySlot, properties {horizontalAlignment=HAlign_Fill, verticalAlignment=VAlign_Fill}}
  }
  VB_MtxTexts_1 {
   widgetClassPath=&VerticalBox
   parent=HorizontalBox
   widget=VB_MtxTexts_1
   properties {renderTransform {angle=0, scale=<1 1>, shear=<0 0>, translation=<0 6>}}
   slot {class=HorizontalBoxSlot, properties {verticalAlignment=VAlign_Center, size {sizeRule=Fill, value=1}}}
  }
  ScaleBox_4 {
   widgetClassPath=&ScaleBox
   parent=VB_MtxTexts_1
   widget=ScaleBox_4
   properties {stretchDirection=DownOnly}
   slot {class=VerticalBoxSlot}
  }
  Text_DisplayName {
   widgetClassPath=&CommonTextBlock
   bInherited=true
   bIsVariable=true
   parent=ScaleBox_4
   widget=Text_DisplayName
   properties {
    justification=Center
    text=Starter kit name
    textTransformPolicy=ToUpper
    font {
     bForceMonospaced=false
     bMaterialIsStencil=false
     fontMaterial=null
     fontObject=&BurbankSmall
     letterSpacing=0
     monospacedWidth=1
     size=28
     skewAmount=0
     typefaceFontName=Black
     outlineSettings {
      bApplyOutlineToDropShadows=false
      bMiteredCorners=false
      bSeparateFillAlpha=false
      outlineMaterial=null
      outlineSize=0
      outlineColor=<rgba 0 0 0 1>
     }
    }
    renderTransform {angle=0, scale=<1 1>, shear=<-8 0>, translation=<0 0>}
    shadowOffset=<0 0>
   }
   slot {class=ScaleBoxSlot}
  }
  Text_OfferContent {
   widgetClassPath=&CommonTextBlock
   bInherited=true
   parent=VB_MtxTexts_1
   widget=Text_OfferContent
   properties {
    justification=Center
    text=600 V-Bucks + 3 items
    textTransformPolicy=ToUpper
    font {
     bForceMonospaced=false
     bMaterialIsStencil=false
     fontMaterial=null
     fontObject=&BurbankSmall
     letterSpacing=0
     monospacedWidth=1
     size=20
     skewAmount=0
     typefaceFontName=Black
     outlineSettings {
      bApplyOutlineToDropShadows=false
      bMiteredCorners=false
      bSeparateFillAlpha=false
      outlineMaterial=null
      outlineSize=0
      outlineColor=<rgba 0 0 0 1>
     }
    }
    renderTransform {angle=0, scale=<1 1>, shear=<-8 0>, translation=<0 -4>}
    shadowOffset=<0 0>
   }
   slot {class=VerticalBoxSlot, properties {verticalAlignment=VAlign_Top}}
  }
  Spacer_48 {
   widgetClassPath=&Spacer
   parent=HorizontalBox_0
   widget=Spacer_48
   slot {class=HorizontalBoxSlot, properties {size {sizeRule=Fill, value=0.18}}}
  }
  Border_BonusPercentArea {
   widgetClassPath=&Border
   parent=Overlay_1
   widget=Border_BonusPercentArea
   properties {
    visibility=HitTestInvisible
    background {
     bIsDynamicallyLoaded=false
     drawAs=Box
     imageType=NoImage
     mirroring=NoMirror
     resourceName=None
     resourceObject=/FortniteCoreUI/Frontend/StoreMain/Art/M_UI_BonusTag
     tiling=NoTile
     imageSize=<32 32>
     margin=<ltrb 0.5 0 0.5 0>
     outlineSettings {
      bUseBrushTransparency=false
      roundingType=HalfHeightRadius
      width=0
      color {colorUseRule=UseColor_Specified, specifiedColor=<rgba 0 0 0 0>}
      cornerRadii=<0 0 0 0>
     }
     tintColor {colorUseRule=UseColor_Specified, specifiedColor=<rgba 1 1 1 1>}
     uVRegion {bIsValid=false, max=<0 0>, min=<0 0>}
    }
    padding=<ltrb 16 8 10 8>
    renderTransform {angle=1, scale=<-1 1>, shear=<0 0>, translation=<0 0>}
   }
   slot {class=OverlaySlot, properties {horizontalAlignment=HAlign_Right, padding=<ltrb -10 -24 -12 0>}}
  }
  HorizontalBox_5 {
   widgetClassPath=&HorizontalBox
   parent=Border_BonusPercentArea
   widget=HorizontalBox_5
   properties {renderTransform {angle=1, scale=<-1 1>, shear=<0 0>, translation=<0 1>}}
   slot {class=BorderSlot, properties {padding=<ltrb 16 8 10 8>}}
  }
  SizeBox_TimerIcon {
   widgetClassPath=&SizeBox
   parent=HorizontalBox_5
   widget=SizeBox_TimerIcon
   properties {bOverride_HeightOverride=true, bOverride_WidthOverride=true, heightOverride=56, widthOverride=56}
   slot {class=HorizontalBoxSlot, properties {padding=<ltrb -8 -18 0 -6>}}
  }
  Image_TimerIcon {
   widgetClassPath=&Image
   bIsVariable=true
   parent=SizeBox_TimerIcon
   widget=Image_TimerIcon
   properties {brush {
     bIsDynamicallyLoaded=false
     drawAs=Image
     imageType=NoImage
     mirroring=NoMirror
     resourceName=None
     resourceObject=&/MI_UI_ReloadMtx_Timer_Icon_Legacy
     tiling=NoTile
     imageSize=<32 32>
     margin=<ltrb 0 0 0 0>
     outlineSettings {
      bUseBrushTransparency=false
      roundingType=HalfHeightRadius
      width=0
      color {colorUseRule=UseColor_Specified, specifiedColor=<rgba 0 0 0 0>}
      cornerRadii=<0 0 0 0>
     }
     tintColor {colorUseRule=UseColor_Specified, specifiedColor=<rgba 1 1 1 1>}
     uVRegion {bIsValid=false, max=<0 0>, min=<0 0>}
    }}
   slot {class=SizeBoxSlot}
  }
  SB_BonusPercentArea {
   widgetClassPath=&SizeBox
   parent=HorizontalBox_5
   widget=SB_BonusPercentArea
   properties {bOverride_MaxDesiredWidth=true, maxDesiredWidth=256}
   slot {class=HorizontalBoxSlot}
  }
  Text_BonusPercent {
   widgetClassPath=&CommonTextBlock
   parent=SB_BonusPercentArea
   widget=Text_BonusPercent
   properties {
    justification=Center
    text=Limited time
    textTransformPolicy=ToUpper
    font {
     bForceMonospaced=false
     bMaterialIsStencil=false
     fontMaterial=null
     fontObject=&BurbankSmall
     letterSpacing=0
     monospacedWidth=1
     size=20
     skewAmount=0
     typefaceFontName=Black
     outlineSettings {
      bApplyOutlineToDropShadows=false
      bMiteredCorners=false
      bSeparateFillAlpha=false
      outlineMaterial=null
      outlineSize=0
      outlineColor=<rgba 0 0 0 1>
     }
    }
    renderTransform {angle=0, scale=<1 1>, shear=<-8 0>, translation=<0 0>}
    shadowOffset=<0 0>
   }
   slot {class=SizeBoxSlot, properties {horizontalAlignment=HAlign_Center, verticalAlignment=VAlign_Center}}
  }
  Text_MtxAmount {widgetClassPath=&CommonTextBlock, bInherited=true, widget=null}
  OptionalSwitcher_CurrencyIcons {widgetClassPath=&CommonVisibilitySwitcher, bInherited=true, widget=null}
 ]
}
