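# Legacy item shop screen for reloading V-Bucks, showing required amount, current balance and shortfall alongside purchase options.
# The native parent binds key widgets, including a switcher holding required-amount and starter-kit offers, plus an alternative offer and More Options button.
# Activation plays a staggered slide-and-fade intro; BlockScreenContent swaps content for a purchase blocker.
# Construct pads offer details for mobile Bacchus frontend, and an override supplies FramingWidget for viewed-item framing.
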WidgetBlueprint /ItemShop/UI/ReloadMtx/DynamicReloadMtxSreen_Legacy
animations [Intro {
  binding_count=10
  length_seconds=0.3919
  object=&self:Intro
  sequence {
   bindings [
    Button_MtxOtherOption {
     id=#1
     possessed_class=null
     tracks [
      {
       type=MovieSceneFloatTrack
       display_name=Render Opacity
       sections [{
         type=MovieSceneFloatSection
         end_frame=22
         start_frame=0
         channels [None {type=MovieSceneScriptingFloatChannel, keys [[0.113,0],[11.099,0],[20.537,1]]}]
        }]
      }
      {
       type=MovieScene2DTransformTrack
       sections [{
         type=MovieScene2DTransformSection
         end_frame=22
         start_frame=0
         channels [
          Translation.X {type=MovieSceneScriptingFloatChannel}
          Translation.Y {type=MovieSceneScriptingFloatChannel, keys [[0.113,40],[11.099,40],[20.537,0]]}
         ]
        }]
      }
     ]
    }
    Overlay_4 {
     id=#2
     possessed_class=Overlay
     tracks [
      {
       type=MovieSceneFloatTrack
       display_name=Render Opacity
       sections [{
         type=MovieSceneFloatSection
         end_frame=23
         start_frame=-1
         channels [None {type=MovieSceneScriptingFloatChannel, keys [[0.113,0],[9.038,0],[18.476,1]]}]
        }]
      }
      {
       type=MovieScene2DTransformTrack
       sections [{
         type=MovieScene2DTransformSection
         end_frame=22
         start_frame=-1
         channels [
          Translation.X {type=MovieSceneScriptingFloatChannel}
          Translation.Y {type=MovieSceneScriptingFloatChannel, keys [[0.113,80],[9.038,80],[18.476,0]]}
          Angle {type=MovieSceneScriptingFloatChannel}
          Scale.X {type=MovieSceneScriptingFloatChannel}
          Scale.Y {type=MovieSceneScriptingFloatChannel}
          Shear.X {type=MovieSceneScriptingFloatChannel}
          Shear.Y {type=MovieSceneScriptingFloatChannel}
         ]
        }]
      }
     ]
    }
    ScaleBox_1 {
     id=#3
     possessed_class=ScaleBox
     tracks [
      {
       type=MovieSceneFloatTrack
       display_name=Render Opacity
       sections [{
         type=MovieSceneFloatSection
         end_frame=22
         start_frame=0
         channels [None {type=MovieSceneScriptingFloatChannel, keys [[0.113,0],[3.346,0],[19,1]]}]
        }]
      }
      {
       type=MovieScene2DTransformTrack
       sections [{
         type=MovieScene2DTransformSection
         end_frame=22
         start_frame=0
         channels [
          Translation.X {type=MovieSceneScriptingFloatChannel, keys [[0.113,-460],[3.346,-460],[19,0]]}
          Translation.Y {type=MovieSceneScriptingFloatChannel}
          Angle {type=MovieSceneScriptingFloatChannel}
          Scale.X {type=MovieSceneScriptingFloatChannel}
          Scale.Y {type=MovieSceneScriptingFloatChannel}
          Shear.X {type=MovieSceneScriptingFloatChannel, keys [[0.113,-10],[3.346,-10],[19,0]]}
          Shear.Y {type=MovieSceneScriptingFloatChannel}
         ]
        }]
      }
     ]
    }
    HB_OR_Separator {
     id=#4
     possessed_class=HorizontalBox
     tracks [
      {
       type=MovieScene2DTransformTrack
       sections [{
         type=MovieScene2DTransformSection
         end_frame=22
         start_frame=0
         channels [
          Translation.X {type=MovieSceneScriptingFloatChannel}
          Translation.Y {type=MovieSceneScriptingFloatChannel, keys [[0.113,60],[10.141,60],[19.579,0]]}
         ]
        }]
      }
      {
       type=MovieSceneFloatTrack
       display_name=Render Opacity
       sections [{
         type=MovieSceneFloatSection
         end_frame=22
         start_frame=0
         channels [None {type=MovieSceneScriptingFloatChannel, keys [[0.113,0],[10.141,0],[19.579,1]]}]
        }]
      }
     ]
    }
    SmallOfferPreview_Widget {
     id=#5
     possessed_class=AthenaSmallOfferPreview_C
     tracks [
      {
       type=MovieScene2DTransformTrack
       sections [{
         type=MovieScene2DTransformSection
         end_frame=22
         start_frame=0
         channels [
          Translation.X {type=MovieSceneScriptingFloatChannel, keys [[0.113,200],[12.009,200],[21.446,0]]}
          Translation.Y {type=MovieSceneScriptingFloatChannel}
         ]
        }]
      }
      {
       type=MovieSceneFloatTrack
       display_name=Render Opacity
       sections [{
         type=MovieSceneFloatSection
         end_frame=22
         start_frame=0
         channels [None {type=MovieSceneScriptingFloatChannel, keys [[0.113,0],[12.009,0],[21.446,1]]}]
        }]
      }
     ]
    }
    IMG_Background {
     id=#6
     possessed_class=Image
     tracks [
      {
       type=MovieScene2DTransformTrack
       sections [{
         type=MovieScene2DTransformSection
         end_frame=22
         start_frame=-1
         channels [
          Translation.X {type=MovieSceneScriptingFloatChannel, keys [[4.314,-510],[19.409,0]]}
          Translation.Y {type=MovieSceneScriptingFloatChannel}
          Angle {type=MovieSceneScriptingFloatChannel}
          Scale.X {type=MovieSceneScriptingFloatChannel}
          Scale.Y {type=MovieSceneScriptingFloatChannel}
          Shear.X {type=MovieSceneScriptingFloatChannel}
          Shear.Y {type=MovieSceneScriptingFloatChannel}
         ]
        }]
      }
      {
       type=MovieSceneFloatTrack
       display_name=Render Opacity
       sections [{
         type=MovieSceneFloatSection
         end_frame=22
         start_frame=0
         channels [None {type=MovieSceneScriptingFloatChannel, keys [[4.228,0],[16.921,1]]}]
        }]
      }
     ]
    }
    HB_Title {
     id=#7
     possessed_class=HorizontalBox
     tracks [
      {
       type=MovieScene2DTransformTrack
       sections [{
         type=MovieScene2DTransformSection
         end_frame=22
         start_frame=0
         channels [
          Translation.X {type=MovieSceneScriptingFloatChannel, keys [[0.113,-460],[16,0]]}
          Translation.Y {type=MovieSceneScriptingFloatChannel}
          Angle {type=MovieSceneScriptingFloatChannel}
          Scale.X {type=MovieSceneScriptingFloatChannel}
          Scale.Y {type=MovieSceneScriptingFloatChannel}
          Shear.X {type=MovieSceneScriptingFloatChannel, keys [[0.113,-10],[16,0]]}
          Shear.Y {type=MovieSceneScriptingFloatChannel}
         ]
        }]
      }
      {
       type=MovieSceneFloatTrack
       display_name=Render Opacity
       sections [{type=MovieSceneFloatSection, end_frame=22, start_frame=0, channels [None {type=MovieSceneScriptingFloatChannel, keys [[0.113,0],[16,1]]}]}]
      }
     ]
    }
    HB_SelectAnOption {
     id=#8
     possessed_class=HorizontalBox
     tracks [
      {
       type=MovieScene2DTransformTrack
       sections [{
         type=MovieScene2DTransformSection
         end_frame=22
         start_frame=0
         channels [
          Translation.X {type=MovieSceneScriptingFloatChannel}
          Translation.Y {type=MovieSceneScriptingFloatChannel, keys [[0.113,120],[8.081,120],[17.519,0]]}
         ]
        }]
      }
      {
       type=MovieSceneFloatTrack
       display_name=Render Opacity
       sections [{
         type=MovieSceneFloatSection
         end_frame=22
         start_frame=0
         channels [None {type=MovieSceneScriptingFloatChannel, keys [[0.113,0],[8.081,0],[17.519,1]]}]
        }]
      }
     ]
    }
    IMG_Separator {
     id=#9
     possessed_class=Image
     tracks [
      {
       type=MovieScene2DTransformTrack
       sections [{
         type=MovieScene2DTransformSection
         end_frame=22
         start_frame=-1
         channels [
          Translation.X {type=MovieSceneScriptingFloatChannel, keys [[4.228,-490],[19.303,0]]}
          Translation.Y {type=MovieSceneScriptingFloatChannel}
          Angle {type=MovieSceneScriptingFloatChannel}
          Scale.X {type=MovieSceneScriptingFloatChannel}
          Scale.Y {type=MovieSceneScriptingFloatChannel}
          Shear.X {type=MovieSceneScriptingFloatChannel}
          Shear.Y {type=MovieSceneScriptingFloatChannel}
         ]
        }]
      }
      {
       type=MovieSceneFloatTrack
       display_name=Render Opacity
       sections [{
         type=MovieSceneFloatSection
         end_frame=18
         start_frame=0
         channels [None {type=MovieSceneScriptingFloatChannel, keys [[0,0],[14.228,0],[16.921,1]]}]
        }]
      }
     ]
    }
    Button_MoreOptions {
     id=#10
     possessed_class=SoloButton_C
     tracks [
      {
       type=MovieSceneFloatTrack
       display_name=Render Opacity
       sections [{
         type=MovieSceneFloatSection
         end_frame=22
         start_frame=0
         channels [None {type=MovieSceneScriptingFloatChannel, keys [[0.113,0],[13.548,0],[21.76,1]]}]
        }]
      }
      {
       type=MovieScene2DTransformTrack
       sections [{
         type=MovieScene2DTransformSection
         end_frame=22
         start_frame=0
         channels [
          Translation.X {type=MovieSceneScriptingFloatChannel}
          Translation.Y {type=MovieSceneScriptingFloatChannel, keys [[0.113,60],[13.548,60],[21.76,0]]}
         ]
        }]
      }
     ]
    }
   ]
   display_rate {numerator=60}
  }
 }]
graphs [
 ContentBlocked {nodes [
   n1 = ContentBlocked!() -> (then, IsContentBlocked:bool=~, ContentBlockedText:Text) {is=title K2Node_FunctionEntry}
   n2 = Get WBP_ContentPurchaseBlocker() -> (WBP_ContentPurchaseBlocker:WBP Content Purchase Blocker Object Reference)
   > Set Data!(execute, self:WBP Content Purchase Blocker Object Reference=@n2, InText:Text=@n1.ContentBlockedText) -> (then) {is=title K2Node_CallFunction}
   > SetVisibility(self=@n2, InVisibility=Select(Option 0=Collapsed, Option 1=SelfHitTestInvisible, Index=@n1))
   n3 = Select(Option 0=SelfHitTestInvisible, Option 1=Collapsed, Index=@n1)
   > SetVisibility(self=Get SafeZone_2() -> (SafeZone_2:Safe Zone Object Reference), InVisibility=@n3)
   > SetVisibility(self=Get SB_OfferDetailsWidget() -> (SB_OfferDetailsWidget:Size Box Object Reference), InVisibility=@n3)
  ]}
 EventGraph {nodes [
   Event Construct()
   n1 = GetRuntimeOptions()
   > Branch(Condition=AND Boolean(A=IsMobileGame(), B=Get bBacchusFrontendEnabled!(self:Fort Runtime Options Object Reference=@n1) -> (bBacchusFrontendEnabled:bool)))
   > SetPadding!(
    execute
    self:Overlay Slot Object Reference=SlotAsOverlaySlot(Widget=Get SB_OfferDetailsWidget())
    InPadding:Margin Structure=Make Margin(Left=0, Top=0, Right=40, Bottom=140)
   ) -> (then)
   n2 = Event BlockScreenContent!() -> (OutputDelegate, then, bBlockScreen:bool, ContentBlockedText:"Text (by ref)") {is=title K2Node_Event}
   > ContentBlocked!(execute, self, IsContentBlocked:bool=@n2.bBlockScreen, ContentBlockedText:Text=@n2.ContentBlockedText) -> (then) {is=title K2Node_CallFunction}
   Event On Activated()
   > PlayAnimation(InAnimation=Get Intro())
  ]}
 GetWidgetForFramingViewedItem {nodes [
   GetWidgetForFramingViewedItem() {is=title K2Node_FunctionEntry}
   > Return Node!(execute, ReturnValue:Widget Object Reference=Get FramingWidget() -> (FramingWidget:Size Box Object Reference))
  ]}
]
variables []
widgets {
 info {
  inheritedWidgetCount=18
  parentClass=/Script/FortniteUI.AthenaDynamicReloadMtxScreenLegacy
  rootWidgetClass=&Overlay
  widgetCount=87
 }
 widgets [
  Overlay_6 {widgetClassPath=&Overlay, widget=Overlay_6}
  ScaleBox_0 {
   widgetClassPath=&ScaleBox
   parent=Overlay_6
   widget=ScaleBox_0
   properties {stretch=ScaleBySafeZone}
   slot {class=OverlaySlot, properties {horizontalAlignment=HAlign_Fill, verticalAlignment=VAlign_Fill}}
  }
  Overlay_2 {
   widgetClassPath=&Overlay
   parent=ScaleBox_0
   widget=Overlay_2
   slot {class=ScaleBoxSlot, properties {horizontalAlignment=HAlign_Fill, verticalAlignment=VAlign_Fill}}
  }
  HorizontalBox_2 {
   widgetClassPath=&HorizontalBox
   parent=Overlay_2
   widget=HorizontalBox_2
   slot {class=OverlaySlot, properties {horizontalAlignment=HAlign_Fill, verticalAlignment=VAlign_Fill}}
  }
  SB_MainContainer {
   widgetClassPath=&SizeBox
   parent=HorizontalBox_2
   widget=SB_MainContainer
   properties {widthOverride=900}
   slot {class=HorizontalBoxSlot, properties {size {sizeRule=Fill, value=0.45}}}
  }
  Overlay_3 {widgetClassPath=&Overlay, parent=SB_MainContainer, widget=Overlay_3, slot {class=SizeBoxSlot}}
  IMG_Separator {
   widgetClassPath=&Image
   bIsVariable=true
   parent=Overlay_3
   widget=IMG_Separator
   properties {brush {
     bIsDynamicallyLoaded=false
     drawAs=Box
     imageType=NoImage
     mirroring=NoMirror
     resourceName=None
     resourceObject=&/MI_UI_ReloadVBucks_Separator_Legacy
     tiling=NoTile
     imageSize=<128 32>
     margin=<ltrb 0 0 0.001 0>
     outlineSettings {
      bUseBrushTransparency=false
      roundingType=HalfHeightRadius
      width=0
      color {colorUseRule=UseColor_Specified, specifiedColor=<rgba 0 0 0 0>}
      cornerRadii=<0 0 0 0>
     }
     tintColor {colorUseRule=UseColor_Specified, specifiedColor=<rgba 1 1 1 1>}
     uVRegion {bIsValid=false, max=<0 0>, min=<0 0>}
    }}
   slot {class=OverlaySlot, properties {horizontalAlignment=HAlign_Right, verticalAlignment=VAlign_Fill, padding=<ltrb 0 0 -30 0>}}
  }
  IMG_Background {
   widgetClassPath=&Image
   bIsVariable=true
   parent=Overlay_3
   widget=IMG_Background
   properties {brush {
     bIsDynamicallyLoaded=false
     drawAs=Box
     imageType=NoImage
     mirroring=NoMirror
     resourceName=None
     resourceObject=&/MI_UI_ReloadVBucks_BG_Legacy
     tiling=NoTile
     imageSize=<32 32>
     margin=<ltrb 0 0 0.001 0>
     outlineSettings {
      bUseBrushTransparency=false
      roundingType=HalfHeightRadius
      width=0
      color {colorUseRule=UseColor_Specified, specifiedColor=<rgba 0 0 0 0>}
      cornerRadii=<0 0 0 0>
     }
     tintColor {colorUseRule=UseColor_Specified, specifiedColor=<rgba 1 1 1 1>}
     uVRegion {bIsValid=false, max=<0 0>, min=<0 0>}
    }}
   slot {class=OverlaySlot, properties {horizontalAlignment=HAlign_Fill, verticalAlignment=VAlign_Fill}}
  }
  SafeZone_2 {
   widgetClassPath=&SafeZone
   bIsVariable=true
   parent=Overlay_3
   widget=SafeZone_2
   properties {padRight=false}
   slot {class=OverlaySlot, properties {horizontalAlignment=HAlign_Fill, verticalAlignment=VAlign_Fill, padding=<ltrb 0 0 20 50>}}
  }
  VerticalBox_1 {widgetClassPath=&VerticalBox, parent=SafeZone_2, widget=VerticalBox_1, slot {class=SafeZoneSlot, properties {bIsTitleSafe=false}}}
  VerticalBox_0 {
   widgetClassPath=&VerticalBox
   parent=VerticalBox_1
   widget=VerticalBox_0
   slot {class=VerticalBoxSlot, properties {padding=<ltrb 30 0 30 0>, size {sizeRule=Fill, value=0.41}}}
  }
  HB_Title_1 {
   widgetClassPath=&HorizontalBox
   parent=VerticalBox_0
   widget=HB_Title_1
   slot {class=VerticalBoxSlot, properties {horizontalAlignment=HAlign_Center, padding=<ltrb 0 10 0 0>, size {sizeRule=Automatic, value=0.14}}}
  }
  IMG_ReloadVBucksLogo {
   widgetClassPath=&Image
   bIsVariable=true
   parent=HB_Title_1
   widget=IMG_ReloadVBucksLogo
   properties {
    brush {
     bIsDynamicallyLoaded=false
     drawAs=Image
     imageType=NoImage
     mirroring=NoMirror
     resourceName=None
     resourceObject=&/MI_ReloadVBucks_GradientLogo_Legacy
     tiling=NoTile
     imageSize=<168 168>
     margin=<ltrb 0 0 0 0>
     outlineSettings {
      bUseBrushTransparency=false
      roundingType=HalfHeightRadius
      width=0
      color {colorUseRule=UseColor_Specified, specifiedColor=<rgba 0 0 0 0>}
      cornerRadii=<0 0 0 0>
     }
     tintColor {colorUseRule=UseColor_Specified, specifiedColor=<rgba 1 1 1 1>}
     uVRegion {bIsValid=false, max=<0 0>, min=<0 0>}
    }
    colorAndOpacity=<rgba 1 1 1 0.6>
   }
   slot {class=HorizontalBoxSlot, properties {horizontalAlignment=HAlign_Center, verticalAlignment=VAlign_Center}}
  }
  ScaleBox_6 {
   widgetClassPath=&ScaleBox
   parent=HB_Title_1
   widget=ScaleBox_6
   properties {stretchDirection=DownOnly, renderTransform {angle=0, scale=<1 1>, shear=<-12 0>, translation=<0 0>}}
   slot {class=HorizontalBoxSlot, properties {verticalAlignment=VAlign_Center, padding=<ltrb 16 0 10 0>, size {sizeRule=Fill, value=1}}}
  }
  CommonTextBlock_Title {
   widgetClassPath=&CommonTextBlock
   parent=ScaleBox_6
   widget=CommonTextBlock_Title
   properties {
    text=Reload V-Bucks
    textTransformPolicy=ToUpper
    font {
     bForceMonospaced=false
     bMaterialIsStencil=false
     fontMaterial=&/MI_UI_ReloadMtx_Title_Legacy
     fontObject=&Burbank
     letterSpacing=0
     monospacedWidth=1
     size=66
     skewAmount=0
     typefaceFontName=BurbankRegular-Black
     outlineSettings {
      bApplyOutlineToDropShadows=false
      bMiteredCorners=false
      bSeparateFillAlpha=false
      outlineMaterial=null
      outlineSize=0
      outlineColor=<rgba 0 0 0 1>
     }
    }
    shadowOffset=<0 0>
   }
   slot {class=ScaleBoxSlot, properties {horizontalAlignment=HAlign_Fill}}
  }
  ScaleBox_1 {
   widgetClassPath=&ScaleBox
   parent=VerticalBox_0
   widget=ScaleBox_1
   properties {stretchDirection=DownOnly}
   slot {class=VerticalBoxSlot, properties {padding=<ltrb 0 0 0 30>, size {sizeRule=Fill, value=0.16}}}
  }
  Overlay_5 {
   widgetClassPath=&Overlay
   parent=ScaleBox_1
   widget=Overlay_5
   slot {class=ScaleBoxSlot, properties {horizontalAlignment=HAlign_Fill}}
  }
  HorizontalBox_3 {
   widgetClassPath=&HorizontalBox
   parent=Overlay_5
   widget=HorizontalBox_3
   slot {class=OverlaySlot, properties {horizontalAlignment=HAlign_Fill, verticalAlignment=VAlign_Fill}}
  }
  VerticalBox_2 {
   widgetClassPath=&VerticalBox
   parent=HorizontalBox_3
   widget=VerticalBox_2
   slot {class=HorizontalBoxSlot, properties {verticalAlignment=VAlign_Top, padding=<ltrb 0 0 20 0>, size {sizeRule=Fill, value=0.5}}}
  }
  Text_AddFundsNeeded {
   widgetClassPath=&CommonTextBlock
   parent=VerticalBox_2
   widget=Text_AddFundsNeeded
   properties {
    justification=Right
    text=Required V-Bucks:
    colorAndOpacity {colorUseRule=UseColor_Specified, specifiedColor=<rgba 0.008 0.6724 1 1>}
    font {
     bForceMonospaced=false
     bMaterialIsStencil=false
     fontMaterial=null
     fontObject=&BurbankSmall
     letterSpacing=0
     monospacedWidth=1
     size=18
     skewAmount=0
     typefaceFontName=Bold
     outlineSettings {
      bApplyOutlineToDropShadows=false
      bMiteredCorners=false
      bSeparateFillAlpha=false
      outlineMaterial=null
      outlineSize=0
      outlineColor=<rgba 0 0 0 1>
     }
    }
    shadowOffset=<0 0>
   }
   slot {class=VerticalBoxSlot, properties {horizontalAlignment=HAlign_Right, verticalAlignment=VAlign_Center, padding=<ltrb 0 0 0 6>}}
  }
  Text_AdditionalFunds_1 {
   widgetClassPath=&CommonTextBlock
   parent=VerticalBox_2
   widget=Text_AdditionalFunds_1
   properties {
    justification=Right
    text=Current Balance:
    colorAndOpacity {colorUseRule=UseColor_Specified, specifiedColor=<rgba 0.008 0.6724 1 1>}
    font {
     bForceMonospaced=false
     bMaterialIsStencil=false
     fontMaterial=null
     fontObject=&BurbankSmall
     letterSpacing=0
     monospacedWidth=1
     size=18
     skewAmount=0
     typefaceFontName=Bold
     outlineSettings {
      bApplyOutlineToDropShadows=false
      bMiteredCorners=false
      bSeparateFillAlpha=false
      outlineMaterial=null
      outlineSize=0
      outlineColor=<rgba 0 0 0 1>
     }
    }
    shadowOffset=<0 0>
   }
   slot {class=VerticalBoxSlot, properties {horizontalAlignment=HAlign_Right, verticalAlignment=VAlign_Center, padding=<ltrb 0 0 0 6>}}
  }
  Text_AdditionalFunds {
   widgetClassPath=&CommonTextBlock
   parent=VerticalBox_2
   widget=Text_AdditionalFunds
   properties {
    justification=Right
    text=Additional V-Bucks Needed:
    font {
     bForceMonospaced=false
     bMaterialIsStencil=false
     fontMaterial=null
     fontObject=&BurbankSmall
     letterSpacing=0
     monospacedWidth=1
     size=18
     skewAmount=0
     typefaceFontName=Bold
     outlineSettings {
      bApplyOutlineToDropShadows=false
      bMiteredCorners=false
      bSeparateFillAlpha=false
      outlineMaterial=null
      outlineSize=0
      outlineColor=<rgba 0 0 0 1>
     }
    }
    shadowOffset=<0 0>
   }
   slot {class=VerticalBoxSlot, properties {horizontalAlignment=HAlign_Right, verticalAlignment=VAlign_Center, padding=<ltrb 0 7 0 0>}}
  }
  HB_TotalNeeded {
   widgetClassPath=&HorizontalBox
   parent=HorizontalBox_3
   widget=HB_TotalNeeded
   slot {class=HorizontalBoxSlot, properties {verticalAlignment=VAlign_Top, padding=<ltrb 20 0 0 0>, size {sizeRule=Fill, value=0.5}}}
  }
  VerticalBox_5 {
   widgetClassPath=&VerticalBox
   parent=HB_TotalNeeded
   widget=VerticalBox_5
   slot {class=HorizontalBoxSlot, properties {horizontalAlignment=HAlign_Right, verticalAlignment=VAlign_Center}}
  }
  Text_TotalNeeded {
   widgetClassPath=&CommonTextBlock
   bInherited=true
   parent=VerticalBox_5
   widget=Text_TotalNeeded
   properties {
    justification=Right
    text=400
    colorAndOpacity {colorUseRule=UseColor_Specified, specifiedColor=<rgba 0.008 0.6724 1 1>}
    font {
     bForceMonospaced=false
     bMaterialIsStencil=false
     fontMaterial=null
     fontObject=&BurbankSmall
     letterSpacing=0
     monospacedWidth=1
     size=18
     skewAmount=0
     typefaceFontName=Black
     outlineSettings {
      bApplyOutlineToDropShadows=false
      bMiteredCorners=false
      bSeparateFillAlpha=false
      outlineMaterial=null
      outlineSize=0
      outlineColor=<rgba 0 0 0 1>
     }
    }
    shadowOffset=<0 0>
   }
   slot {class=VerticalBoxSlot, properties {horizontalAlignment=HAlign_Right, verticalAlignment=VAlign_Center, padding=<ltrb 0 0 0 6>}}
  }
  Text_CurrentBalance {
   widgetClassPath=&CommonTextBlock
   bInherited=true
   parent=VerticalBox_5
   widget=Text_CurrentBalance
   properties {
    justification=Right
    text=400
    colorAndOpacity {colorUseRule=UseColor_Specified, specifiedColor=<rgba 0.008 0.6724 1 1>}
    font {
     bForceMonospaced=false
     bMaterialIsStencil=false
     fontMaterial=null
     fontObject=&BurbankSmall
     letterSpacing=0
     monospacedWidth=1
     size=18
     skewAmount=0
     typefaceFontName=Black
     outlineSettings {
      bApplyOutlineToDropShadows=false
      bMiteredCorners=false
      bSeparateFillAlpha=false
      outlineMaterial=null
      outlineSize=0
      outlineColor=<rgba 0 0 0 1>
     }
    }
    shadowOffset=<0 0>
   }
   slot {class=VerticalBoxSlot, properties {horizontalAlignment=HAlign_Right, verticalAlignment=VAlign_Center, padding=<ltrb 0 0 0 6>}}
  }
  VerticalBox_7 {
   widgetClassPath=&VerticalBox
   parent=VerticalBox_5
   widget=VerticalBox_7
   slot {class=VerticalBoxSlot, properties {horizontalAlignment=HAlign_Right, verticalAlignment=VAlign_Center, padding=<ltrb 0 7 0 0>}}
  }
  IMG_LeftStroke {
   widgetClassPath=&Image
   bIsVariable=true
   parent=VerticalBox_7
   widget=IMG_LeftStroke
   properties {
    brush {
     bIsDynamicallyLoaded=false
     drawAs=Image
     imageType=NoImage
     mirroring=NoMirror
     resourceName=None
     resourceObject=null
     tiling=NoTile
     imageSize=<32 3>
     margin=<ltrb 0 0 0 0>
     outlineSettings {
      bUseBrushTransparency=false
      roundingType=HalfHeightRadius
      width=0
      color {colorUseRule=UseColor_Specified, specifiedColor=<rgba 0 0 0 0>}
      cornerRadii=<0 0 0 0>
     }
     tintColor {colorUseRule=UseColor_Specified, specifiedColor=<rgba 1 1 1 1>}
     uVRegion {bIsValid=false, max=<0 0>, min=<0 0>}
    }
    colorAndOpacity=<rgba 0.008 0.6724 1 1>
    renderTransform {angle=0, scale=<1 1>, shear=<0 0>, translation=<0 -10>}
   }
   slot {class=VerticalBoxSlot, properties {verticalAlignment=VAlign_Center, padding=<ltrb 20 0 -120 0>}}
  }
  HorizontalBox_8 {
   widgetClassPath=&HorizontalBox
   parent=VerticalBox_7
   widget=HorizontalBox_8
   properties {renderTransform {angle=0, scale=<1 1>, shear=<0 0>, translation=<0 -1.5>}}
   slot {class=VerticalBoxSlot}
  }
  IMG_ReloadVBucksIcon_1 {
   widgetClassPath=&Image
   bIsVariable=true
   parent=HorizontalBox_8
   widget=IMG_ReloadVBucksIcon_1
   properties {brush {
     bIsDynamicallyLoaded=false
     drawAs=Image
     imageType=NoImage
     mirroring=NoMirror
     resourceName=None
     resourceObject=/Game/UI/Foundation/Shell/Textures/T-Icon-VBucks-L
     tiling=NoTile
     imageSize=<30 30>
     margin=<ltrb 0 0 0 0>
     outlineSettings {
      bUseBrushTransparency=false
      roundingType=HalfHeightRadius
      width=0
      color {colorUseRule=UseColor_Specified, specifiedColor=<rgba 0 0 0 0>}
      cornerRadii=<0 0 0 0>
     }
     tintColor {colorUseRule=UseColor_Specified, specifiedColor=<rgba 1 1 1 1>}
     uVRegion {bIsValid=false, max=<0 0>, min=<0 0>}
    }}
   slot {class=HorizontalBoxSlot, properties {horizontalAlignment=HAlign_Right, verticalAlignment=VAlign_Center, padding=<ltrb 0 0 10 0>}}
  }
  Text_AmountNeeded {
   widgetClassPath=&CommonTextBlock
   bInherited=true
   parent=HorizontalBox_8
   widget=Text_AmountNeeded
   properties {
    justification=Right
    text=400
    font {
     bForceMonospaced=false
     bMaterialIsStencil=false
     fontMaterial=null
     fontObject=&BurbankSmall
     letterSpacing=0
     monospacedWidth=1
     size=18
     skewAmount=0
     typefaceFontName=Black
     outlineSettings {
      bApplyOutlineToDropShadows=false
      bMiteredCorners=false
      bSeparateFillAlpha=false
      outlineMaterial=null
      outlineSize=0
      outlineColor=<rgba 0 0 0 1>
     }
    }
    shadowOffset=<0 0>
   }
   slot {class=HorizontalBoxSlot}
  }
  VerticalBox_6 {
   widgetClassPath=&VerticalBox
   parent=HB_TotalNeeded
   widget=VerticalBox_6
   slot {class=HorizontalBoxSlot, properties {horizontalAlignment=HAlign_Left, verticalAlignment=VAlign_Center, padding=<ltrb 10 0 0 0>}}
  }
  Text_V-Bucks {
   widgetClassPath=&CommonTextBlock
   parent=VerticalBox_6
   widget=Text_V-Bucks
   properties {
    text=V-Bucks
    textTransformPolicy=ToUpper
    colorAndOpacity {colorUseRule=UseColor_Specified, specifiedColor=<rgba 0.008 0.6724 1 1>}
    font {
     bForceMonospaced=false
     bMaterialIsStencil=false
     fontMaterial=null
     fontObject=&BurbankSmall
     letterSpacing=0
     monospacedWidth=1
     size=18
     skewAmount=0
     typefaceFontName=Black
     outlineSettings {
      bApplyOutlineToDropShadows=false
      bMiteredCorners=false
      bSeparateFillAlpha=false
      outlineMaterial=null
      outlineSize=0
      outlineColor=<rgba 0 0 0 1>
     }
    }
    shadowOffset=<0 0>
   }
   slot {class=VerticalBoxSlot, properties {horizontalAlignment=HAlign_Left, verticalAlignment=VAlign_Center, padding=<ltrb 0 0 0 6>}}
  }
  Text_V-Bucks_1 {
   widgetClassPath=&CommonTextBlock
   parent=VerticalBox_6
   widget=Text_V-Bucks_1
   properties {
    text=V-Bucks
    textTransformPolicy=ToUpper
    colorAndOpacity {colorUseRule=UseColor_Specified, specifiedColor=<rgba 0.008 0.6724 1 1>}
    font {
     bForceMonospaced=false
     bMaterialIsStencil=false
     fontMaterial=null
     fontObject=&BurbankSmall
     letterSpacing=0
     monospacedWidth=1
     size=18
     skewAmount=0
     typefaceFontName=Black
     outlineSettings {
      bApplyOutlineToDropShadows=false
      bMiteredCorners=false
      bSeparateFillAlpha=false
      outlineMaterial=null
      outlineSize=0
      outlineColor=<rgba 0 0 0 1>
     }
    }
    shadowOffset=<0 0>
   }
   slot {class=VerticalBoxSlot, properties {horizontalAlignment=HAlign_Left, verticalAlignment=VAlign_Center, padding=<ltrb 0 0 0 6>}}
  }
  Text_V-Bucks_2 {
   widgetClassPath=&CommonTextBlock
   parent=VerticalBox_6
   widget=Text_V-Bucks_2
   properties {
    text=V-Bucks
    textTransformPolicy=ToUpper
    font {
     bForceMonospaced=false
     bMaterialIsStencil=false
     fontMaterial=null
     fontObject=&BurbankSmall
     letterSpacing=0
     monospacedWidth=1
     size=18
     skewAmount=0
     typefaceFontName=Black
     outlineSettings {
      bApplyOutlineToDropShadows=false
      bMiteredCorners=false
      bSeparateFillAlpha=false
      outlineMaterial=null
      outlineSize=0
      outlineColor=<rgba 0 0 0 1>
     }
    }
    shadowOffset=<0 0>
   }
   slot {class=VerticalBoxSlot, properties {horizontalAlignment=HAlign_Left, verticalAlignment=VAlign_Center, padding=<ltrb 0 7 0 0>}}
  }
  VB_Options {
   widgetClassPath=&VerticalBox
   parent=VerticalBox_1
   widget=VB_Options
   slot {class=VerticalBoxSlot, properties {size {sizeRule=Fill, value=0.75}}}
  }
  SizeBox_1 {
   widgetClassPath=&SizeBox
   parent=VB_Options
   widget=SizeBox_1
   properties {bOverride_MinDesiredWidth=true, minDesiredWidth=800}
   slot {class=VerticalBoxSlot, properties {horizontalAlignment=HAlign_Center, padding=<ltrb 30 30 30 0>, size {sizeRule=Automatic, value=0.15}}}
  }
  HB_SelectAnOption {widgetClassPath=&HorizontalBox, parent=SizeBox_1, widget=HB_SelectAnOption, slot {class=SizeBoxSlot}}
  IMG_LeftStroke_1 {
   widgetClassPath=&Image
   bIsVariable=true
   parent=HB_SelectAnOption
   widget=IMG_LeftStroke_1
   properties {brush {
     bIsDynamicallyLoaded=false
     drawAs=Box
     imageType=NoImage
     mirroring=NoMirror
     resourceName=None
     resourceObject=&/MI_UI_ReloadVBucks_Lines2_Legacy
     tiling=NoTile
     imageSize=<32 12>
     margin=<ltrb 0.115 0 0.11 0>
     outlineSettings {
      bUseBrushTransparency=false
      roundingType=HalfHeightRadius
      width=0
      color {colorUseRule=UseColor_Specified, specifiedColor=<rgba 0 0 0 0>}
      cornerRadii=<0 0 0 0>
     }
     tintColor {colorUseRule=UseColor_Specified, specifiedColor=<rgba 1 1 1 1>}
     uVRegion {bIsValid=false, max=<0 0>, min=<0 0>}
    }}
   slot {class=HorizontalBoxSlot, properties {verticalAlignment=VAlign_Center, size {sizeRule=Fill, value=0.5}}}
  }
  CommonTextBlock_628 {
   widgetClassPath=&CommonTextBlock
   parent=HB_SelectAnOption
   widget=CommonTextBlock_628
   properties {
    justification=Center
    text=Select an option
    textTransformPolicy=ToUpper
    font {
     bForceMonospaced=false
     bMaterialIsStencil=false
     fontMaterial=null
     fontObject=&Burbank
     letterSpacing=0
     monospacedWidth=1
     size=24
     skewAmount=0
     typefaceFontName=Burbank-PlayerPower
     outlineSettings {
      bApplyOutlineToDropShadows=false
      bMiteredCorners=false
      bSeparateFillAlpha=false
      outlineMaterial=null
      outlineSize=0
      outlineColor=<rgba 0 0 0 1>
     }
    }
    renderTransform {angle=0, scale=<1 0.9>, shear=<-12 0>, translation=<0 0>}
    shadowOffset=<0 0>
   }
   slot {class=HorizontalBoxSlot, properties {verticalAlignment=VAlign_Center, padding=<ltrb 20 0 20 0>}}
  }
  IMG_RightStroke {
   widgetClassPath=&Image
   bIsVariable=true
   parent=HB_SelectAnOption
   widget=IMG_RightStroke
   properties {brush {
     bIsDynamicallyLoaded=false
     drawAs=Box
     imageType=NoImage
     mirroring=NoMirror
     resourceName=None
     resourceObject=&MI_UI_ReloadVBucks_Lines2_Legacy
     tiling=NoTile
     imageSize=<32 12>
     margin=<ltrb 0.115 0 0.11 0>
     outlineSettings {
      bUseBrushTransparency=false
      roundingType=HalfHeightRadius
      width=0
      color {colorUseRule=UseColor_Specified, specifiedColor=<rgba 0 0 0 0>}
      cornerRadii=<0 0 0 0>
     }
     tintColor {colorUseRule=UseColor_Specified, specifiedColor=<rgba 1 1 1 1>}
     uVRegion {bIsValid=false, max=<0 0>, min=<0 0>}
    }}
   slot {class=HorizontalBoxSlot, properties {verticalAlignment=VAlign_Center, size {sizeRule=Fill, value=0.5}}}
  }
  VerticalBox_4 {
   widgetClassPath=&VerticalBox
   parent=VB_Options
   widget=VerticalBox_4
   slot {class=VerticalBoxSlot, properties {padding=<ltrb 30 24 30 0>, size {sizeRule=Fill, value=1}}}
  }
  HorizontalBox_4 {widgetClassPath=&HorizontalBox, parent=VerticalBox_4, widget=HorizontalBox_4, slot {class=VerticalBoxSlot}}
  Spacer_194 {
   widgetClassPath=&Spacer
   parent=HorizontalBox_4
   widget=Spacer_194
   slot {class=HorizontalBoxSlot, properties {size {sizeRule=Fill, value=0.1}}}
  }
  Overlay_4 {
   widgetClassPath=&Overlay
   parent=HorizontalBox_4
   widget=Overlay_4
   slot {class=HorizontalBoxSlot, properties {size {sizeRule=Fill, value=1}}}
  }
  Switcher_DynamicOptions {
   widgetClassPath=&CommonVisibilitySwitcher
   bInherited=true
   parent=Overlay_4
   widget=Switcher_DynamicOptions
   slot {class=OverlaySlot, properties {horizontalAlignment=HAlign_Center, verticalAlignment=VAlign_Center}}
  }
  Overlay_MtxRequired {
   widgetClassPath=&Overlay
   bInherited=true
   parent=Switcher_DynamicOptions
   widget=Overlay_MtxRequired
   slot {class=CommonVisibilitySwitcherSlot}
  }
  SB_MtxRequired {
   widgetClassPath=&SizeBox
   parent=Overlay_MtxRequired
   widget=SB_MtxRequired
   properties {bOverride_MaxDesiredWidth=true, maxAspectRatio=5, maxDesiredWidth=820, minDesiredWidth=850}
   slot {class=OverlaySlot}
  }
  Button_MtxRequiredOption {
   widgetClassPath=&/DynamicMtxOfferButton_Legacy
   bInherited=true
   bIsVariable=true
   parent=SB_MtxRequired
   widget=Button_MtxRequiredOption
   properties {
    onButtonBaseClicked=(&self:WidgetTree.Button_MtxRequiredOption.HandleOnButtonClicked)
    onButtonBaseDoubleClicked=(&self:WidgetTree.Button_MtxRequiredOption.HandleOnButtonDoubleClicked)
    onButtonBaseHovered=(&self:WidgetTree.Button_MtxRequiredOption.HandleOnButtonHovered)
    onButtonBaseUnhovered=(&self:WidgetTree.Button_MtxRequiredOption.HandleOnButtonUnhovered)
    onSelectedChangedBase=(&self:WidgetTree.Button_MtxRequiredOption.HandleOnSelectedChanged)
   }
   slot {class=SizeBoxSlot}
  }
  Overlay_StarterKit {
   widgetClassPath=&Overlay
   bInherited=true
   parent=Switcher_DynamicOptions
   widget=Overlay_StarterKit
   slot {class=CommonVisibilitySwitcherSlot}
  }
  SB_StarterKit {
   widgetClassPath=&SizeBox
   parent=Overlay_StarterKit
   widget=SB_StarterKit
   properties {bOverride_MaxDesiredWidth=true, maxDesiredWidth=820, minDesiredWidth=850}
   slot {class=OverlaySlot}
  }
  Button_StarterKitOption {
   widgetClassPath=&/DynamicMtxStarterKitButton_Legacy
   bInherited=true
   bIsVariable=true
   parent=SB_StarterKit
   widget=Button_StarterKitOption
   properties {
    onButtonBaseClicked=(&self:WidgetTree.Button_StarterKitOption.HandleOnButtonClicked)
    onButtonBaseDoubleClicked=(&self:WidgetTree.Button_StarterKitOption.HandleOnButtonDoubleClicked)
    onButtonBaseHovered=(&self:WidgetTree.Button_StarterKitOption.HandleOnButtonHovered)
    onButtonBaseUnhovered=(&self:WidgetTree.Button_StarterKitOption.HandleOnButtonUnhovered)
    onSelectedChangedBase=(&self:WidgetTree.Button_StarterKitOption.HandleOnSelectedChanged)
   }
   slot {class=SizeBoxSlot}
  }
  Spacer_84 {
   widgetClassPath=&Spacer
   parent=HorizontalBox_4
   widget=Spacer_84
   slot {class=HorizontalBoxSlot, properties {size {sizeRule=Fill, value=0.1}}}
  }
  SizeBox_0 {
   widgetClassPath=&SizeBox
   parent=VerticalBox_4
   widget=SizeBox_0
   properties {bOverride_MinDesiredWidth=true, maxAspectRatio=20, maxDesiredWidth=2790.3411, minDesiredWidth=400}
   slot {class=VerticalBoxSlot, properties {horizontalAlignment=HAlign_Center, padding=<ltrb 0 24 0 24>, size {sizeRule=Automatic, value=0}}}
  }
  HB_OR_Separator {widgetClassPath=&HorizontalBox, parent=SizeBox_0, widget=HB_OR_Separator, slot {class=SizeBoxSlot}}
  IMG_LeftStroke_2 {
   widgetClassPath=&Image
   bIsVariable=true
   parent=HB_OR_Separator
   widget=IMG_LeftStroke_2
   properties {brush {
     bIsDynamicallyLoaded=false
     drawAs=Box
     imageType=NoImage
     mirroring=NoMirror
     resourceName=None
     resourceObject=&MI_UI_ReloadVBucks_Lines2_Legacy
     tiling=NoTile
     imageSize=<32 12>
     margin=<ltrb 0.115 0 0.11 0>
     outlineSettings {
      bUseBrushTransparency=false
      roundingType=HalfHeightRadius
      width=0
      color {colorUseRule=UseColor_Specified, specifiedColor=<rgba 0 0 0 0>}
      cornerRadii=<0 0 0 0>
     }
     tintColor {colorUseRule=UseColor_Specified, specifiedColor=<rgba 1 1 1 1>}
     uVRegion {bIsValid=false, max=<0 0>, min=<0 0>}
    }}
   slot {class=HorizontalBoxSlot, properties {verticalAlignment=VAlign_Center, size {sizeRule=Fill, value=0.5}}}
  }
  CommonTextBlock {
   widgetClassPath=&CommonTextBlock
   parent=HB_OR_Separator
   widget=CommonTextBlock
   properties {
    justification=Center
    text=Or
    textTransformPolicy=ToUpper
    font {
     bForceMonospaced=false
     bMaterialIsStencil=false
     fontMaterial=null
     fontObject=&Burbank
     letterSpacing=0
     monospacedWidth=1
     size=24
     skewAmount=0
     typefaceFontName=Burbank-PlayerPower
     outlineSettings {
      bApplyOutlineToDropShadows=false
      bMiteredCorners=false
      bSeparateFillAlpha=false
      outlineMaterial=null
      outlineSize=0
      outlineColor=<rgba 0 0 0 1>
     }
    }
    renderTransform {angle=0, scale=<1 0.9>, shear=<-12 0>, translation=<0 0>}
    shadowOffset=<0 0>
   }
   slot {class=HorizontalBoxSlot, properties {verticalAlignment=VAlign_Center, padding=<ltrb 20 0 20 0>}}
  }
  IMG_RightStroke_1 {
   widgetClassPath=&Image
   bIsVariable=true
   parent=HB_OR_Separator
   widget=IMG_RightStroke_1
   properties {brush {
     bIsDynamicallyLoaded=false
     drawAs=Box
     imageType=NoImage
     mirroring=NoMirror
     resourceName=None
     resourceObject=&MI_UI_ReloadVBucks_Lines2_Legacy
     tiling=NoTile
     imageSize=<32 12>
     margin=<ltrb 0.115 0 0.11 0>
     outlineSettings {
      bUseBrushTransparency=false
      roundingType=HalfHeightRadius
      width=0
      color {colorUseRule=UseColor_Specified, specifiedColor=<rgba 0 0 0 0>}
      cornerRadii=<0 0 0 0>
     }
     tintColor {colorUseRule=UseColor_Specified, specifiedColor=<rgba 1 1 1 1>}
     uVRegion {bIsValid=false, max=<0 0>, min=<0 0>}
    }}
   slot {class=HorizontalBoxSlot, properties {verticalAlignment=VAlign_Center, size {sizeRule=Fill, value=0.5}}}
  }
  HorizontalBox_5 {widgetClassPath=&HorizontalBox, parent=VerticalBox_4, widget=HorizontalBox_5, slot {class=VerticalBoxSlot}}
  Spacer {
   widgetClassPath=&Spacer
   parent=HorizontalBox_5
   widget=Spacer
   slot {class=HorizontalBoxSlot, properties {size {sizeRule=Fill, value=0.1}}}
  }
  Overlay_0 {
   widgetClassPath=&Overlay
   parent=HorizontalBox_5
   widget=Overlay_0
   slot {class=HorizontalBoxSlot, properties {size {sizeRule=Fill, value=1}}}
  }
  SizeBox_3 {
   widgetClassPath=&SizeBox
   parent=Overlay_0
   widget=SizeBox_3
   properties {bOverride_MaxDesiredWidth=true, maxAspectRatio=5, maxDesiredWidth=850}
   slot {class=OverlaySlot, properties {horizontalAlignment=HAlign_Center, verticalAlignment=VAlign_Center}}
  }
  Button_MtxOtherOption {
   widgetClassPath=&/StaticMtxOfferButton_Legacy
   bInherited=true
   bIsVariable=true
   parent=SizeBox_3
   widget=Button_MtxOtherOption
   properties {
    onButtonBaseClicked=(&self:WidgetTree.Button_MtxOtherOption.HandleOnButtonClicked)
    onButtonBaseDoubleClicked=(&self:WidgetTree.Button_MtxOtherOption.HandleOnButtonDoubleClicked)
    onButtonBaseHovered=(&self:WidgetTree.Button_MtxOtherOption.HandleOnButtonHovered)
    onButtonBaseUnhovered=(&self:WidgetTree.Button_MtxOtherOption.HandleOnButtonUnhovered)
    onSelectedChangedBase=(&self:WidgetTree.Button_MtxOtherOption.HandleOnSelectedChanged)
   }
   slot {class=SizeBoxSlot}
  }
  Spacer_1 {
   widgetClassPath=&Spacer
   parent=HorizontalBox_5
   widget=Spacer_1
   slot {class=HorizontalBoxSlot, properties {size {sizeRule=Fill, value=0.1}}}
  }
  Button_MoreOptions {
   widgetClassPath=/FortUILibrary/Foundation/Buttons/SoloButton
   bInherited=true
   bIsVariable=true
   parent=VerticalBox_4
   widget=Button_MoreOptions
   properties {
    bIsFocusable=true
    button Description=More options
    clickMethod=DownAndUp
    navigation=&self:WidgetTree.Button_MoreOptions.WidgetNavigation_0
    onButtonBaseClicked=(&self:WidgetTree.Button_MoreOptions.HandleOnButtonClicked)
    onButtonBaseDoubleClicked=(&self:WidgetTree.Button_MoreOptions.HandleOnButtonDoubleClicked)
    onButtonBaseHovered=(&self:WidgetTree.Button_MoreOptions.HandleOnButtonHovered)
    onButtonBaseUnhovered=(&self:WidgetTree.Button_MoreOptions.HandleOnButtonUnhovered)
    onSelectedChangedBase=(&self:WidgetTree.Button_MoreOptions.HandleOnSelectedChanged)
    pressMethod=DownAndUp
    hoveredSlateSoundOverride {resourceObject=/Game/Sounds/Fort_UI_Sounds/FrontEnd/New_FrontEnd/Fortnite_UI_MainMenu_Tab_Hover_01_Cue}
    pressedSlateSoundOverride {resourceObject=/Game/Sounds/Fort_UI_Sounds/FrontEnd/New_FrontEnd/Fortnite_UI_MainMenu_Generic_01_Cue}
   }
   slot {class=VerticalBoxSlot, properties {horizontalAlignment=HAlign_Center, verticalAlignment=VAlign_Bottom, padding=<ltrb 0 40 0 0>}}
  }
  WBP_ContentPurchaseBlocker {
   widgetClassPath=/FortniteCoreUI/UI/ContentControls/WBP_ContentPurchaseBlocker
   bIsVariable=true
   parent=Overlay_3
   widget=WBP_ContentPurchaseBlocker
   properties {visibility=Collapsed}
   slot {class=OverlaySlot, properties {horizontalAlignment=HAlign_Center, verticalAlignment=VAlign_Center}}
  }
  SafeZone_3 {
   widgetClassPath=&SafeZone
   bIsVariable=true
   parent=HorizontalBox_2
   widget=SafeZone_3
   properties {padLeft=false}
   slot {class=HorizontalBoxSlot, properties {size {sizeRule=Fill, value=0.5}}}
  }
  OfferDetails_Container {widgetClassPath=&Overlay, parent=SafeZone_3, widget=OfferDetails_Container, slot {class=SafeZoneSlot}}
  HorizontalBox_6 {
   widgetClassPath=&HorizontalBox
   parent=OfferDetails_Container
   widget=HorizontalBox_6
   slot {class=OverlaySlot, properties {horizontalAlignment=HAlign_Fill, verticalAlignment=VAlign_Fill}}
  }
  FramingWidget {
   widgetClassPath=&SizeBox
   bIsVariable=true
   parent=HorizontalBox_6
   widget=FramingWidget
   slot {class=HorizontalBoxSlot, properties {padding=<ltrb -181 24 -361 24>, size {sizeRule=Fill, value=1}}}
  }
  Widget_CameraFrame {widgetClassPath=&Overlay, bInherited=true, parent=FramingWidget, widget=Widget_CameraFrame, slot {class=SizeBoxSlot}}
  Spacer_92 {widgetClassPath=&Spacer, parent=HorizontalBox_6, widget=Spacer_92, properties {size=<440 1>}, slot {class=HorizontalBoxSlot}}
  SB_OfferDetailsWidget {
   widgetClassPath=&SizeBox
   bIsVariable=true
   parent=OfferDetails_Container
   widget=SB_OfferDetailsWidget
   properties {maxDesiredWidth=102.5564, visibility=HitTestInvisible, widthOverride=500}
   slot {class=OverlaySlot, properties {horizontalAlignment=HAlign_Right, verticalAlignment=VAlign_Bottom, padding=<ltrb 0 0 40 70>}}
  }
  Overlay_1 {widgetClassPath=&Overlay, parent=SB_OfferDetailsWidget, widget=Overlay_1, slot {class=SizeBoxSlot}}
  SmallOfferPreview_Widget {
   widgetClassPath=/FortniteCoreUI/Frontend/AthenaSmallOfferPreview
   bInherited=true
   bIsVariable=true
   parent=Overlay_1
   widget=SmallOfferPreview_Widget
   slot {class=OverlaySlot, properties {horizontalAlignment=HAlign_Fill, verticalAlignment=VAlign_Fill}}
  }
  HorizontalBox_12 {
   widgetClassPath=&HorizontalBox
   parent=Overlay_2
   widget=HorizontalBox_12
   slot {class=OverlaySlot, properties {horizontalAlignment=HAlign_Right, padding=<ltrb 0 24 24 0>}}
  }
  MTXButtonBox {
   widgetClassPath=&SizeBox
   bIsVariable=true
   parent=HorizontalBox_12
   widget=MTXButtonBox
   slot {class=HorizontalBoxSlot, properties {horizontalAlignment=HAlign_Right}}
  }
  MTXButton {
   widgetClassPath=/Game/UI/Frontend/TopBar/MTXButton
   bIsVariable=true
   parent=MTXButtonBox
   widget=MTXButton
   properties {
    onButtonBaseClicked=(&self:WidgetTree.MTXButton.HandleOnButtonClicked)
    onButtonBaseDoubleClicked=(&self:WidgetTree.MTXButton.HandleOnButtonDoubleClicked)
    onButtonBaseHovered=(&self:WidgetTree.MTXButton.HandleOnButtonHovered)
    onButtonBaseUnhovered=(&self:WidgetTree.MTXButton.HandleOnButtonUnhovered)
    onSelectedChangedBase=(&self:WidgetTree.MTXButton.HandleOnSelectedChanged)
    visibility=HitTestInvisible
   }
   slot {class=SizeBoxSlot}
  }
  VisibilityBorder_Touch {
   widgetClassPath=&CommonHardwareVisibilityBorder
   parent=HorizontalBox_12
   widget=VisibilityBorder_Touch
   properties {visibilityQuery {
     autoDescription=" ANY( Input.Touch )"
     tokenStreamVersion=0
     userDescription=""
     queryTokenStream [0,1,1,1,0]
     tagDictionary [{tagName=Input.Touch}]
    }}
   slot {class=HorizontalBoxSlot, properties {padding=<ltrb 8 0 0 0>}}
  }
  VerticalBox_3 {
   widgetClassPath=&VerticalBox
   parent=VisibilityBorder_Touch
   widget=VerticalBox_3
   slot {class=BorderSlot, properties {padding=<ltrb 0 0 0 0>}}
  }
  ForcenonMobilePosition {
   widgetClassPath=&Spacer
   parent=VerticalBox_3
   widget=ForcenonMobilePosition
   properties {size=<48 0>}
   slot {class=VerticalBoxSlot, properties {horizontalAlignment=HAlign_Right, verticalAlignment=VAlign_Top}}
  }
  Button_CloseTouch {
   widgetClassPath=/FortUILibrary/Foundation/Buttons/CloseButton
   bInherited=true
   bIsVariable=true
   parent=VerticalBox_3
   widget=Button_CloseTouch
   properties {
    onButtonBaseClicked=(&self:WidgetTree.Button_CloseTouch.HandleOnButtonClicked)
    onButtonBaseDoubleClicked=(&self:WidgetTree.Button_CloseTouch.HandleOnButtonDoubleClicked)
    onButtonBaseHovered=(&self:WidgetTree.Button_CloseTouch.HandleOnButtonHovered)
    onButtonBaseUnhovered=(&self:WidgetTree.Button_CloseTouch.HandleOnButtonUnhovered)
    onSelectedChangedBase=(&self:WidgetTree.Button_CloseTouch.HandleOnSelectedChanged)
   }
   slot {class=VerticalBoxSlot, properties {padding=<ltrb 14 0 0 0>}}
  }
  Button_PreviewAction {widgetClassPath=&CommonButtonBase, bInherited=true, widget=null}
  Button_PreviewAction1 {widgetClassPath=&CommonButtonBase, bInherited=true, widget=null}
  Button_Camera {widgetClassPath=&CommonButtonBase, bInherited=true, widget=null}
  ItemDownloadIndicator {widgetClassPath=/Script/FortniteUI.FortItemDownloadIndicator, bInherited=true, widget=null}
  Text_DebugPreviewDetails {widgetClassPath=&CommonTextBlock, bInherited=true, widget=null}
 ]
}
