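# Translucent UI material for the legacy Reload V-Bucks logo in the item shop.
# It blends two colour parameters for tint and multiplies the sprite texture's alpha by a bottom fade mask.
# The mask is procedural, built from scaled, offset, inverted and clamped vertical UVs, with a scalar parameter blending between unmasked and fully masked.
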Material /ItemShop/UI/ReloadMtx/M_ReloadVBucks_Logo_Legacy
graph {
 properties {
  bUsedWithStaticMesh=false
  blendMode=BLEND_Translucent
  materialDomain=MD_UI
  pixelDepthOffsetMode=PDOM_Legacy
  usedShadingModels=Default Lit
 }
 nodes [
  LinearInterpolate(
   A=VectorParameter() {group=Color, parameterName=colour1, defaultValue=<rgba 1 1 1 1>}.RGB
   B=VectorParameter() {group=Color, parameterName=colour2, defaultValue=<rgba 1 1 1 0>}.RGB
  )
  n1 = Multiply(A=TextureCoordinate()) {constB=2.3}
  n2 = OneMinus(None=ComponentMask(None=Add(A=@n1) {constB=-1}) {g=true})
  Multiply(
   A=TextureSampleParameter2D() {parameterName=Sprite, samplerType=SAMPLERTYPE_LinearGrayscale, texture=&/ReloadVBucks_Logo_Legacy}.A
   B=LinearInterpolate(
    A=Constant() {r=1}
    B=Clamp(None=@n2)
    Alpha=ScalarParameter() {defaultValue=1, desc="0-1 (no mask - with)", group=Bottom Mask, parameterName=Mask bottom, sortPriority=2}
   )
  )
 ]
}
