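# Translucent UI material for the Item Shop's Reload V-Bucks UI (legacy square variant) that draws a procedural, skewed square frame with a gradient fill.
# Colour is the sum of two Color1-to-Color2 blends, one along a rotated U axis and one along a power-adjusted V axis.
# Outer and inner quads are each built from four individually rotated SmoothStep edges combined with Max.
# Subtracting inner from outer leaves the frame mask, scaled by Color1 alpha.
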Material /ItemShop/UI/ReloadMtx/M_UI_ReloadVBucks_Square_Legacy
graph {
 properties {
  bUsedWithStaticMesh=false
  blendMode=BLEND_Translucent
  materialDomain=MD_UI
  pixelDepthOffsetMode=PDOM_Legacy
  usedShadingModels=Default Lit
 }
 nodes [
  n1 = VectorParameter() {group=Colour, parameterName=Color1, defaultValue=<rgba 0 0.0692 0.5573 1>}
  n2 = VectorParameter() {group=Colour, parameterName=Color2, defaultValue=<rgba 0.0006 0.003 0.1042 1>}
  n3 = TextureCoordinate()
  n4 = MaterialFunctionCall(UVs=@n3, "Rotation Angle (0-1)"=ScalarParameter() {defaultValue=0.45, group=Colour, parameterName=Rotation, sliderMax=1}) {materialFunction=&CustomRotator}
  n5 = ComponentMask(None=@n3) {g=true}
  n6 = ScalarParameter() {defaultValue=0.5, group=Colour, parameterName=GradPow, sortPriority=20}
  Add(
   A=LinearInterpolate(A=@n1.RGB, B=@n2.RGB, Alpha=ComponentMask(None=@n4.Rotated Values) {r=true})
   B=LinearInterpolate(A=@n1.RGB, B=@n2.RGB, Alpha=Power(Base=@n5, Exp=@n6))
  )
  n7 = ScalarParameter() {defaultValue=0.0347, group=Outer, parameterName=LeftEdge}
  n8 = MaterialFunctionCall(
   UVs=TextureCoordinate()
   Rotation Center=AppendVector(A=@n7, B=Constant() {r=0.5})
   "Rotation Angle (0-1)"=ScalarParameter() {defaultValue=0.01, group=Outer, parameterName=LeftRotation, sliderMax=0.25, sliderMin=-0.25}
  ) {materialFunction=&CustomRotator}
  n9 = ScalarParameter() {defaultValue=0.005, group=Outer, parameterName=EdgeSoftness, sortPriority=1}
  n10 = OneMinus(None=MaterialFunctionCall(Alpha=ComponentMask(None=@n8.Rotated Values) {r=true}, Min=@n7, Max=Add(A=@n7, B=@n9)) {materialFunction=&SmoothStep}.Result)
  n11 = TextureCoordinate()
  n12 = OneMinus(None=ScalarParameter() {defaultValue=0.1, group=Outer, parameterName=RightEdge})
  n13 = AppendVector(A=@n12, B=Constant() {r=0.5})
  n14 = ScalarParameter() {defaultValue=-0.019, group=Outer, parameterName=RightRotation, sliderMax=0.25, sliderMin=-0.25}
  n15 = MaterialFunctionCall(
   Alpha=ComponentMask(None=MaterialFunctionCall(UVs=@n11, Rotation Center=@n13, "Rotation Angle (0-1)"=@n14) {materialFunction=&CustomRotator}.Rotated Values) {r=true}
   Min=@n12
   Max=Add(A=@n12, B=ScalarParameter() {defaultValue=0.005, group=Outer, parameterName=EdgeSoftness, sortPriority=1})
  ) {materialFunction=&SmoothStep}
  n16 = OneMinus(None=ScalarParameter() {defaultValue=0.185, group=Outer, parameterName=BottomEdge})
  n17 = Constant() {r=0.5}
  n18 = ComponentMask(None=MaterialFunctionCall(
    UVs=TextureCoordinate()
    Rotation Center=AppendVector(A=@n16, B=@n17)
    "Rotation Angle (0-1)"=ScalarParameter() {defaultValue=0.019, group=Outer, parameterName=BottomRotation, sliderMax=0.25, sliderMin=-0.25}
   ) {materialFunction=&CustomRotator}.Rotated Values) {g=true}
  n19 = Add(A=@n16, B=ScalarParameter() {defaultValue=0.005, group=Outer, parameterName=EdgeSoftness, sortPriority=1})
  n20 = ScalarParameter() {defaultValue=0.128, group=Outer, parameterName=TopEdge}
  n21 = Constant() {r=0.5}
  n22 = ComponentMask(None=MaterialFunctionCall(
    UVs=TextureCoordinate()
    Rotation Center=AppendVector(A=@n20, B=@n21)
    "Rotation Angle (0-1)"=ScalarParameter() {defaultValue=-0.023, group=Outer, parameterName=TopRotation, sliderMax=0.25, sliderMin=-0.25}
   ) {materialFunction=&CustomRotator}.Rotated Values) {g=true}
  n23 = Add(A=@n20, B=ScalarParameter() {defaultValue=0.005, group=Outer, parameterName=EdgeSoftness, sortPriority=1})
  n24 = Max(
   A=Max(
    A=Max(A=@n10, B=@n15.Result)
    B=MaterialFunctionCall(Alpha=@n18, Min=@n16, Max=@n19) {materialFunction=&SmoothStep}.Result
   )
   B=OneMinus(None=MaterialFunctionCall(Alpha=@n22, Min=@n20, Max=@n23) {materialFunction=&SmoothStep}.Result)
  )
  n25 = ScalarParameter() {defaultValue=0.14, group=Inner, parameterName=LeftEdge_1}
  n26 = MaterialFunctionCall(
   UVs=TextureCoordinate()
   Rotation Center=AppendVector(A=@n25, B=Constant() {r=0.5})
   "Rotation Angle (0-1)"=ScalarParameter() {defaultValue=0.014, group=Inner, parameterName=LeftRotation_1, sliderMax=0.25, sliderMin=-0.25}
  ) {materialFunction=&CustomRotator}
  n27 = ScalarParameter() {defaultValue=0.005, group=Inner, parameterName=EdgeSoftness}
  n28 = OneMinus(None=MaterialFunctionCall(Alpha=ComponentMask(None=@n26.Rotated Values) {r=true}, Min=@n25, Max=Add(A=@n25, B=@n27)) {materialFunction=&SmoothStep}.Result)
  n29 = TextureCoordinate()
  n30 = OneMinus(None=ScalarParameter() {defaultValue=0.215, group=Inner, parameterName=RightEdge_1})
  n31 = AppendVector(A=@n30, B=Constant() {r=0.5})
  n32 = ScalarParameter() {defaultValue=-0.0047, group=Inner, parameterName=RightRotation_1, sliderMax=0.25, sliderMin=-0.25}
  n33 = MaterialFunctionCall(
   Alpha=ComponentMask(None=MaterialFunctionCall(UVs=@n29, Rotation Center=@n31, "Rotation Angle (0-1)"=@n32) {materialFunction=&CustomRotator}.Rotated Values) {r=true}
   Min=@n30
   Max=Add(A=@n30, B=ScalarParameter() {defaultValue=0.005, group=Inner, parameterName=EdgeSoftness})
  ) {materialFunction=&SmoothStep}
  n34 = OneMinus(None=ScalarParameter() {defaultValue=0.23, group=Inner, parameterName=BottomEdge_1})
  n35 = Constant() {r=0.5}
  n36 = ComponentMask(None=MaterialFunctionCall(
    UVs=TextureCoordinate()
    Rotation Center=AppendVector(A=@n34, B=@n35)
    "Rotation Angle (0-1)"=ScalarParameter() {defaultValue=0.0095, group=Inner, parameterName=BottomRotation_1, sliderMax=0.25, sliderMin=-0.25}
   ) {materialFunction=&CustomRotator}.Rotated Values) {g=true}
  n37 = Add(A=@n34, B=ScalarParameter() {defaultValue=0.005, group=Inner, parameterName=EdgeSoftness})
  n38 = ScalarParameter() {defaultValue=0.253, group=Inner, parameterName=TopEdge_1}
  n39 = Constant() {r=0.5}
  n40 = ComponentMask(None=MaterialFunctionCall(
    UVs=TextureCoordinate()
    Rotation Center=AppendVector(A=@n38, B=@n39)
    "Rotation Angle (0-1)"=ScalarParameter() {defaultValue=-0.0095, group=Inner, parameterName=TopRotation_1, sliderMax=0.25, sliderMin=-0.25}
   ) {materialFunction=&CustomRotator}.Rotated Values) {g=true}
  n41 = Add(A=@n38, B=ScalarParameter() {defaultValue=0.005, group=Inner, parameterName=EdgeSoftness})
  n42 = Max(
   A=Max(
    A=Max(A=@n28, B=@n33.Result)
    B=MaterialFunctionCall(Alpha=@n36, Min=@n34, Max=@n37) {materialFunction=&SmoothStep}.Result
   )
   B=OneMinus(None=MaterialFunctionCall(Alpha=@n40, Min=@n38, Max=@n41) {materialFunction=&SmoothStep}.Result)
  )
  Multiply(A=@n1.A, B=Subtract(A=OneMinus(None=@n24), B=OneMinus(None=@n42)))
 ]
}
