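# Legacy V-Bucks offer button for the item shop's reload list, showing currency art, amount, bonus, real-money price and bonus-percent tag.
# Layout fills bound widgets of AthenaMtxOfferListEntryButtonLegacy; the graph sets no offer data.
# Hover plays OnHover forward, and unhover reverses it faster, whitening the background, darkening text, scaling currency art and revealing the input-action icon.
# A visibility wrapper limits the PlayStation Store logo by platform.
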WidgetBlueprint /ItemShop/UI/ReloadMtx/MtxOfferListEntryButton_Legacy
animations [OnHover {
  binding_count=12
  length_seconds=0.2025
  object=&self:OnHover
  sequence {bindings [
    LoadGuardCurrencySpinner {
     id=#1
     possessed_class=CommonLoadGuard
     tracks [{
       type=MovieScene2DTransformTrack
       sections [{
         type=MovieScene2DTransformSection
         end_frame=4
         start_frame=0
         channels [
          Translation.X {type=MovieSceneScriptingFloatChannel}
          Translation.Y {type=MovieSceneScriptingFloatChannel}
          Angle {type=MovieSceneScriptingFloatChannel}
          Scale.X {type=MovieSceneScriptingFloatChannel, keys [[0,1],[4,1.2]]}
          Scale.Y {type=MovieSceneScriptingFloatChannel, keys [[0,1],[4,1.2]]}
          Shear.X {type=MovieSceneScriptingFloatChannel}
          Shear.Y {type=MovieSceneScriptingFloatChannel}
         ]
        }]
      }]
    }
    SizeBox_2 {
     id=#2
     possessed_class=SizeBox
     tracks [
      {
       type=MovieSceneFloatTrack
       display_name=Render Opacity
       sections [{type=MovieSceneFloatSection, end_frame=4, start_frame=0, channels [None {type=MovieSceneScriptingFloatChannel, keys [[0,0],[4,1]]}]}]
      }
      {
       type=MovieScene2DTransformTrack
       sections [{
         type=MovieScene2DTransformSection
         end_frame=4
         start_frame=0
         channels [
          Translation.X {type=MovieSceneScriptingFloatChannel, const [0,2]}
          Translation.Y {type=MovieSceneScriptingFloatChannel, const [14,2]}
          Angle {type=MovieSceneScriptingFloatChannel}
          Scale.X {type=MovieSceneScriptingFloatChannel, keys [[0,0.75],[4,1]]}
          Scale.Y {type=MovieSceneScriptingFloatChannel, keys [[0,0.75],[4,1]]}
          Shear.X {type=MovieSceneScriptingFloatChannel}
          Shear.Y {type=MovieSceneScriptingFloatChannel}
         ]
        }]
      }
     ]
    }
    IMG_ButtonBackground {
     id=#3
     possessed_class=Image
     tracks [{
       type=MovieSceneColorTrack
       display_name=Color and Opacity
       sections [{
         type=MovieSceneColorSection
         end_frame=4
         start_frame=0
         channels [
          Color.R {type=MovieSceneScriptingFloatChannel, keys [[0,0],[4,1]]}
          Color.G {type=MovieSceneScriptingFloatChannel, keys [[0,0.3613],[4,1]]}
          Color.B {type=MovieSceneScriptingFloatChannel, const [1,2]}
          Color.A {type=MovieSceneScriptingFloatChannel, const [1,2]}
         ]
        }]
      }]
    }
    Text_VBucks {
     id=#4
     possessed_class=CommonTextBlock
     tracks [{
       type=MovieSceneColorTrack
       display_name=Color and Opacity
       sections [{
         type=MovieSceneColorSection
         end_frame=4
         start_frame=0
         channels [
          Color.R {type=MovieSceneScriptingFloatChannel, keys [[0,1],[4,0]]}
          Color.G {type=MovieSceneScriptingFloatChannel, keys [[0,1],[4,0]]}
          Color.B {type=MovieSceneScriptingFloatChannel, keys [[0,1],[4,0]]}
          Color.A {type=MovieSceneScriptingFloatChannel, const [1,2]}
         ]
        }]
      }]
    }
    CommonTextBlock_7 {
     id=#5
     possessed_class=CommonTextBlock
     tracks [{
       type=MovieSceneColorTrack
       display_name=Color and Opacity
       sections [{
         type=MovieSceneColorSection
         end_frame=4
         start_frame=0
         channels [
          Color.R {type=MovieSceneScriptingFloatChannel, keys [[0,1],[4,0]]}
          Color.G {type=MovieSceneScriptingFloatChannel, keys [[0,1],[4,0]]}
          Color.B {type=MovieSceneScriptingFloatChannel, keys [[0,1],[4,0]]}
          Color.A {type=MovieSceneScriptingFloatChannel, const [1,2]}
         ]
        }]
      }]
    }
    Text_Bonus {
     id=#6
     possessed_class=CommonTextBlock
     tracks [{
       type=MovieSceneColorTrack
       display_name=Color and Opacity
       sections [{
         type=MovieSceneColorSection
         end_frame=4
         start_frame=0
         channels [
          Color.R {type=MovieSceneScriptingFloatChannel, keys [[0,1],[4,0]]}
          Color.G {type=MovieSceneScriptingFloatChannel, keys [[0,1],[4,0]]}
          Color.B {type=MovieSceneScriptingFloatChannel, keys [[0,1],[4,0]]}
          Color.A {type=MovieSceneScriptingFloatChannel, const [1,2]}
         ]
        }]
      }]
    }
    SizeBoxSlot_0 {
     id=#7
     parent_id=#8
     possessed_class=SizeBoxSlot
     tracks [{
       type=MovieSceneMarginTrack
       display_name=Padding
       sections [{
         type=MovieSceneMarginSection
         end_frame=4
         start_frame=0
         channels [
          Left {type=MovieSceneScriptingFloatChannel, const [0,2]}
          Top {type=MovieSceneScriptingFloatChannel, keys [[0,0],[4,-12]]}
          Right {type=MovieSceneScriptingFloatChannel, const [0,2]}
          Bottom {type=MovieSceneScriptingFloatChannel, keys [[0,0],[4,-12]]}
         ]
        }]
      }]
    }
    IMGCurrency_XL {
     id=#9
     possessed_class=Image
     tracks [{
       type=MovieScene2DTransformTrack
       sections [{
         type=MovieScene2DTransformSection
         end_frame=4
         start_frame=0
         channels [
          Translation.X {type=MovieSceneScriptingFloatChannel, keys [[0,0],[4,-20]]}
          Translation.Y {type=MovieSceneScriptingFloatChannel}
          Angle {type=MovieSceneScriptingFloatChannel}
          Scale.X {type=MovieSceneScriptingFloatChannel, keys [[0,1],[4,0.85]]}
          Scale.Y {type=MovieSceneScriptingFloatChannel, keys [[0,1],[4,0.85]]}
          Shear.X {type=MovieSceneScriptingFloatChannel}
          Shear.Y {type=MovieSceneScriptingFloatChannel}
         ]
        }]
      }]
    }
    Text_BonusMtxAmount {
     id=#10
     possessed_class=CommonTextBlock
     tracks [{
       type=MovieSceneColorTrack
       display_name=Color and Opacity
       sections [{
         type=MovieSceneColorSection
         end_frame=4
         start_frame=0
         channels [
          Color.R {type=MovieSceneScriptingFloatChannel, keys [[0,1],[4,0]]}
          Color.G {type=MovieSceneScriptingFloatChannel, keys [[0,1],[4,0]]}
          Color.B {type=MovieSceneScriptingFloatChannel, keys [[0,1],[4,0]]}
          Color.A {type=MovieSceneScriptingFloatChannel, const [1,2]}
         ]
        }]
      }]
    }
    Overlay_36 {id=#8, possessed_class=Overlay, tracks []}
    CommonVisibilitySwitcherSlot_1 {
     id=#11
     parent_id=#9
     possessed_class=CommonVisibilitySwitcherSlot
     tracks [{
       type=MovieSceneMarginTrack
       display_name=Padding
       sections [{
         type=MovieSceneMarginSection
         end_frame=4
         start_frame=0
         channels [
          Left {type=MovieSceneScriptingFloatChannel, keys [[0,-150],[4,-130]]}
          Top {type=MovieSceneScriptingFloatChannel}
          Right {type=MovieSceneScriptingFloatChannel}
          Bottom {type=MovieSceneScriptingFloatChannel}
         ]
        }]
      }]
    }
    Text_MtxAmount {
     id=#12
     possessed_class=CommonTextBlock
     tracks [{
       type=MovieSceneColorTrack
       display_name=Color and Opacity
       sections [{
         type=MovieSceneColorSection
         end_frame=4
         start_frame=0
         channels [
          Color.R {type=MovieSceneScriptingFloatChannel, keys [[0,1],[4,0]]}
          Color.G {type=MovieSceneScriptingFloatChannel, keys [[0,1],[4,0]]}
          Color.B {type=MovieSceneScriptingFloatChannel, keys [[0,1],[4,0]]}
          Color.A {type=MovieSceneScriptingFloatChannel, const [1,2]}
         ]
        }]
      }]
    }
   ]}
 }]
graphs [EventGraph {nodes [
   Event On Hovered()
   n1 = Get OnHover() -> (OnHover:Widget Animation Object Reference)
   > PlayAnimationForward(InAnimation=@n1, PlaybackSpeed=1.5)
   Event On Unhovered()
   > PlayAnimationReverse(InAnimation=@n1, PlaybackSpeed=2)
  ]}]
variables []
widgets {
 info {
  inheritedWidgetCount=10
  parentClass=/Script/FortniteUI.AthenaMtxOfferListEntryButtonLegacy
  rootWidgetClass=&Overlay
  widgetCount=42
 }
 widgets [
  Overlay_All {widgetClassPath=&Overlay, widget=Overlay_All}
  ButtonSize {
   widgetClassPath=&SizeBox
   parent=Overlay_All
   widget=ButtonSize
   properties {
    bOverride_HeightOverride=true
    bOverride_MaxDesiredWidth=true
    bOverride_MinDesiredWidth=true
    heightOverride=110
    maxDesiredHeight=118
    maxDesiredWidth=850
    minDesiredHeight=110
    minDesiredWidth=630
    widthOverride=630
   }
   slot {class=OverlaySlot, properties {horizontalAlignment=HAlign_Fill, verticalAlignment=VAlign_Fill}}
  }
  Overlay_36 {widgetClassPath=&Overlay, parent=ButtonSize, widget=Overlay_36, slot {class=SizeBoxSlot}}
  IMG_ButtonBackground {
   widgetClassPath=&Image
   bIsVariable=true
   parent=Overlay_36
   widget=IMG_ButtonBackground
   properties {
    brush {
     bIsDynamicallyLoaded=false
     drawAs=Box
     imageType=NoImage
     mirroring=NoMirror
     resourceName=None
     resourceObject=&/MI_UI_ReloadVBucks_OfferButton_Legacy
     tiling=NoTile
     imageSize=<32 32>
     margin=<ltrb 0 0 0 0>
     outlineSettings {
      bUseBrushTransparency=false
      roundingType=HalfHeightRadius
      width=0
      color {colorUseRule=UseColor_Specified, specifiedColor=<rgba 0 0 0 0>}
      cornerRadii=<0 0 0 0>
     }
     tintColor {colorUseRule=UseColor_Specified, specifiedColor=<rgba 1 1 1 1>}
     uVRegion {bIsValid=false, max=<0 0>, min=<0 0>}
    }
    colorAndOpacity=<rgba 0 0.3613 1 1>
   }
   slot {class=OverlaySlot, properties {horizontalAlignment=HAlign_Fill, verticalAlignment=VAlign_Fill}}
  }
  Overlay_0 {
   widgetClassPath=&Overlay
   parent=Overlay_36
   widget=Overlay_0
   slot {class=OverlaySlot, properties {horizontalAlignment=HAlign_Fill, verticalAlignment=VAlign_Fill}}
  }
  HorizontalBox_1 {
   widgetClassPath=&HorizontalBox
   parent=Overlay_0
   widget=HorizontalBox_1
   slot {class=OverlaySlot, properties {horizontalAlignment=HAlign_Fill, verticalAlignment=VAlign_Fill}}
  }
  LoadGuardCurrencySpinner {
   widgetClassPath=/Script/CommonUI.CommonLoadGuard
   bInherited=true
   bIsVariable=true
   parent=HorizontalBox_1
   widget=LoadGuardCurrencySpinner
   properties {loadingText="", textStyle=&TextStyle-Base_C, visibility=HitTestInvisible}
   slot {class=HorizontalBoxSlot, properties {horizontalAlignment=HAlign_Left, size {sizeRule=Fill, value=0.3}}}
  }
  ScaleBox_0 {
   widgetClassPath=&ScaleBox
   parent=LoadGuardCurrencySpinner
   widget=ScaleBox_0
   properties {stretchDirection=DownOnly}
   slot {class=LoadGuardSlot}
  }
  OptionalSwitcher_CurrencyIcons {
   widgetClassPath=&CommonVisibilitySwitcher
   bInherited=true
   parent=ScaleBox_0
   widget=OptionalSwitcher_CurrencyIcons
   slot {class=ScaleBoxSlot}
  }
  IMGCurrency_XS {
   widgetClassPath=&Image
   bIsVariable=true
   parent=OptionalSwitcher_CurrencyIcons
   widget=IMGCurrency_XS
   properties {brush {
     bIsDynamicallyLoaded=false
     drawAs=Image
     imageType=NoImage
     mirroring=NoMirror
     resourceName=None
     resourceObject=/Game/UI/Foundation/Textures/Founders/T_VBucks_Tile_1000_New
     tiling=NoTile
     imageSize=<512 512>
     margin=<ltrb 0 0 0 0>
     outlineSettings {
      bUseBrushTransparency=false
      roundingType=HalfHeightRadius
      width=0
      color {colorUseRule=UseColor_Specified, specifiedColor=<rgba 0 0 0 0>}
      cornerRadii=<0 0 0 0>
     }
     tintColor {colorUseRule=UseColor_Specified, specifiedColor=<rgba 1 1 1 1>}
     uVRegion {bIsValid=false, max=<0 0>, min=<0 0>}
    }}
   slot {class=CommonVisibilitySwitcherSlot}
  }
  IMGCurrency_S {
   widgetClassPath=&Image
   bIsVariable=true
   parent=OptionalSwitcher_CurrencyIcons
   widget=IMGCurrency_S
   properties {brush {
     bIsDynamicallyLoaded=false
     drawAs=Image
     imageType=NoImage
     mirroring=NoMirror
     resourceName=None
     resourceObject=/Game/UI/Foundation/Textures/Founders/T_VBucks_Tile_2800_New
     tiling=NoTile
     imageSize=<520 520>
     margin=<ltrb 0 0 0 0>
     outlineSettings {
      bUseBrushTransparency=false
      roundingType=HalfHeightRadius
      width=0
      color {colorUseRule=UseColor_Specified, specifiedColor=<rgba 0 0 0 0>}
      cornerRadii=<0 0 0 0>
     }
     tintColor {colorUseRule=UseColor_Specified, specifiedColor=<rgba 1 1 1 1>}
     uVRegion {bIsValid=false, max=<0 0>, min=<0 0>}
    }}
   slot {class=CommonVisibilitySwitcherSlot}
  }
  IMGCurrency_M {
   widgetClassPath=&Image
   bIsVariable=true
   parent=OptionalSwitcher_CurrencyIcons
   widget=IMGCurrency_M
   properties {brush {
     bIsDynamicallyLoaded=false
     drawAs=Image
     imageType=NoImage
     mirroring=NoMirror
     resourceName=None
     resourceObject=/Game/UI/Foundation/Textures/Founders/T-VBucks-Tile-5000
     tiling=NoTile
     imageSize=<512 512>
     margin=<ltrb 0 0 0 0>
     outlineSettings {
      bUseBrushTransparency=false
      roundingType=HalfHeightRadius
      width=0
      color {colorUseRule=UseColor_Specified, specifiedColor=<rgba 0 0 0 0>}
      cornerRadii=<0 0 0 0>
     }
     tintColor {colorUseRule=UseColor_Specified, specifiedColor=<rgba 1 1 1 1>}
     uVRegion {bIsValid=false, max=<0 0>, min=<0 0>}
    }}
   slot {class=CommonVisibilitySwitcherSlot, properties {padding=<ltrb -20 -20 -20 -20>}}
  }
  IMGCurrency_L {
   widgetClassPath=&Image
   bIsVariable=true
   parent=OptionalSwitcher_CurrencyIcons
   widget=IMGCurrency_L
   properties {brush {
     bIsDynamicallyLoaded=false
     drawAs=Image
     imageType=NoImage
     mirroring=NoMirror
     resourceName=None
     resourceObject=/Game/UI/Foundation/Textures/Founders/T-VBuckpile-7800_Tile
     tiling=NoTile
     imageSize=<512 512>
     margin=<ltrb 0 0 0 0>
     outlineSettings {
      bUseBrushTransparency=false
      roundingType=HalfHeightRadius
      width=0
      color {colorUseRule=UseColor_Specified, specifiedColor=<rgba 0 0 0 0>}
      cornerRadii=<0 0 0 0>
     }
     tintColor {colorUseRule=UseColor_Specified, specifiedColor=<rgba 1 1 1 1>}
     uVRegion {bIsValid=false, max=<0 0>, min=<0 0>}
    }}
   slot {class=CommonVisibilitySwitcherSlot, properties {padding=<ltrb -30 -30 -30 -30>}}
  }
  IMGCurrency_XL {
   widgetClassPath=&Image
   bIsVariable=true
   parent=OptionalSwitcher_CurrencyIcons
   widget=IMGCurrency_XL
   properties {brush {
     bIsDynamicallyLoaded=false
     drawAs=Image
     imageType=NoImage
     mirroring=NoMirror
     resourceName=None
     resourceObject=/Game/UI/Foundation/Textures/Founders/T_VBucks_Tile_13500_New
     tiling=NoTile
     imageSize=<512 512>
     margin=<ltrb 0 0 0 0>
     outlineSettings {
      bUseBrushTransparency=false
      roundingType=HalfHeightRadius
      width=0
      color {colorUseRule=UseColor_Specified, specifiedColor=<rgba 0 0 0 0>}
      cornerRadii=<0 0 0 0>
     }
     tintColor {colorUseRule=UseColor_Specified, specifiedColor=<rgba 1 1 1 1>}
     uVRegion {bIsValid=false, max=<0 0>, min=<0 0>}
    }}
   slot {class=CommonVisibilitySwitcherSlot, properties {padding=<ltrb -150 -110 -80 -110>}}
  }
  Overlay_33 {
   widgetClassPath=&Overlay
   parent=HorizontalBox_1
   widget=Overlay_33
   slot {class=HorizontalBoxSlot, properties {size {sizeRule=Fill, value=0.8}}}
  }
  ScaleBox_2 {
   widgetClassPath=&ScaleBox
   parent=Overlay_33
   widget=ScaleBox_2
   properties {stretchDirection=DownOnly}
   slot {class=OverlaySlot, properties {horizontalAlignment=HAlign_Fill, verticalAlignment=VAlign_Fill, padding=<ltrb 0 13 0 13>}}
  }
  HorizontalBox_3 {
   widgetClassPath=&HorizontalBox
   parent=ScaleBox_2
   widget=HorizontalBox_3
   slot {class=ScaleBoxSlot, properties {verticalAlignment=VAlign_Fill}}
  }
  VB_MtxTexts {
   widgetClassPath=&VerticalBox
   parent=HorizontalBox_3
   widget=VB_MtxTexts
   slot {class=HorizontalBoxSlot, properties {verticalAlignment=VAlign_Center}}
  }
  Text_MtxAmount {
   widgetClassPath=&CommonTextBlock
   bInherited=true
   bIsVariable=true
   parent=VB_MtxTexts
   widget=Text_MtxAmount
   properties {
    justification=Center
    text="2,800"
    font {
     bForceMonospaced=false
     bMaterialIsStencil=false
     fontMaterial=null
     fontObject=&BurbankSmall
     letterSpacing=0
     monospacedWidth=1
     size=28
     skewAmount=0
     typefaceFontName=Black
     outlineSettings {
      bApplyOutlineToDropShadows=false
      bMiteredCorners=false
      bSeparateFillAlpha=false
      outlineMaterial=null
      outlineSize=0
      outlineColor=<rgba 0 0 0 1>
     }
    }
    renderTransform {angle=0, scale=<1 1>, shear=<-8 0>, translation=<0 0>}
    shadowOffset=<0 0>
   }
   slot {class=VerticalBoxSlot, properties {verticalAlignment=VAlign_Bottom, padding=<ltrb 0 4 0 0>}}
  }
  Text_VBucks {
   widgetClassPath=&CommonTextBlock
   parent=VB_MtxTexts
   widget=Text_VBucks
   properties {
    justification=Center
    text=V-Bucks
    textTransformPolicy=ToUpper
    font {
     bForceMonospaced=false
     bMaterialIsStencil=false
     fontMaterial=null
     fontObject=&BurbankSmall
     letterSpacing=0
     monospacedWidth=1
     size=20
     skewAmount=0
     typefaceFontName=Black
     outlineSettings {
      bApplyOutlineToDropShadows=false
      bMiteredCorners=false
      bSeparateFillAlpha=false
      outlineMaterial=null
      outlineSize=0
      outlineColor=<rgba 0 0 0 1>
     }
    }
    renderTransform {angle=0, scale=<1 1>, shear=<-8 0>, translation=<0 -4>}
    shadowOffset=<0 0>
   }
   slot {class=VerticalBoxSlot, properties {verticalAlignment=VAlign_Top}}
  }
  HB_BonusMtxArea {
   widgetClassPath=&HorizontalBox
   bInherited=true
   bIsVariable=true
   parent=HorizontalBox_3
   widget=HB_BonusMtxArea
   slot {class=HorizontalBoxSlot, properties {verticalAlignment=VAlign_Center}}
  }
  CommonTextBlock_7 {
   widgetClassPath=&CommonTextBlock
   parent=HB_BonusMtxArea
   widget=CommonTextBlock_7
   properties {
    text=+
    font {
     bForceMonospaced=false
     bMaterialIsStencil=false
     fontMaterial=null
     fontObject=&BurbankSmall
     letterSpacing=0
     monospacedWidth=1
     size=40
     skewAmount=0
     typefaceFontName=Bold
     outlineSettings {
      bApplyOutlineToDropShadows=false
      bMiteredCorners=false
      bSeparateFillAlpha=false
      outlineMaterial=null
      outlineSize=0
      outlineColor=<rgba 0 0 0 1>
     }
    }
    shadowOffset=<0 0>
   }
   slot {class=HorizontalBoxSlot, properties {verticalAlignment=VAlign_Center, padding=<ltrb 14 0 14 0>}}
  }
  VB_BonusTexts {
   widgetClassPath=&VerticalBox
   parent=HB_BonusMtxArea
   widget=VB_BonusTexts
   slot {class=HorizontalBoxSlot, properties {horizontalAlignment=HAlign_Left, size {sizeRule=Fill, value=1}}}
  }
  Text_BonusMtxAmount {
   widgetClassPath=&CommonTextBlock
   bInherited=true
   bIsVariable=true
   parent=VB_BonusTexts
   widget=Text_BonusMtxAmount
   properties {
    justification=Center
    text=0000
    font {
     bForceMonospaced=false
     bMaterialIsStencil=false
     fontMaterial=null
     fontObject=&BurbankSmall
     letterSpacing=0
     monospacedWidth=1
     size=28
     skewAmount=0
     typefaceFontName=Black
     outlineSettings {
      bApplyOutlineToDropShadows=false
      bMiteredCorners=false
      bSeparateFillAlpha=false
      outlineMaterial=null
      outlineSize=0
      outlineColor=<rgba 0 0 0 1>
     }
    }
    renderTransform {angle=0, scale=<1 1>, shear=<-8 0>, translation=<0 0>}
    shadowOffset=<0 0>
   }
   slot {class=VerticalBoxSlot, properties {verticalAlignment=VAlign_Bottom, padding=<ltrb 0 4 0 0>}}
  }
  Text_Bonus {
   widgetClassPath=&CommonTextBlock
   parent=VB_BonusTexts
   widget=Text_Bonus
   properties {
    justification=Center
    text=Bonus
    textTransformPolicy=ToUpper
    font {
     bForceMonospaced=false
     bMaterialIsStencil=false
     fontMaterial=null
     fontObject=&BurbankSmall
     letterSpacing=0
     monospacedWidth=1
     size=20
     skewAmount=0
     typefaceFontName=Black
     outlineSettings {
      bApplyOutlineToDropShadows=false
      bMiteredCorners=false
      bSeparateFillAlpha=false
      outlineMaterial=null
      outlineSize=0
      outlineColor=<rgba 0 0 0 1>
     }
    }
    renderTransform {angle=0, scale=<1 1>, shear=<-8 0>, translation=<0 -4>}
    shadowOffset=<0 0>
   }
   slot {class=VerticalBoxSlot, properties {verticalAlignment=VAlign_Top}}
  }
  Spacer_118 {
   widgetClassPath=&Spacer
   parent=HorizontalBox_1
   widget=Spacer_118
   slot {class=HorizontalBoxSlot, properties {size {sizeRule=Fill, value=0.1}}}
  }
  Overlay {
   widgetClassPath=&Overlay
   parent=HorizontalBox_1
   widget=Overlay
   slot {class=HorizontalBoxSlot, properties {size {sizeRule=Fill, value=0.36}}}
  }
  IMG_PriceBackground {
   widgetClassPath=&Image
   bIsVariable=true
   parent=Overlay
   widget=IMG_PriceBackground
   properties {
    brush {
     bIsDynamicallyLoaded=false
     drawAs=Image
     imageType=NoImage
     mirroring=NoMirror
     resourceName=None
     resourceObject=&/MI_UI_ReloadVBucks_OfferContainer_Legacy
     tiling=NoTile
     imageSize=<32 32>
     margin=<ltrb 0 0 0 0>
     outlineSettings {
      bUseBrushTransparency=false
      roundingType=HalfHeightRadius
      width=0
      color {colorUseRule=UseColor_Specified, specifiedColor=<rgba 0 0 0 0>}
      cornerRadii=<0 0 0 0>
     }
     tintColor {colorUseRule=UseColor_Specified, specifiedColor=<rgba 1 1 1 1>}
     uVRegion {bIsValid=false, max=<0 0>, min=<0 0>}
    }
    colorAndOpacity=<rgba 1 0.9734 0 1>
   }
   slot {class=OverlaySlot, properties {horizontalAlignment=HAlign_Fill, verticalAlignment=VAlign_Fill}}
  }
  HorizontalBox_2 {
   widgetClassPath=&HorizontalBox
   parent=Overlay
   widget=HorizontalBox_2
   slot {class=OverlaySlot, properties {horizontalAlignment=HAlign_Center, verticalAlignment=VAlign_Center, padding=<ltrb 6 6 8 4>}}
  }
  ScaleBox_1 {
   widgetClassPath=&ScaleBox
   parent=HorizontalBox_2
   widget=ScaleBox_1
   properties {stretch=ScaleToFitX, stretchDirection=DownOnly}
   slot {class=HorizontalBoxSlot, properties {horizontalAlignment=HAlign_Center, verticalAlignment=VAlign_Center}}
  }
  Text_RealMoneyPrice {
   widgetClassPath=&CommonTextBlock
   bInherited=true
   parent=ScaleBox_1
   widget=Text_RealMoneyPrice
   properties {
    justification=Center
    text=3.99$ US
    textTransformPolicy=ToUpper
    colorAndOpacity {colorUseRule=UseColor_Specified, specifiedColor=<rgba 0 0 0 1>}
    font {
     bForceMonospaced=false
     bMaterialIsStencil=false
     fontMaterial=null
     fontObject=&BurbankSmall
     letterSpacing=0
     monospacedWidth=1
     size=20
     skewAmount=0
     typefaceFontName=Black
     outlineSettings {
      bApplyOutlineToDropShadows=false
      bMiteredCorners=false
      bSeparateFillAlpha=false
      outlineMaterial=null
      outlineSize=0
      outlineColor=<rgba 0 0 0 1>
     }
    }
    renderTransform {angle=0, scale=<1 1>, shear=<-8 0>, translation=<0 0>}
    shadowOffset=<0 0>
   }
   slot {class=ScaleBoxSlot, properties {horizontalAlignment=HAlign_Fill, verticalAlignment=VAlign_Fill}}
  }
  CommonVisibilityWidget_0 {
   widgetClassPath=&CommonVisibilityWidgetLegacy
   parent=HorizontalBox_2
   widget=CommonVisibilityWidget_0
   properties {
    bShowForAndroid=false
    bShowForErebus=false
    bShowForIOS=false
    bShowForMac=false
    bShowForPC=false
    bShowForSwitch2=false
    bShowForTouch=false
    bShowForXBox=false
    bShowForXSX=false
    padding=<ltrb 0 0 0 0>
   }
   slot {class=HorizontalBoxSlot, properties {horizontalAlignment=HAlign_Left, verticalAlignment=VAlign_Center, padding=<ltrb 10 0 0 0>}}
  }
  PlayStationStoreLogo {
   widgetClassPath=&Image
   bIsVariable=true
   parent=CommonVisibilityWidget_0
   widget=PlayStationStoreLogo
   properties {
    visibility=SelfHitTestInvisible
    brush {
     bIsDynamicallyLoaded=false
     drawAs=Image
     imageType=NoImage
     mirroring=NoMirror
     resourceName=None
     resourceObject=/Game/UI/Frontend/ItemShop/Art/T_UI_PS_Store_Logo_64x
     tiling=NoTile
     imageSize=<40 40>
     margin=<ltrb 0 0 0 0>
     outlineSettings {
      bUseBrushTransparency=false
      roundingType=HalfHeightRadius
      width=0
      color {colorUseRule=UseColor_Specified, specifiedColor=<rgba 0 0 0 0>}
      cornerRadii=<0 0 0 0>
     }
     tintColor {colorUseRule=UseColor_Specified, specifiedColor=<rgba 1 1 1 1>}
     uVRegion {bIsValid=false, max=<0 0>, min=<0 0>}
    }
    renderTransform {angle=0, scale=<1 1>, shear=<0 0>, translation=<0 -2>}
   }
   slot {class=BorderSlot, properties {padding=<ltrb 0 0 0 0>}}
  }
  HorizontalBox_0 {
   widgetClassPath=&HorizontalBox
   parent=Overlay_0
   widget=HorizontalBox_0
   slot {class=OverlaySlot, properties {horizontalAlignment=HAlign_Fill, verticalAlignment=VAlign_Bottom}}
  }
  Spacer {
   widgetClassPath=&Spacer
   parent=HorizontalBox_0
   widget=Spacer
   slot {class=HorizontalBoxSlot, properties {size {sizeRule=Fill, value=0.485}}}
  }
  SizeBox_2 {
   widgetClassPath=&SizeBox
   parent=HorizontalBox_0
   widget=SizeBox_2
   properties {
    bOverride_HeightOverride=true
    bOverride_WidthOverride=true
    heightOverride=44
    widthOverride=44
    renderTransform {angle=0, scale=<1 1>, shear=<0 0>, translation=<0 14>}
   }
   slot {class=HorizontalBoxSlot}
  }
  InputActionWidget {
   widgetClassPath=&CommonActionWidget
   bInherited=true
   bIsVariable=true
   parent=SizeBox_2
   widget=InputActionWidget
   properties {iconRimBrush {
     bIsDynamicallyLoaded=false
     drawAs=Image
     imageType=NoImage
     mirroring=NoMirror
     resourceName=None
     resourceObject=/FortUILibrary/Foundation/Buttons/Materials/M-Button-Console-WhiteOutline
     tiling=NoTile
     imageSize=<44 44>
     margin=<ltrb 0 0 0 0>
     outlineSettings {
      bUseBrushTransparency=false
      roundingType=HalfHeightRadius
      width=0
      color {colorUseRule=UseColor_Specified, specifiedColor=<rgba 0 0 0 0>}
      cornerRadii=<0 0 0 0>
     }
     tintColor {colorUseRule=UseColor_Specified, specifiedColor=<rgba 1 1 1 1>}
     uVRegion {bIsValid=false, max=<0 0>, min=<0 0>}
    }}
   slot {class=SizeBoxSlot}
  }
  Spacer_1 {
   widgetClassPath=&Spacer
   parent=HorizontalBox_0
   widget=Spacer_1
   slot {class=HorizontalBoxSlot, properties {size {sizeRule=Fill, value=0.58}}}
  }
  Border_BonusPercentArea {
   widgetClassPath=&Border
   parent=Overlay_36
   widget=Border_BonusPercentArea
   properties {
    visibility=HitTestInvisible
    background {
     bIsDynamicallyLoaded=false
     drawAs=Box
     imageType=NoImage
     mirroring=NoMirror
     resourceName=None
     resourceObject=/FortniteCoreUI/Frontend/StoreMain/Art/M_UI_BonusTag
     tiling=NoTile
     imageSize=<32 32>
     margin=<ltrb 0.5 0 0.5 0>
     outlineSettings {
      bUseBrushTransparency=false
      roundingType=HalfHeightRadius
      width=0
      color {colorUseRule=UseColor_Specified, specifiedColor=<rgba 0 0 0 0>}
      cornerRadii=<0 0 0 0>
     }
     tintColor {colorUseRule=UseColor_Specified, specifiedColor=<rgba 1 1 1 1>}
     uVRegion {bIsValid=false, max=<0 0>, min=<0 0>}
    }
    padding=<ltrb 18 8 16 8>
   }
   slot {class=OverlaySlot, properties {horizontalAlignment=HAlign_Right, padding=<ltrb 0 -18 -12 0>}}
  }
  SB_BonusPercentArea {
   widgetClassPath=&SizeBox
   bInherited=true
   parent=Border_BonusPercentArea
   widget=SB_BonusPercentArea
   properties {bOverride_MaxDesiredWidth=true, maxDesiredWidth=256}
   slot {class=BorderSlot, properties {padding=<ltrb 18 8 16 8>}}
  }
  Text_BonusPercent {
   widgetClassPath=&CommonTextBlock
   bInherited=true
   parent=SB_BonusPercentArea
   widget=Text_BonusPercent
   properties {
    justification=Center
    text="20% Bonus!"
    textTransformPolicy=ToUpper
    font {
     bForceMonospaced=false
     bMaterialIsStencil=false
     fontMaterial=null
     fontObject=&BurbankSmall
     letterSpacing=0
     monospacedWidth=1
     size=20
     skewAmount=0
     typefaceFontName=Black
     outlineSettings {
      bApplyOutlineToDropShadows=false
      bMiteredCorners=false
      bSeparateFillAlpha=false
      outlineMaterial=null
      outlineSize=0
      outlineColor=<rgba 0 0 0 1>
     }
    }
    renderTransform {angle=0, scale=<1 1>, shear=<-8 0>, translation=<0 0>}
    shadowOffset=<0 0>
   }
   slot {class=SizeBoxSlot, properties {horizontalAlignment=HAlign_Center, verticalAlignment=VAlign_Center}}
  }
  OptionalIMG_StarterKit {widgetClassPath=&Image, bInherited=true, widget=null}
 ]
}
