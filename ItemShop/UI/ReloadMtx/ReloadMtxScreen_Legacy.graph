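# Legacy Reload V-Bucks screen widget for the item shop, built on AthenaReloadMtxScreenLegacy.
# A left panel shows the amount needed, a starter kit option and an offer list or no-offers notice, beside a camera frame and small offer preview.
# Activation plays the Intro slide/fade; BlockScreenContent swaps the options and preview for a purchase blocker.
# Construct adjusts preview padding on mobile with the Bacchus frontend enabled.
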WidgetBlueprint /ItemShop/UI/ReloadMtx/ReloadMtxScreen_Legacy
animations [Intro {
  binding_count=7
  length_seconds=0.3667
  object=&self:Intro
  sequence {
   bindings [
    IMG_Separator {
     id=#1
     possessed_class=Image
     tracks [
      {
       type=MovieScene2DTransformTrack
       sections [{
         type=MovieScene2DTransformSection
         end_frame=22
         start_frame=-1
         channels [
          Translation.X {type=MovieSceneScriptingFloatChannel, keys [[4.228,-490],[19.303,0]]}
          Translation.Y {type=MovieSceneScriptingFloatChannel}
          Angle {type=MovieSceneScriptingFloatChannel}
          Scale.X {type=MovieSceneScriptingFloatChannel}
          Scale.Y {type=MovieSceneScriptingFloatChannel}
          Shear.X {type=MovieSceneScriptingFloatChannel}
          Shear.Y {type=MovieSceneScriptingFloatChannel}
         ]
        }]
      }
      {
       type=MovieSceneFloatTrack
       display_name=Render Opacity
       sections [{
         type=MovieSceneFloatSection
         end_frame=18
         start_frame=0
         channels [None {type=MovieSceneScriptingFloatChannel, keys [[0,0],[14.228,0],[16.921,1]]}]
        }]
      }
     ]
    }
    HorizontalBox_0 {
     id=#2
     possessed_class=HorizontalBox
     tracks [
      {
       type=MovieSceneFloatTrack
       display_name=Render Opacity
       sections [{
         type=MovieSceneFloatSection
         end_frame=22
         start_frame=-1
         channels [None {type=MovieSceneScriptingFloatChannel, keys [[0.181,0],[11.327,0],[20.338,1]]}]
        }]
      }
      {
       type=MovieScene2DTransformTrack
       sections [{
         type=MovieScene2DTransformSection
         end_frame=22
         start_frame=-1
         channels [
          Translation.X {type=MovieSceneScriptingFloatChannel}
          Translation.Y {type=MovieSceneScriptingFloatChannel, keys [[0.181,120],[11.327,120],[20.338,0]]}
          Angle {type=MovieSceneScriptingFloatChannel}
          Scale.X {type=MovieSceneScriptingFloatChannel}
          Scale.Y {type=MovieSceneScriptingFloatChannel}
          Shear.X {type=MovieSceneScriptingFloatChannel}
          Shear.Y {type=MovieSceneScriptingFloatChannel}
         ]
        }]
      }
     ]
    }
    HB_SelectAnOption {
     id=#3
     possessed_class=HorizontalBox
     tracks [
      {
       type=MovieScene2DTransformTrack
       sections [{
         type=MovieScene2DTransformSection
         end_frame=20
         start_frame=0
         channels [
          Translation.X {type=MovieSceneScriptingFloatChannel}
          Translation.Y {type=MovieSceneScriptingFloatChannel, keys [[0,0],[8.081,120],[17.519,0]]}
         ]
        }]
      }
      {
       type=MovieSceneFloatTrack
       display_name=Render Opacity
       sections [{
         type=MovieSceneFloatSection
         end_frame=20
         start_frame=0
         channels [None {type=MovieSceneScriptingFloatChannel, keys [[0,0],[8.081,0],[17.519,1]]}]
        }]
      }
     ]
    }
    SmallOfferPreview_Widget {
     id=#4
     possessed_class=AthenaSmallOfferPreview_C
     tracks [
      {
       type=MovieScene2DTransformTrack
       sections [{
         type=MovieScene2DTransformSection
         end_frame=22
         start_frame=0
         channels [
          Translation.X {type=MovieSceneScriptingFloatChannel, keys [[0,0],[12.009,200],[21.446,0]]}
          Translation.Y {type=MovieSceneScriptingFloatChannel}
         ]
        }]
      }
      {
       type=MovieSceneFloatTrack
       display_name=Render Opacity
       sections [{
         type=MovieSceneFloatSection
         end_frame=22
         start_frame=0
         channels [None {type=MovieSceneScriptingFloatChannel, keys [[0,0],[12.009,0],[21.446,1]]}]
        }]
      }
     ]
    }
    IMG_Background {
     id=#5
     possessed_class=Image
     tracks [
      {
       type=MovieScene2DTransformTrack
       sections [{
         type=MovieScene2DTransformSection
         end_frame=22
         start_frame=-1
         channels [
          Translation.X {type=MovieSceneScriptingFloatChannel, keys [[4.314,-510],[19.409,0]]}
          Translation.Y {type=MovieSceneScriptingFloatChannel}
          Angle {type=MovieSceneScriptingFloatChannel}
          Scale.X {type=MovieSceneScriptingFloatChannel}
          Scale.Y {type=MovieSceneScriptingFloatChannel}
          Shear.X {type=MovieSceneScriptingFloatChannel}
          Shear.Y {type=MovieSceneScriptingFloatChannel}
         ]
        }]
      }
      {
       type=MovieSceneFloatTrack
       display_name=Render Opacity
       sections [{
         type=MovieSceneFloatSection
         end_frame=18
         start_frame=0
         channels [None {type=MovieSceneScriptingFloatChannel, keys [[4.228,0],[16.921,1]]}]
        }]
      }
     ]
    }
    ScaleBox_9 {
     id=#6
     possessed_class=ScaleBox
     tracks [
      {
       type=MovieSceneFloatTrack
       display_name=Render Opacity
       sections [{
         type=MovieSceneFloatSection
         end_frame=21
         start_frame=0
         channels [None {type=MovieSceneScriptingFloatChannel, keys [[3.494,0],[19.494,1]]}]
        }]
      }
      {
       type=MovieScene2DTransformTrack
       sections [{
         type=MovieScene2DTransformSection
         end_frame=21
         start_frame=0
         channels [
          Translation.X {type=MovieSceneScriptingFloatChannel, keys [[3.494,-460],[19.494,0]]}
          Translation.Y {type=MovieSceneScriptingFloatChannel}
          Angle {type=MovieSceneScriptingFloatChannel}
          Scale.X {type=MovieSceneScriptingFloatChannel}
          Scale.Y {type=MovieSceneScriptingFloatChannel}
          Shear.X {type=MovieSceneScriptingFloatChannel, keys [[3.494,-10],[19.494,0]]}
          Shear.Y {type=MovieSceneScriptingFloatChannel}
         ]
        }]
      }
     ]
    }
    HB_Title {
     id=#7
     possessed_class=HorizontalBox
     tracks [
      {
       type=MovieScene2DTransformTrack
       sections [{
         type=MovieScene2DTransformSection
         end_frame=19
         start_frame=0
         channels [
          Translation.X {type=MovieSceneScriptingFloatChannel, keys [[0,-460],[16,0]]}
          Translation.Y {type=MovieSceneScriptingFloatChannel}
          Angle {type=MovieSceneScriptingFloatChannel}
          Scale.X {type=MovieSceneScriptingFloatChannel}
          Scale.Y {type=MovieSceneScriptingFloatChannel}
          Shear.X {type=MovieSceneScriptingFloatChannel, keys [[0,-10],[16,0]]}
          Shear.Y {type=MovieSceneScriptingFloatChannel}
         ]
        }]
      }
      {
       type=MovieSceneFloatTrack
       display_name=Render Opacity
       sections [{type=MovieSceneFloatSection, end_frame=19, start_frame=0, channels [None {type=MovieSceneScriptingFloatChannel, keys [[0,0],[16,1]]}]}]
      }
     ]
    }
   ]
   display_rate {numerator=60}
  }
 }]
graphs [
 ContentBlocked {nodes [
   n1 = ContentBlocked!() -> (then, IsContentBlocked:bool=~, ContentBlockedText:Text) {is=title K2Node_FunctionEntry}
   > Set Data!(
    execute
    self:WBP Content Purchase Blocker Object Reference=Get WBP_ContentPurchaseBlocker() -> (WBP_ContentPurchaseBlocker:WBP Content Purchase Blocker Object Reference)
    InText:Text=@n1.ContentBlockedText
   ) -> (then) {is=title K2Node_CallFunction}
   > SetVisibility(self=Get WBP_ContentPurchaseBlocker(), InVisibility=Select(Option 0=Collapsed, Option 1=SelfHitTestInvisible, Index=@n1))
   n2 = Select(Option 0=SelfHitTestInvisible, Option 1=Collapsed, Index=@n1)
   > SetVisibility(self=Get SafeZone_2() -> (SafeZone_2:Safe Zone Object Reference), InVisibility=@n2)
   > SetVisibility(self=Get SB_OfferDetailsWidget() -> (SB_OfferDetailsWidget:Size Box Object Reference), InVisibility=@n2)
  ]}
 EventGraph {nodes [
   Event Construct()
   n1 = GetRuntimeOptions()
   > Branch(Condition=AND Boolean(A=IsMobileGame(), B=Get bBacchusFrontendEnabled!(self:Fort Runtime Options Object Reference=@n1) -> (bBacchusFrontendEnabled:bool)))
   > SetPadding!(
    execute
    self:Overlay Slot Object Reference=SlotAsOverlaySlot(Widget=Get SB_OfferDetailsWidget())
    InPadding:Margin Structure=Make Margin(Left=0, Top=0, Right=40, Bottom=140)
   ) -> (then)
   n2 = Event BlockScreenContent!() -> (OutputDelegate, then, bBlockScreen:bool, ContentBlockedText:"Text (by ref)") {is=title K2Node_Event}
   > ContentBlocked!(execute, self, IsContentBlocked:bool=@n2.bBlockScreen, ContentBlockedText:Text=@n2.ContentBlockedText) -> (then) {is=title K2Node_CallFunction}
   Event On Activated()
   > PlayAnimation(InAnimation=Get Intro())
  ]}
]
variables []
widgets {
 info {inheritedWidgetCount=14, parentClass=/Script/FortniteUI.AthenaReloadMtxScreenLegacy, rootWidgetClass=&Overlay, widgetCount=64}
 widgets [
  Overlay_0 {widgetClassPath=&Overlay, widget=Overlay_0}
  ScaleBox_0 {
   widgetClassPath=&ScaleBox
   parent=Overlay_0
   widget=ScaleBox_0
   properties {stretch=ScaleBySafeZone}
   slot {class=OverlaySlot, properties {horizontalAlignment=HAlign_Fill, verticalAlignment=VAlign_Fill}}
  }
  Overlay_2 {
   widgetClassPath=&Overlay
   parent=ScaleBox_0
   widget=Overlay_2
   slot {class=ScaleBoxSlot, properties {horizontalAlignment=HAlign_Fill, verticalAlignment=VAlign_Fill}}
  }
  HorizontalBox_2 {
   widgetClassPath=&HorizontalBox
   parent=Overlay_2
   widget=HorizontalBox_2
   slot {class=OverlaySlot, properties {horizontalAlignment=HAlign_Fill, verticalAlignment=VAlign_Fill}}
  }
  SB_MainContainer {
   widgetClassPath=&SizeBox
   parent=HorizontalBox_2
   widget=SB_MainContainer
   properties {widthOverride=900}
   slot {class=HorizontalBoxSlot, properties {size {sizeRule=Fill, value=0.45}}}
  }
  Overlay_3 {widgetClassPath=&Overlay, parent=SB_MainContainer, widget=Overlay_3, slot {class=SizeBoxSlot}}
  IMG_Separator {
   widgetClassPath=&Image
   bIsVariable=true
   parent=Overlay_3
   widget=IMG_Separator
   properties {
    visibility=SelfHitTestInvisible
    brush {
     bIsDynamicallyLoaded=false
     drawAs=Box
     imageType=NoImage
     mirroring=NoMirror
     resourceName=None
     resourceObject=&/MI_UI_ReloadVBucks_Separator_Legacy
     tiling=NoTile
     imageSize=<128 32>
     margin=<ltrb 0 0 0.001 0>
     outlineSettings {
      bUseBrushTransparency=false
      roundingType=HalfHeightRadius
      width=0
      color {colorUseRule=UseColor_Specified, specifiedColor=<rgba 0 0 0 0>}
      cornerRadii=<0 0 0 0>
     }
     tintColor {colorUseRule=UseColor_Specified, specifiedColor=<rgba 1 1 1 1>}
     uVRegion {bIsValid=false, max=<0 0>, min=<0 0>}
    }
   }
   slot {class=OverlaySlot, properties {horizontalAlignment=HAlign_Right, verticalAlignment=VAlign_Fill, padding=<ltrb 0 0 -30 0>}}
  }
  IMG_Background {
   widgetClassPath=&Image
   bIsVariable=true
   parent=Overlay_3
   widget=IMG_Background
   properties {
    visibility=SelfHitTestInvisible
    brush {
     bIsDynamicallyLoaded=false
     drawAs=Box
     imageType=NoImage
     mirroring=NoMirror
     resourceName=None
     resourceObject=&/MI_UI_ReloadVBucks_BG_Legacy
     tiling=NoTile
     imageSize=<32 32>
     margin=<ltrb 0 0 0.001 0>
     outlineSettings {
      bUseBrushTransparency=false
      roundingType=HalfHeightRadius
      width=0
      color {colorUseRule=UseColor_Specified, specifiedColor=<rgba 0 0 0 0>}
      cornerRadii=<0 0 0 0>
     }
     tintColor {colorUseRule=UseColor_Specified, specifiedColor=<rgba 1 1 1 1>}
     uVRegion {bIsValid=false, max=<0 0>, min=<0 0>}
    }
   }
   slot {class=OverlaySlot, properties {horizontalAlignment=HAlign_Fill, verticalAlignment=VAlign_Fill}}
  }
  SafeZone_2 {
   widgetClassPath=&SafeZone
   bIsVariable=true
   parent=Overlay_3
   widget=SafeZone_2
   properties {padRight=false}
   slot {class=OverlaySlot, properties {horizontalAlignment=HAlign_Fill, verticalAlignment=VAlign_Fill, padding=<ltrb 0 0 20 0>}}
  }
  VerticalBox_0 {widgetClassPath=&VerticalBox, parent=SafeZone_2, widget=VerticalBox_0, slot {class=SafeZoneSlot}}
  VerticalBox_1 {
   widgetClassPath=&VerticalBox
   parent=VerticalBox_0
   widget=VerticalBox_1
   slot {class=VerticalBoxSlot, properties {padding=<ltrb 30 0 30 0>, size {sizeRule=Fill, value=0.3}}}
  }
  HB_Title {
   widgetClassPath=&HorizontalBox
   parent=VerticalBox_1
   widget=HB_Title
   slot {class=VerticalBoxSlot, properties {horizontalAlignment=HAlign_Center, padding=<ltrb 0 10 0 0>, size {sizeRule=Automatic, value=0.3}}}
  }
  IMG_ReloadVBucksLogo {
   widgetClassPath=&Image
   bIsVariable=true
   parent=HB_Title
   widget=IMG_ReloadVBucksLogo
   properties {
    brush {
     bIsDynamicallyLoaded=false
     drawAs=Image
     imageType=NoImage
     mirroring=NoMirror
     resourceName=None
     resourceObject=&/MI_ReloadVBucks_GradientLogo_Legacy
     tiling=NoTile
     imageSize=<168 168>
     margin=<ltrb 0 0 0 0>
     outlineSettings {
      bUseBrushTransparency=false
      roundingType=HalfHeightRadius
      width=0
      color {colorUseRule=UseColor_Specified, specifiedColor=<rgba 0 0 0 0>}
      cornerRadii=<0 0 0 0>
     }
     tintColor {colorUseRule=UseColor_Specified, specifiedColor=<rgba 1 1 1 1>}
     uVRegion {bIsValid=false, max=<0 0>, min=<0 0>}
    }
    colorAndOpacity=<rgba 1 1 1 0.6>
   }
   slot {class=HorizontalBoxSlot, properties {horizontalAlignment=HAlign_Center, verticalAlignment=VAlign_Center}}
  }
  ScaleBox_6 {
   widgetClassPath=&ScaleBox
   parent=HB_Title
   widget=ScaleBox_6
   properties {stretchDirection=DownOnly, renderTransform {angle=0, scale=<1 1>, shear=<-12 0>, translation=<0 0>}}
   slot {class=HorizontalBoxSlot, properties {verticalAlignment=VAlign_Center, padding=<ltrb 16 0 10 0>, size {sizeRule=Fill, value=1}}}
  }
  CommonTextBlock_Title {
   widgetClassPath=&CommonTextBlock
   parent=ScaleBox_6
   widget=CommonTextBlock_Title
   properties {
    text=Reload V-Bucks
    textTransformPolicy=ToUpper
    font {
     bForceMonospaced=false
     bMaterialIsStencil=false
     fontMaterial=&/MI_UI_ReloadMtx_Title_Legacy
     fontObject=&Burbank
     letterSpacing=0
     monospacedWidth=1
     size=66
     skewAmount=0
     typefaceFontName=BurbankRegular-Black
     outlineSettings {
      bApplyOutlineToDropShadows=false
      bMiteredCorners=false
      bSeparateFillAlpha=false
      outlineMaterial=null
      outlineSize=0
      outlineColor=<rgba 0 0 0 1>
     }
    }
    shadowOffset=<0 0>
   }
   slot {class=ScaleBoxSlot, properties {horizontalAlignment=HAlign_Fill}}
  }
  ScaleBox_9 {
   widgetClassPath=&ScaleBox
   parent=VerticalBox_1
   widget=ScaleBox_9
   properties {stretchDirection=DownOnly}
   slot {class=VerticalBoxSlot, properties {padding=<ltrb 0 0 0 30>, size {sizeRule=Automatic, value=0.16}}}
  }
  HB_AmountNeeded {widgetClassPath=&HorizontalBox, parent=ScaleBox_9, widget=HB_AmountNeeded, slot {class=ScaleBoxSlot}}
  Text_AdditionalFunds {
   widgetClassPath=&CommonTextBlock
   parent=HB_AmountNeeded
   widget=Text_AdditionalFunds
   properties {
    justification=Right
    text=Additional V-Bucks Needed:
    font {
     bForceMonospaced=false
     bMaterialIsStencil=false
     fontMaterial=null
     fontObject=&BurbankSmall
     letterSpacing=0
     monospacedWidth=1
     size=18
     skewAmount=0
     typefaceFontName=Bold
     outlineSettings {
      bApplyOutlineToDropShadows=false
      bMiteredCorners=false
      bSeparateFillAlpha=false
      outlineMaterial=null
      outlineSize=0
      outlineColor=<rgba 0 0 0 1>
     }
    }
    shadowOffset=<0 0>
   }
   slot {class=HorizontalBoxSlot, properties {horizontalAlignment=HAlign_Right, verticalAlignment=VAlign_Center}}
  }
  IMG_ReloadVBucksIcon_1 {
   widgetClassPath=&Image
   bIsVariable=true
   parent=HB_AmountNeeded
   widget=IMG_ReloadVBucksIcon_1
   properties {brush {
     bIsDynamicallyLoaded=false
     drawAs=Image
     imageType=NoImage
     mirroring=NoMirror
     resourceName=None
     resourceObject=/Game/UI/Foundation/Shell/Textures/T-Icon-VBucks-L
     tiling=NoTile
     imageSize=<30 30>
     margin=<ltrb 0 0 0 0>
     outlineSettings {
      bUseBrushTransparency=false
      roundingType=HalfHeightRadius
      width=0
      color {colorUseRule=UseColor_Specified, specifiedColor=<rgba 0 0 0 0>}
      cornerRadii=<0 0 0 0>
     }
     tintColor {colorUseRule=UseColor_Specified, specifiedColor=<rgba 1 1 1 1>}
     uVRegion {bIsValid=false, max=<0 0>, min=<0 0>}
    }}
   slot {class=HorizontalBoxSlot, properties {horizontalAlignment=HAlign_Center, verticalAlignment=VAlign_Center, padding=<ltrb 26 0 16 0>}}
  }
  Text_AmountNeeded {
   widgetClassPath=&CommonTextBlock
   bInherited=true
   parent=HB_AmountNeeded
   widget=Text_AmountNeeded
   properties {
    text=400
    font {
     bForceMonospaced=false
     bMaterialIsStencil=false
     fontMaterial=null
     fontObject=&BurbankSmall
     letterSpacing=0
     monospacedWidth=1
     size=18
     skewAmount=0
     typefaceFontName=Black
     outlineSettings {
      bApplyOutlineToDropShadows=false
      bMiteredCorners=false
      bSeparateFillAlpha=false
      outlineMaterial=null
      outlineSize=0
      outlineColor=<rgba 0 0 0 1>
     }
    }
    shadowOffset=<0 0>
   }
   slot {class=HorizontalBoxSlot, properties {verticalAlignment=VAlign_Center}}
  }
  Text_V-Bucks {
   widgetClassPath=&CommonTextBlock
   parent=HB_AmountNeeded
   widget=Text_V-Bucks
   properties {
    text=V-Bucks
    textTransformPolicy=ToUpper
    font {
     bForceMonospaced=false
     bMaterialIsStencil=false
     fontMaterial=null
     fontObject=&BurbankSmall
     letterSpacing=0
     monospacedWidth=1
     size=18
     skewAmount=0
     typefaceFontName=Black
     outlineSettings {
      bApplyOutlineToDropShadows=false
      bMiteredCorners=false
      bSeparateFillAlpha=false
      outlineMaterial=null
      outlineSize=0
      outlineColor=<rgba 0 0 0 1>
     }
    }
    shadowOffset=<0 0>
   }
   slot {class=HorizontalBoxSlot, properties {verticalAlignment=VAlign_Center, padding=<ltrb 10 0 0 0>}}
  }
  VB_Options {
   widgetClassPath=&VerticalBox
   parent=VerticalBox_0
   widget=VB_Options
   slot {class=VerticalBoxSlot, properties {size {sizeRule=Fill, value=1}}}
  }
  SizeBox_0 {
   widgetClassPath=&SizeBox
   parent=VB_Options
   widget=SizeBox_0
   properties {bOverride_MinDesiredWidth=true, minDesiredWidth=800}
   slot {class=VerticalBoxSlot, properties {horizontalAlignment=HAlign_Center, padding=<ltrb 30 30 30 0>, size {sizeRule=Automatic, value=0.1}}}
  }
  HB_SelectAnOption {widgetClassPath=&HorizontalBox, parent=SizeBox_0, widget=HB_SelectAnOption, slot {class=SizeBoxSlot}}
  IMG_LeftStroke {
   widgetClassPath=&Image
   bIsVariable=true
   parent=HB_SelectAnOption
   widget=IMG_LeftStroke
   properties {brush {
     bIsDynamicallyLoaded=false
     drawAs=Box
     imageType=NoImage
     mirroring=NoMirror
     resourceName=None
     resourceObject=&/MI_UI_ReloadVBucks_Lines2_Legacy
     tiling=NoTile
     imageSize=<32 12>
     margin=<ltrb 0.115 0 0.11 0>
     outlineSettings {
      bUseBrushTransparency=false
      roundingType=HalfHeightRadius
      width=0
      color {colorUseRule=UseColor_Specified, specifiedColor=<rgba 0 0 0 0>}
      cornerRadii=<0 0 0 0>
     }
     tintColor {colorUseRule=UseColor_Specified, specifiedColor=<rgba 1 1 1 1>}
     uVRegion {bIsValid=false, max=<0 0>, min=<0 0>}
    }}
   slot {class=HorizontalBoxSlot, properties {verticalAlignment=VAlign_Center, size {sizeRule=Fill, value=0.5}}}
  }
  CommonTextBlock_628 {
   widgetClassPath=&CommonTextBlock
   parent=HB_SelectAnOption
   widget=CommonTextBlock_628
   properties {
    justification=Center
    text=Select an option
    textTransformPolicy=ToUpper
    font {
     bForceMonospaced=false
     bMaterialIsStencil=false
     fontMaterial=null
     fontObject=&Burbank
     letterSpacing=0
     monospacedWidth=1
     size=24
     skewAmount=0
     typefaceFontName=Burbank-PlayerPower
     outlineSettings {
      bApplyOutlineToDropShadows=false
      bMiteredCorners=false
      bSeparateFillAlpha=false
      outlineMaterial=null
      outlineSize=0
      outlineColor=<rgba 0 0 0 1>
     }
    }
    renderTransform {angle=0, scale=<1 0.9>, shear=<-12 0>, translation=<0 0>}
    shadowOffset=<0 0>
   }
   slot {class=HorizontalBoxSlot, properties {verticalAlignment=VAlign_Center, padding=<ltrb 20 0 20 0>}}
  }
  IMG_RightStroke {
   widgetClassPath=&Image
   bIsVariable=true
   parent=HB_SelectAnOption
   widget=IMG_RightStroke
   properties {brush {
     bIsDynamicallyLoaded=false
     drawAs=Box
     imageType=NoImage
     mirroring=NoMirror
     resourceName=None
     resourceObject=&MI_UI_ReloadVBucks_Lines2_Legacy
     tiling=NoTile
     imageSize=<32 12>
     margin=<ltrb 0.115 0 0.11 0>
     outlineSettings {
      bUseBrushTransparency=false
      roundingType=HalfHeightRadius
      width=0
      color {colorUseRule=UseColor_Specified, specifiedColor=<rgba 0 0 0 0>}
      cornerRadii=<0 0 0 0>
     }
     tintColor {colorUseRule=UseColor_Specified, specifiedColor=<rgba 1 1 1 1>}
     uVRegion {bIsValid=false, max=<0 0>, min=<0 0>}
    }}
   slot {class=HorizontalBoxSlot, properties {verticalAlignment=VAlign_Center, size {sizeRule=Fill, value=0.5}}}
  }
  Switcher_AvailableOptions {
   widgetClassPath=&CommonVisibilitySwitcher
   bInherited=true
   parent=VB_Options
   widget=Switcher_AvailableOptions
   slot {class=VerticalBoxSlot, properties {size {sizeRule=Fill, value=1}}}
  }
  Overlay_AvailableOptions {
   widgetClassPath=&Overlay
   bInherited=true
   parent=Switcher_AvailableOptions
   widget=Overlay_AvailableOptions
   slot {class=CommonVisibilitySwitcherSlot, properties {verticalAlignment=VAlign_Center}}
  }
  ScaleBox_1 {
   widgetClassPath=&ScaleBox
   parent=Overlay_AvailableOptions
   widget=ScaleBox_1
   properties {stretch=ScaleToFitY, stretchDirection=DownOnly}
   slot {class=OverlaySlot, properties {horizontalAlignment=HAlign_Fill}}
  }
  HorizontalBox_0 {
   widgetClassPath=&HorizontalBox
   parent=ScaleBox_1
   widget=HorizontalBox_0
   slot {class=ScaleBoxSlot, properties {horizontalAlignment=HAlign_Fill, verticalAlignment=VAlign_Fill}}
  }
  Spacer_194 {
   widgetClassPath=&Spacer
   parent=HorizontalBox_0
   widget=Spacer_194
   slot {class=HorizontalBoxSlot, properties {size {sizeRule=Fill, value=0.05}}}
  }
  Overlay_1 {
   widgetClassPath=&Overlay
   parent=HorizontalBox_0
   widget=Overlay_1
   slot {class=HorizontalBoxSlot, properties {padding=<ltrb 0 40 0 0>, size {sizeRule=Fill, value=1}}}
  }
  SizeBox_3 {
   widgetClassPath=&SizeBox
   parent=Overlay_1
   widget=SizeBox_3
   properties {bOverride_MaxDesiredWidth=true, maxDesiredWidth=950}
   slot {class=OverlaySlot, properties {horizontalAlignment=HAlign_Center, verticalAlignment=VAlign_Fill}}
  }
  VerticalBox_2 {widgetClassPath=&VerticalBox, parent=SizeBox_3, widget=VerticalBox_2, slot {class=SizeBoxSlot}}
  Button_StarterKitOption {
   widgetClassPath=&/MtxStarterKitListButton_Legacy
   bInherited=true
   bIsVariable=true
   parent=VerticalBox_2
   widget=Button_StarterKitOption
   properties {
    onButtonBaseClicked=(&self:WidgetTree.Button_StarterKitOption.HandleOnButtonClicked)
    onButtonBaseDoubleClicked=(&self:WidgetTree.Button_StarterKitOption.HandleOnButtonDoubleClicked)
    onButtonBaseHovered=(&self:WidgetTree.Button_StarterKitOption.HandleOnButtonHovered)
    onButtonBaseUnhovered=(&self:WidgetTree.Button_StarterKitOption.HandleOnButtonUnhovered)
    onSelectedChangedBase=(&self:WidgetTree.Button_StarterKitOption.HandleOnSelectedChanged)
   }
   slot {class=VerticalBoxSlot, properties {padding=<ltrb 30 0 30 42>}}
  }
  ListView_MtxOptions {
   widgetClassPath=&CommonListView
   bInherited=true
   bIsVariable=true
   parent=VerticalBox_2
   widget=ListView_MtxOptions
   properties {
    clipping=Inherit
    entryWidgetClass=&/MtxOfferListEntryButton_Legacy.MtxOfferListEntryButton_Legacy_C
    horizontalEntrySpacing=42
    navigation=&self:WidgetTree.ListView_MtxOptions.WidgetNavigation_0
    numDesignerPreviewEntries=4
    verticalEntrySpacing=42
   }
   slot {class=VerticalBoxSlot, properties {padding=<ltrb 30 0 30 50>, size {sizeRule=Fill, value=1}}}
  }
  Spacer {
   widgetClassPath=&Spacer
   parent=HorizontalBox_0
   widget=Spacer
   slot {class=HorizontalBoxSlot, properties {size {sizeRule=Fill, value=0.05}}}
  }
  Overlay_NoOptions {
   widgetClassPath=&Overlay
   bInherited=true
   parent=Switcher_AvailableOptions
   widget=Overlay_NoOptions
   slot {class=CommonVisibilitySwitcherSlot, properties {horizontalAlignment=HAlign_Center, verticalAlignment=VAlign_Center}}
  }
  NoOffersBox {
   widgetClassPath=&CommonBorder
   bIsVariable=true
   parent=Overlay_NoOptions
   widget=NoOffersBox
   properties {
    horizontalAlignment=HAlign_Center
    style=/FortUILibrary/Styles/Border/Border_StatRow_WhiteLowOpacity.Border_StatRow_WhiteLowOpacity_C
    padding=<ltrb 0 0 0 0>
   }
   slot {class=OverlaySlot}
  }
  SizeBox {
   widgetClassPath=&SizeBox
   parent=NoOffersBox
   widget=SizeBox
   properties {bOverride_HeightOverride=true, heightOverride=48}
   slot {class=BorderSlot, properties {horizontalAlignment=HAlign_Center, padding=<ltrb 0 0 0 0>}}
  }
  CommonTextBlock_2 {
   widgetClassPath=&CommonTextBlock
   parent=SizeBox
   widget=CommonTextBlock_2
   properties {
    style=/FortUILibrary/Styles/Typography/Foundation/TextStyle-Base-XS-B-70pc.TextStyle-Base-XS-B-70pc_C
    text=No Offers Available
    colorAndOpacity {colorUseRule=UseColor_Specified, specifiedColor=<rgba 1 1 1 0.7>}
    font {
     bForceMonospaced=false
     bMaterialIsStencil=false
     fontMaterial=null
     fontObject=&NotoSans
     letterSpacing=0
     monospacedWidth=1
     size=16
     skewAmount=0
     typefaceFontName=NotoSans-Bold
     outlineSettings {
      bApplyOutlineToDropShadows=false
      bMiteredCorners=false
      bSeparateFillAlpha=false
      outlineMaterial=null
      outlineSize=0
      outlineColor=<rgba 0 0 0 1>
     }
    }
    shadowOffset=<0 0>
   }
   slot {class=SizeBoxSlot, properties {padding=<ltrb 32 8 32 8>}}
  }
  WBP_ContentPurchaseBlocker {
   widgetClassPath=/FortniteCoreUI/UI/ContentControls/WBP_ContentPurchaseBlocker
   bIsVariable=true
   parent=Overlay_3
   widget=WBP_ContentPurchaseBlocker
   properties {visibility=Collapsed}
   slot {class=OverlaySlot, properties {horizontalAlignment=HAlign_Center, verticalAlignment=VAlign_Center}}
  }
  SafeZone_3 {
   widgetClassPath=&SafeZone
   bIsVariable=true
   parent=HorizontalBox_2
   widget=SafeZone_3
   properties {padLeft=false}
   slot {class=HorizontalBoxSlot, properties {size {sizeRule=Fill, value=0.5}}}
  }
  OfferDetails_Container {widgetClassPath=&Overlay, parent=SafeZone_3, widget=OfferDetails_Container, slot {class=SafeZoneSlot}}
  HorizontalBox_3 {
   widgetClassPath=&HorizontalBox
   parent=OfferDetails_Container
   widget=HorizontalBox_3
   slot {class=OverlaySlot, properties {horizontalAlignment=HAlign_Fill, verticalAlignment=VAlign_Fill}}
  }
  FramingWidget {
   widgetClassPath=&SizeBox
   bIsVariable=true
   parent=HorizontalBox_3
   widget=FramingWidget
   slot {class=HorizontalBoxSlot, properties {padding=<ltrb -181 24 -361 24>, size {sizeRule=Fill, value=1}}}
  }
  Widget_cameraframe {widgetClassPath=&Overlay, bInherited=true, parent=FramingWidget, widget=Widget_cameraframe, slot {class=SizeBoxSlot}}
  Spacer_50 {widgetClassPath=&Spacer, parent=HorizontalBox_3, widget=Spacer_50, properties {size=<440 1>}, slot {class=HorizontalBoxSlot}}
  SB_OfferDetailsWidget {
   widgetClassPath=&SizeBox
   bIsVariable=true
   parent=OfferDetails_Container
   widget=SB_OfferDetailsWidget
   properties {maxDesiredWidth=102.5564, visibility=HitTestInvisible, widthOverride=500}
   slot {class=OverlaySlot, properties {horizontalAlignment=HAlign_Right, verticalAlignment=VAlign_Bottom, padding=<ltrb 0 0 40 70>}}
  }
  Overlay {widgetClassPath=&Overlay, parent=SB_OfferDetailsWidget, widget=Overlay, slot {class=SizeBoxSlot}}
  SmallOfferPreview_Widget {
   widgetClassPath=/FortniteCoreUI/Frontend/AthenaSmallOfferPreview
   bInherited=true
   bIsVariable=true
   parent=Overlay
   widget=SmallOfferPreview_Widget
   slot {class=OverlaySlot, properties {horizontalAlignment=HAlign_Fill, verticalAlignment=VAlign_Fill}}
  }
  HorizontalBox_1 {
   widgetClassPath=&HorizontalBox
   parent=Overlay_2
   widget=HorizontalBox_1
   slot {class=OverlaySlot, properties {horizontalAlignment=HAlign_Right, padding=<ltrb 0 24 24 0>}}
  }
  MTXButtonBox {
   widgetClassPath=&SizeBox
   bIsVariable=true
   parent=HorizontalBox_1
   widget=MTXButtonBox
   slot {class=HorizontalBoxSlot, properties {horizontalAlignment=HAlign_Right}}
  }
  MTXButton {
   widgetClassPath=/Game/UI/Frontend/TopBar/MTXButton
   bIsVariable=true
   parent=MTXButtonBox
   widget=MTXButton
   properties {
    onButtonBaseClicked=(&self:WidgetTree.MTXButton.HandleOnButtonClicked)
    onButtonBaseDoubleClicked=(&self:WidgetTree.MTXButton.HandleOnButtonDoubleClicked)
    onButtonBaseHovered=(&self:WidgetTree.MTXButton.HandleOnButtonHovered)
    onButtonBaseUnhovered=(&self:WidgetTree.MTXButton.HandleOnButtonUnhovered)
    onSelectedChangedBase=(&self:WidgetTree.MTXButton.HandleOnSelectedChanged)
    visibility=HitTestInvisible
   }
   slot {class=SizeBoxSlot}
  }
  VisibilityBorder_Touch {
   widgetClassPath=&CommonHardwareVisibilityBorder
   parent=HorizontalBox_1
   widget=VisibilityBorder_Touch
   properties {visibilityQuery {
     autoDescription=" ANY( Input.Touch )"
     tokenStreamVersion=0
     userDescription=""
     queryTokenStream [0,1,1,1,0]
     tagDictionary [{tagName=Input.Touch}]
    }}
   slot {class=HorizontalBoxSlot, properties {padding=<ltrb 8 0 0 0>}}
  }
  VerticalBox_3 {
   widgetClassPath=&VerticalBox
   parent=VisibilityBorder_Touch
   widget=VerticalBox_3
   slot {class=BorderSlot, properties {padding=<ltrb 0 0 0 0>}}
  }
  ForcenonMobilePosition {
   widgetClassPath=&Spacer
   parent=VerticalBox_3
   widget=ForcenonMobilePosition
   properties {size=<48 0>}
   slot {class=VerticalBoxSlot, properties {horizontalAlignment=HAlign_Right, verticalAlignment=VAlign_Top}}
  }
  Button_CloseTouch {
   widgetClassPath=/FortUILibrary/Foundation/Buttons/CloseButton
   bInherited=true
   bIsVariable=true
   parent=VerticalBox_3
   widget=Button_CloseTouch
   properties {
    onButtonBaseClicked=(&self:WidgetTree.Button_CloseTouch.HandleOnButtonClicked)
    onButtonBaseDoubleClicked=(&self:WidgetTree.Button_CloseTouch.HandleOnButtonDoubleClicked)
    onButtonBaseHovered=(&self:WidgetTree.Button_CloseTouch.HandleOnButtonHovered)
    onButtonBaseUnhovered=(&self:WidgetTree.Button_CloseTouch.HandleOnButtonUnhovered)
    onSelectedChangedBase=(&self:WidgetTree.Button_CloseTouch.HandleOnSelectedChanged)
   }
   slot {class=VerticalBoxSlot, properties {padding=<ltrb 14 0 0 0>}}
  }
  Button_PreviewAction {widgetClassPath=&CommonButtonBase, bInherited=true, widget=null}
  Button_PreviewAction1 {widgetClassPath=&CommonButtonBase, bInherited=true, widget=null}
  Button_Camera {widgetClassPath=&CommonButtonBase, bInherited=true, widget=null}
  ItemDownloadIndicator {widgetClassPath=/Script/FortniteUI.FortItemDownloadIndicator, bInherited=true, widget=null}
  Text_DebugPreviewDetails {widgetClassPath=&CommonTextBlock, bInherited=true, widget=null}
 ]
}
